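# UB_Goto3DHBDoc is an editor utility Blueprint that opens the 3DHB documentation from within the Unreal editor.
# Its Goto3DHBDocument function calls LaunchURL with a hardcoded Google Docs link.
# The event graph is empty and the Blueprint has no variables.
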EditorUtilityBlueprint /Game/Tools/3DHB/UB_Goto3DHBDoc
graphs [
 EventGraph {links [], nodes []}
 Goto3DHBDocument {nodes [
   Goto3DHBDocument() {is=title K2Node_FunctionEntry}
   > LaunchURL!(execute, URL:String="https://docs.google.com/document/d/1n2ky6_yNPVxs2t9xfvO0Em8pdDumYaLC9WOOD14qias/edit?usp=sharing") -> (then)
  ]}
]
variables []
