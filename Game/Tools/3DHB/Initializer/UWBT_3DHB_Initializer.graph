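# Editor utility widget that initializes an asset workspace for a 3D Hologram Backpack (3DHB) data asset.
# A DetailsView bound to the widget exposes its settings; the button creates folders beside the target asset and duplicates template textures and a material instance, named by task.
# It writes TaskName and MaterialInstance to the asset, but not the textures.
# Folder checks set readiness, broadcast WhenStatChange and disable the button when ready or untargeted.
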EditorUtilityWidgetBlueprint /Game/Tools/3DHB/Initializer/UWBT_3DHB_Initializer
graphs [
 CreateWorkspace {nodes [
   n1 = Sequence!(execute) -> (then_0, then_1, then_2:Exec, then_3:Exec, then_4:Exec)
   n2 = Append(B=T_3DHB_ForCopy)
   n3 = Get Target3DHBDataAsset() -> (Target3DHBDataAsset:Primary Data Asset Object Reference)
   n4 = Get Asset Object Folder Path!(execute=@n17.then_0, self, TargetObj:Object Reference=@n3) -> (then, Path:String) {is=title K2Node_CallFunction}
   n5 = Append(A=@n4, B=Texture/)
   n6 = Get InitialWorkSpace_TaskName() -> (InitialWorkSpace_TaskName:String)
   n7 = Append!(A=@n5, B=T_, C:String=@n6, D:String=_WPO) -> (ReturnValue)
   DuplicateAsset!(execute=@n1.then_0, SourceAssetPath:String=@n2, DestinationAssetPath:String=@n7) -> (then, ReturnValue:Object Reference)
   n8 = Append(A=@n5, B=T_, C=@n6, D=_Normal)
   DuplicateAsset(execute=@n1.then_1, SourceAssetPath=@n2, DestinationAssetPath=@n8)
   n9 = Append(A=@n5, B=T_, C=@n6, D=_MeshMask)
   DuplicateAsset(execute=@n1.then_2, SourceAssetPath=@n2, DestinationAssetPath=@n9)
   n10 = Append(A=@n5, B=T_, C=@n6, D=_Diffuse)
   DuplicateAsset(execute=@n1.then_3, SourceAssetPath=@n2, DestinationAssetPath=@n10)
   n11 = Cast To 3DHologramBackpackDataAsset!(execute=@n1.then_4, Object=@n3) -> (then, CastFailed, As3DHologram Backpack Data Asset:3DHologram Backpack Data Asset Object Reference) {is=title K2Node_DynamicCast}
   > n12 = Sequence()
   > n13 = LoadAsset(AssetPath=@n7)
   n14 = LoadAsset(execute=@n12.then_1, AssetPath=@n8)
   n15 = LoadAsset(execute=@n12.then_2, AssetPath=@n9)
   n16 = LoadAsset(execute=@n12.then_3, AssetPath=@n10)
   Set TaskName!(execute=@n12.then_4, TaskName:String=Get InitialWorkSpace_TaskName(), self:3DHologram Backpack Data Asset Object Reference=@n11) -> (then, Output_Get:String)
   CreateWorkspace() {is=title K2Node_FunctionEntry}
   > Is Valid(InputObject=@n3)
   > n17 = Sequence!(execute) -> (then_0, then_1, then_2)
   n18 = Get Asset Object Folder Path(execute=@n17.then_2, TargetObj=Get ForCopyMaterialInstance() -> (ForCopyMaterialInstance:Material Instance Object Reference))
   > n19 = DuplicateAsset(
    SourceAssetPath=Append!(A=@n18, B=MI_3DHB_ForCopy) -> (ReturnValue)
    DestinationAssetPath=Append!(A=Append!(A=@n4, B=Material/) -> (ReturnValue), B=MI_, C:String=@n6) -> (ReturnValue)
   )
   > n20 = Cast To 3DHologramBackpackDataAsset(Object=@n3)
   > Set MaterialInstance!(
    execute
    MaterialInstance:Material Instance Object Reference=Cast To MaterialInstance!(Object=@n19) -> (AsMaterial Instance:Material Instance Object Reference, bSuccess:bool=~) {is=title K2Node_DynamicCast}
    self:3DHologram Backpack Data Asset Object Reference=@n20
   ) -> (then, Output_Get:Material Instance Object Reference)
   n21 = For Each Loop(Exec=@n4.then, Array=Get FolderList() -> (FolderList:Array of Strings))
   > MakeDirectory!(execute, DirectoryPath:String=Append(A=@n4, B=@n21)) -> (then, ReturnValue:bool)
   Cast To Texture2D!(Object=@n16) -> (AsTexture 2D, bSuccess:bool=~)
   Cast To Texture2D(Object=@n15) -> (bSuccess=~)
   Cast To Texture2D(Object=@n14) -> (bSuccess=~)
   Cast To Texture2D(Object=@n13) -> (bSuccess=~)
  ]}
 DoesHaveExistDirectory {nodes [
   DoesHaveExistDirectory() {is=title K2Node_FunctionEntry}
   > Set bWorkspaceExist(bWorkspaceExist:bool=true) -> (Output_Get:bool)
   n1 = Get Target3DHBDataAsset()
   > n2 = Is Valid(InputObject=@n1)
   Return Node!(execute=[@n5.else, @n2.Is Not Valid], bool:bool=false)
   > n3 = Get Asset Object Folder Path(TargetObj=@n1)
   > n4 = For Each Loop(Array=Get FolderList())
   n5 = Branch(execute=@n4.Completed, Condition=Get bWorkspaceExist() -> (bWorkspaceExist:bool))
   > Return Node(bool=true)
   n6 = DoesDirectoryExist!(execute=@n4.LoopBody, DirectoryPath:String=Append(A=@n3, B=@n4)) -> (then, ReturnValue:bool)
   > Set bWorkspaceExist(bWorkspaceExist=AND Boolean(A=Get bWorkspaceExist(), B=@n6))
  ]}
 EventGraph {nodes [
   "On Clicked (CreateNewWorkspace)"() {is=title K2Node_ComponentBoundEvent}
   > Is Valid(InputObject=Get Target3DHBDataAsset())
   > CreateWorkspace() {is=title K2Node_CallFunction}
   > n1 = Set TaskName(
    TaskName=Get InitialWorkSpace_TaskName()
    self=Cast To 3DHologramBackpackDataAsset!(Object=Get Target3DHBDataAsset()) -> (As3DHologram Backpack Data Asset, bSuccess:bool=~)
   )
   n2 = Sequence!(execute=[@n11.then, @n1.then]) -> (then_0, then_1)
   n3 = Get Target3DHBDataAsset()
   > Is Valid(InputObject=@n3)
   n4 = Get TaskName!(self:3DHologram Backpack Data Asset Object Reference=Cast To 3DHologramBackpackDataAsset(Object=@n3) -> (bSuccess=~)) -> (TaskName:String)
   > n5 = Branch(Condition=IsEmpty(InString=@n4))
   Set InitialWorkSpace_TaskName(execute=@n5.else, InitialWorkSpace_TaskName:String=@n4) -> (Output_Get:String)
   n6 = DoesHaveExistDirectory!(execute=[@n12.then, @n2.then_1], self) -> (then, bool:bool) {is=title K2Node_CallFunction}
   > n7 = Set bWorkspaceReady(bWorkspaceReady:bool=@n6) -> (Output_Get:bool)
   > Call When Stat Change!(execute:Exec, self:UWBT 3DHB Initializer Object Reference, bWorkspaceReady:bool=@n7) -> (then:Exec) {is=title K2Node_CallDelegate}
   > n8 = Sequence()
   n9 = Get CreateNewWorkspace() -> (CreateNewWorkspace:Button Object Reference)
   > SetIsEnabled(self=@n9, bInIsEnabled=NOT Boolean(A=Get bWorkspaceReady() -> (bWorkspaceReady:bool)))
   n10 = Is Valid(exec=@n8.then_1, InputObject=Get Target3DHBDataAsset())
   SetIsEnabled(execute=@n10.Is Not Valid, self=@n9, bInIsEnabled=false)
   n11 = "On Property Changed (DetailsView_0)"!() -> (OutputDelegate, then, PropertyName:Name=None) {is=title K2Node_ComponentBoundEvent}
   Event Construct()
   > n12 = SetObject(self=Get DetailsView_0() -> (DetailsView_0:Details View Object Reference), NewObject=Self-Reference())
  ]}
 Get Asset Object Folder Path {nodes [
   n1 = Get Asset Object Folder Path!() -> (then, TargetObj:Object Reference) {is=title K2Node_FunctionEntry}
   > n2 = Is Valid(InputObject=@n1)
   Return Node!(execute=@n2.Is Not Valid, Path:String)
   > n3 = GetPathNameForLoadedAsset(LoadedAsset=@n1)
   n4 = GetObjectName(Object=@n1)
   > Return Node(Path=Split(SourceString=@n3, InStr=Append!(A=@n4, B=., C:String=@n4) -> (ReturnValue), SearchCase=CaseSensitive, SearchDir=FromEnd))
  ]}
 WhenStatChange {nodes [WhenStatChange!() -> (then, bWorkspaceReady:bool=true) {is=title K2Node_FunctionEntry}]}
]
variables [
 bWorkspaceReady {type=boolean}
 FolderList {type="array<string>", default [Texture,Material,Mesh,FX]}
 ForCopyMaterialInstance {type=MaterialInstance, default=/Game/Tools/3DHB/Initializer/ForCopy/MI_3DHB_ForCopy}
 InitialWorkSpace_TaskName {type=string, default=3DHB}
 Target3DHBDataAsset {type=PrimaryDataAsset, default=null}
 WhenStatChange {default=()}
]
widgets {
 info {rootWidgetClass=&CanvasPanel, widgetCount=6}
 widgets [
  CanvasPanel_0 {widgetClassPath=&CanvasPanel, widget=CanvasPanel_0}
  VerticalBox_33 {
   widgetClassPath=&VerticalBox
   parent=CanvasPanel_0
   widget=VerticalBox_33
   slot {class=CanvasPanelSlot, properties {layoutData {alignment=<0 0>, anchors {maximum=<1 1>, minimum=<0 0>}, offsets=<ltrb 0 0 0 0>}}}
  }
  DetailsView_0 {
   widgetClassPath=&DetailsView
   parent=VerticalBox_33
   widget=DetailsView_0
   properties {categoriesToShow [3DHB Initializer]}
   slot {class=VerticalBoxSlot}
  }
  CreateNewWorkspace {
   widgetClassPath=&Button
   parent=VerticalBox_33
   widget=CreateNewWorkspace
   properties {
    backgroundColor=<rgba 0.1 0.1 0.1 1>
    widgetStyle {
     clickedSlateSound {resourceObject=null}
     disabled {
      bIsDynamicallyLoaded=false
      drawAs=RoundedBox
      imageType=NoImage
      mirroring=NoMirror
      resourceName=None
      resourceObject=null
      tiling=NoTile
      imageSize=<0 0>
      margin=<ltrb 0 0 0 0>
      outlineSettings {
       bUseBrushTransparency=false
       roundingType=FixedRadius
       width=1
       color {colorUseRule=UseColor_ColorTable, specifiedColor=<rgba 0 0 0 0>}
       cornerRadii=<4 4 4 4>
      }
      tintColor {colorUseRule=UseColor_Specified, specifiedColor=<rgba 0.0395 0.0395 0.0395 1>}
      uVRegion {bIsValid=false, max=<0 0>, min=<0 0>}
     }
     disabledForeground {colorUseRule=UseColor_ColorTable, specifiedColor=<rgba 0 0 0 0>}
     hovered {
      bIsDynamicallyLoaded=false
      drawAs=RoundedBox
      imageType=NoImage
      mirroring=NoMirror
      resourceName=None
      resourceObject=null
      tiling=NoTile
      imageSize=<0 0>
      margin=<ltrb 0 0 0 0>
      outlineSettings {
       bUseBrushTransparency=false
       roundingType=FixedRadius
       width=1
       color {colorUseRule=UseColor_ColorTable, specifiedColor=<rgba 0 0 0 0>}
       cornerRadii=<4 4 4 4>
      }
      tintColor {colorUseRule=UseColor_Specified, specifiedColor=<rgba 0.0953 0.0953 0.0953 1>}
      uVRegion {bIsValid=false, max=<0 0>, min=<0 0>}
     }
     hoveredForeground {colorUseRule=UseColor_ColorTable, specifiedColor=<rgba 0 0 0 0>}
     hoveredSlateSound {resourceObject=null}
     normal {
      bIsDynamicallyLoaded=false
      drawAs=RoundedBox
      imageType=NoImage
      mirroring=NoMirror
      resourceName=None
      resourceObject=null
      tiling=NoTile
      imageSize=<0 0>
      margin=<ltrb 0 0 0 0>
      outlineSettings {
       bUseBrushTransparency=false
       roundingType=FixedRadius
       width=1
       color {colorUseRule=UseColor_ColorTable, specifiedColor=<rgba 0 0 0 0>}
       cornerRadii=<4 4 4 4>
      }
      tintColor {colorUseRule=UseColor_Specified, specifiedColor=<rgba 0.0395 0.0395 0.0395 1>}
      uVRegion {bIsValid=false, max=<0 0>, min=<0 0>}
     }
     normalForeground {colorUseRule=UseColor_ColorTable, specifiedColor=<rgba 0 0 0 0>}
     normalPadding=<ltrb 12 1.5 12 1.5>
     pressed {
      bIsDynamicallyLoaded=false
      drawAs=RoundedBox
      imageType=NoImage
      mirroring=NoMirror
      resourceName=None
      resourceObject=null
      tiling=NoTile
      imageSize=<0 0>
      margin=<ltrb 0 0 0 0>
      outlineSettings {
       bUseBrushTransparency=false
       roundingType=FixedRadius
       width=1
       color {colorUseRule=UseColor_ColorTable, specifiedColor=<rgba 0 0 0 0>}
       cornerRadii=<4 4 4 4>
      }
      tintColor {colorUseRule=UseColor_Specified, specifiedColor=<rgba 0.0284 0.0284 0.0284 1>}
      uVRegion {bIsValid=false, max=<0 0>, min=<0 0>}
     }
     pressedForeground {colorUseRule=UseColor_ColorTable, specifiedColor=<rgba 0 0 0 0>}
     pressedPadding=<ltrb 12 2 12 0>
     pressedSlateSound {resourceObject=null}
    }
   }
   slot {class=VerticalBoxSlot}
  }
  CommonTextBlock_78 {
   widgetClassPath=&CommonTextBlock
   parent=CreateNewWorkspace
   widget=CommonTextBlock_78
   properties {
    style=/Game/UI/Foundation/Items/Parts/TextStyles/ItemCardPowerRatingTextStyle_S.ItemCardPowerRatingTextStyle_S_C
    text=Initial workspace for new data asset
    font {
     bForceMonospaced=false
     bMaterialIsStencil=false
     fontMaterial=null
     fontObject=&NotoSans
     letterSpacing=0
     monospacedWidth=1
     size=18
     skewAmount=0
     typefaceFontName=NotoSans-Bold
     outlineSettings {
      bApplyOutlineToDropShadows=false
      bMiteredCorners=false
      bSeparateFillAlpha=false
      outlineMaterial=null
      outlineSize=0
      outlineColor=<rgba 0 0 0 1>
     }
    }
    shadowOffset=<0 0>
   }
   slot {class=ButtonSlot}
  }
  VerticalBox_87 {widgetClassPath=&VerticalBox, parent=VerticalBox_33, widget=VerticalBox_87, slot {class=VerticalBoxSlot}}
 ]
}
