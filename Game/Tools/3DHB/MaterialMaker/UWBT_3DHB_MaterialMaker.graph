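# Editor utility widget for the 3D Hologram Backpack tool that creates and configures materials for a backpack data asset.
# One button duplicates the template master material and opens it for editing.
# Another duplicates a template instance into the asset's Material folder, parents it to the selected master and stores it on the asset.
# ApplyTexture sets diffuse, WPO and normal parameters by 3D/2D workflow, mostly preferring replacement textures, then achieved textures, then TextureBaker render targets.
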EditorUtilityWidgetBlueprint /Game/Tools/3DHB/MaterialMaker/UWBT_3DHB_MaterialMaker
graphs [
 EventGraph {nodes [
   "On Clicked (Button_132)"() {is=title K2Node_ComponentBoundEvent}
   > n1 = Branch(Condition=IsEmpty(InString=Get NewMaterialInstanceName() -> (NewMaterialInstanceName:String)))
   n2 = Get 3DHBDataAsset() -> (3DHBDataAsset:3DHologram Backpack Data Asset Object Reference)
   n3 = GetPathNameForLoadedAsset(execute=@n1.else, LoadedAsset=@n2)
   n4 = Get ForCopyInstance() -> (ForCopyInstance:Material Instance Object Reference)
   > n5 = GetPathNameForLoadedAsset(LoadedAsset=@n4)
   n6 = GetObjectName(Object=@n4)
   n7 = Append!(A=@n6, B=., C:String=@n6) -> (ReturnValue)
   n8 = GetObjectName(Object=@n2)
   n9 = Append(
    A=Split(SourceString=@n3, InStr=Append(A=@n8, B=., C=@n8), SearchCase=CaseSensitive, SearchDir=FromEnd)
    B=Material/
    C=Get NewMaterialInstanceName()
   )
   > DuplicateAsset!(
    execute
    SourceAssetPath:String=Append!(A=Split(SourceString=@n5, InStr=@n7, SearchCase=CaseSensitive, SearchDir=FromEnd), B=@n6) -> (ReturnValue)
    DestinationAssetPath:String=@n9
   ) -> (then, ReturnValue:Object Reference)
   > n10 = LoadAsset(AssetPath=@n9)
   n11 = Cast To MaterialInstanceConstant!(Object=@n10) -> (AsMaterial Instance Constant, bSuccess:bool=~)
   > SetMaterialInstanceParent!(
    execute
    Instance:Material Instance Constant Object Reference=@n11
    NewParent:Material Interface Object Reference=Get MasterMaterial() -> (MasterMaterial:Material Object Reference)
   ) -> (then)
   > Set MaterialInstance!(execute, MaterialInstance:Material Instance Object Reference=@n11, self:3DHologram Backpack Data Asset Object Reference=Get 3DHBDataAsset()) -> (then, Output_Get:Material Instance Object Reference)
   > ApplyTexture()
   "On Clicked (Button)"() {is=title K2Node_ComponentBoundEvent}
   > n12 = Branch(Condition=IsEmpty(InString=Get NewMasterMaterialName() -> (NewMasterMaterialName:String)))
   n13 = Get ForCopyMaster() -> (ForCopyMaster:Material Object Reference)
   n14 = GetPathNameForLoadedAsset(execute=@n12.else, LoadedAsset=@n13)
   n15 = GetObjectName(Object=@n13)
   n16 = Split(SourceString=@n14, InStr=Append!(A=@n15, B=., C:String=@n15) -> (ReturnValue), SearchCase=CaseSensitive, SearchDir=FromEnd)
   > n17 = DuplicateAsset(SourceAssetPath=Append!(A=@n16, B=@n15) -> (ReturnValue), DestinationAssetPath=Append(A=@n16, B=Get NewMasterMaterialName()))
   > n18 = Cast To Material!(execute, Object=@n17) -> (then, CastFailed, AsMaterial:Material Object Reference) {is=title K2Node_DynamicCast}
   > Set MasterMaterial(MasterMaterial:Material Object Reference=@n18) -> (Output_Get:Material Object Reference)
   > OpenEditorForAssets!(
    execute
    self:Asset Editor Subsystem Object Reference=Get AssetEditorSubsystem()
    Assets:Array of Object References=Make Array("[0]":Object Reference=@n18) -> (Array:Array of Object References)
    OpenedMethod:EAssetTypeActivationOpenedMethod Enum=Edit
   ) -> (then, ReturnValue:bool)
   "On Property Changed (CurrentMaterial)"!() -> (OutputDelegate, then, PropertyName:Name=None) {is=title K2Node_ComponentBoundEvent}
   > Call Property Change!(execute:Exec, self:UWBT 3DHB Material Maker Object Reference) -> (then:Exec) {is=title K2Node_CallDelegate}
   n19 = Initial!() -> (OutputDelegate, then, DataAsset:3DHologram Backpack Data Asset Object Reference) {is=title K2Node_CustomEvent}
   > n20 = Set 3DHBDataAsset(3DHBDataAsset:3DHologram Backpack Data Asset Object Reference=@n19.DataAsset) -> (Output_Get:3DHologram Backpack Data Asset Object Reference)
   > SetObject(self=Get CurrentMaterial() -> (CurrentMaterial:Details View Object Reference), NewObject=@n20)
   > ApplyTexture()
   ApplyTexture() {is=title K2Node_CustomEvent}
   n21 = Get 3DHBDataAsset()
   > Is Valid(InputObject=Get MaterialInstance!(self:3DHologram Backpack Data Asset Object Reference=@n21) -> (MaterialInstance:Material Instance Object Reference))
   > n22 = Sequence!(execute) -> (then_0, then_1, then_2:Exec)
   n23 = "Equal (Integer)"(A=Get WorkflowType!(self:3DHologram Backpack Data Asset Object Reference=@n21) -> (WorkflowType:int=0), B=0) {is=title K2Node_CallFunction}
   > n24 = Branch(Condition=@n23)
   > n25 = Sequence()
   n26 = Get From3D_ReplaceDiffuseTexture!(self:3DHologram Backpack Data Asset Object Reference=@n21) -> (From3D_ReplaceDiffuseTexture:Texture 2D Object Reference)
   > n27 = Is Valid(InputObject=@n26)
   n28 = Get 3DHBDataAsset()
   n29 = Get AchievedDiffuseTexture!(self:3DHologram Backpack Data Asset Object Reference=@n28) -> (AchievedDiffuseTexture:Texture 2D Object Reference)
   n30 = Is Valid(exec=@n27.Is Not Valid, InputObject=@n29)
   n31 = Get MaterialInstance(self=@n21)
   n32 = Cast To MaterialInstanceConstant(Object=@n31) -> (bSuccess=~)
   SetMaterialInstanceTextureParameterValue!(
    execute=@n30.Is Not Valid
    Instance:Material Instance Constant Object Reference=@n32
    ParameterName:Name=Base_DiffuseTexture
    Value:Texture Object Reference=/Game/Tools/3DHB/TextureBaker/RenderTarget/RT_3DHB_DiffuseCaptureOut
    Association:EMaterialParameterAssociation Enum=GlobalParameter
   ) -> (then, ReturnValue:bool)
   SetMaterialInstanceTextureParameterValue(execute=@n30.Is Valid, Instance=@n32, ParameterName=Base_DiffuseTexture, Value=@n29, Association=GlobalParameter)
   SetMaterialInstanceTextureParameterValue(execute=@n27.Is Valid, Instance=@n32, ParameterName=Base_DiffuseTexture, Value=@n26, Association=GlobalParameter)
   n33 = Get From3D_ReplaceWPOTexture!(self:3DHologram Backpack Data Asset Object Reference=@n21) -> (From3D_ReplaceWPOTexture:Texture 2D Object Reference)
   n34 = Is Valid(exec=@n25.then_1, InputObject=@n33)
   n35 = Get AchievedWPOTexture!(self:3DHologram Backpack Data Asset Object Reference=@n28) -> (AchievedWPOTexture:Texture 2D Object Reference)
   n36 = Is Valid(exec=@n34.Is Not Valid, InputObject=@n35)
   SetMaterialInstanceTextureParameterValue(
    execute=@n36.Is Not Valid
    Instance=@n32
    ParameterName=Base_WPOTexture
    Value=/Game/Tools/3DHB/TextureBaker/RenderTarget/RT_3DHB_HeightBlurTexture
    Association=GlobalParameter
   )
   SetMaterialInstanceTextureParameterValue(execute=@n36.Is Valid, Instance=@n32, ParameterName=Base_WPOTexture, Value=@n35, Association=GlobalParameter)
   SetMaterialInstanceTextureParameterValue(execute=@n34.Is Valid, Instance=@n32, ParameterName=Base_WPOTexture, Value=@n33, Association=GlobalParameter)
   n37 = Get From3D_ReplaceNormalTexture!(self:3DHologram Backpack Data Asset Object Reference=@n21) -> (From3D_ReplaceNormalTexture:Texture 2D Object Reference)
   n38 = Is Valid(exec=@n25.then_2, InputObject=@n37)
   n39 = Get AchievedNormalTexture!(self:3DHologram Backpack Data Asset Object Reference=@n28) -> (AchievedNormalTexture:Texture 2D Object Reference)
   n40 = Is Valid(exec=@n38.Is Not Valid, InputObject=@n39)
   SetMaterialInstanceTextureParameterValue(
    execute=@n40.Is Not Valid
    Instance=@n32
    ParameterName=Base_NormalTexture
    Value=/Game/Tools/3DHB/TextureBaker/RenderTarget/RT_3DHB_NormalCaptureOut
    Association=GlobalParameter
   )
   SetMaterialInstanceTextureParameterValue(execute=@n40.Is Valid, Instance=@n32, ParameterName=Base_NormalTexture, Value=@n39, Association=GlobalParameter)
   SetMaterialInstanceTextureParameterValue(execute=@n38.Is Valid, Instance=@n32, ParameterName=Base_NormalTexture, Value=@n37, Association=GlobalParameter)
   n41 = Sequence(execute=@n24.else)
   n42 = Get From2D_SourceTexture!(self:3DHologram Backpack Data Asset Object Reference=@n21) -> (From2D_SourceTexture:Texture 2D Object Reference)
   > Is Valid(InputObject=@n42)
   > SetMaterialInstanceTextureParameterValue(Instance=@n32, ParameterName=Base_DiffuseTexture, Value=@n42, Association=GlobalParameter)
   n43 = Get From2D_ReplaceWPOTexture!(self:3DHologram Backpack Data Asset Object Reference=@n21) -> (From2D_ReplaceWPOTexture:Texture 2D Object Reference)
   n44 = Is Valid(exec=@n41.then_1, InputObject=@n43)
   n45 = Get 3DHBDataAsset()
   n46 = Get AchievedWPOTexture(self=@n45)
   n47 = Is Valid(exec=@n44.Is Not Valid, InputObject=@n46)
   SetMaterialInstanceTextureParameterValue(
    execute=@n47.Is Not Valid
    Instance=@n32
    ParameterName=Base_WPOTexture
    Value=&RT_3DHB_HeightBlurTexture
    Association=GlobalParameter
   )
   SetMaterialInstanceTextureParameterValue(execute=@n47.Is Valid, Instance=@n32, ParameterName=Base_WPOTexture, Value=@n46, Association=GlobalParameter)
   SetMaterialInstanceTextureParameterValue(execute=@n44.Is Valid, Instance=@n32, ParameterName=Base_WPOTexture, Value=@n43, Association=GlobalParameter)
   n48 = Get From2D_ReplaceNormalTexture!(self:3DHologram Backpack Data Asset Object Reference=@n21) -> (From2D_ReplaceNormalTexture:Texture 2D Object Reference)
   n49 = Is Valid(exec=@n41.then_2, InputObject=@n48)
   n50 = Get AchievedNormalTexture(self=@n45)
   n51 = Is Valid(exec=@n49.Is Not Valid, InputObject=@n50)
   SetMaterialInstanceTextureParameterValue(
    execute=@n51.Is Not Valid
    Instance=@n32
    ParameterName=Base_NormalTexture
    Value=/Game/Tools/3DHB/TextureBaker/RenderTarget/RT_3DHB_NormalTexture
    Association=GlobalParameter
   )
   SetMaterialInstanceTextureParameterValue(execute=@n51.Is Valid, Instance=@n32, ParameterName=Base_NormalTexture, Value=@n50, Association=GlobalParameter)
   SetMaterialInstanceTextureParameterValue(execute=@n49.Is Valid, Instance=@n32, ParameterName=Base_NormalTexture, Value=@n48, Association=GlobalParameter)
   SetMaterialInstanceScalarParameterValue!(
    execute=@n22.then_1
    Instance:Material Instance Constant Object Reference=@n32
    ParameterName:Name=Base_IsFrom3D
    Value:float=SelectFloat(A=1, B=0, bPickA=@n23)
    Association:EMaterialParameterAssociation Enum=GlobalParameter
   ) -> (then, ReturnValue:bool)
   SetEditorProperty(execute=@n22.then_2, Object=@n31, PropertyName=bOverrideSubsurfaceProfile, PropertyValue=OR Boolean(A=true, B=false))
   > SetEditorProperty(Object=@n31, PropertyName=bOverrideSubsurfaceProfile, PropertyValue=OR Boolean(A=false, B=false))
   Event Construct()
   > SetObject(self=Get CreateMaterial() -> (CreateMaterial:Details View Object Reference), NewObject=Self-Reference())
   SaveLoadedAsset(AssetToSave=@n31, bOnlyIfIsDirty=false)
   CheckoutLoadedAsset!(execute, AssetToCheckout:Object Reference=@n31) -> (then, ReturnValue:bool)
  ]}
 PropertyChange {nodes [PropertyChange() {is=title K2Node_FunctionEntry}]}
]
variables [
 3DHBDataAsset {type=3DHologramBackpackDataAsset_C, default=null}
 ForCopyInstance {type=MaterialInstance, default=/Game/Tools/3DHB/Initializer/ForCopy/MI_3DHB_ForCopy}
 ForCopyMaster {type=Material, default=/Game/Tools/3DHB/Shader/MasterLib/M_3DHB, class=~, graph=~}
 MasterMaterial {type=Material, default=&M_3DHB, class=~, graph=~}
 NewMasterMaterialName {type=string, default=""}
 NewMaterialInstanceName {type=string, default=""}
 PropertyChange {default=()}
]
widgets {
 info {rootWidgetClass=&CanvasPanel, widgetCount=12}
 widgets [
  CanvasPanel_0 {widgetClassPath=&CanvasPanel, widget=CanvasPanel_0}
  HorizontalBox_34 {
   widgetClassPath=&HorizontalBox
   parent=CanvasPanel_0
   widget=HorizontalBox_34
   slot {class=CanvasPanelSlot, properties {layoutData {alignment=<0 0>, anchors {maximum=<1 1>, minimum=<0 0>}, offsets=<ltrb 0 0 0 0>}}}
  }
  VerticalBox_26 {
   widgetClassPath=&VerticalBox
   parent=HorizontalBox_34
   widget=VerticalBox_26
   slot {class=HorizontalBoxSlot, properties {size {sizeRule=Fill, value=1}}}
  }
  CommonTextBlock_53 {
   widgetClassPath=&CommonTextBlock
   parent=VerticalBox_26
   widget=CommonTextBlock_53
   properties {
    text=Current Material :
    font {
     bForceMonospaced=false
     bMaterialIsStencil=false
     fontMaterial=null
     fontObject=&Burbank
     letterSpacing=0
     monospacedWidth=1
     size=20
     skewAmount=0
     typefaceFontName=Burbank-Black
     outlineSettings {
      bApplyOutlineToDropShadows=false
      bMiteredCorners=false
      bSeparateFillAlpha=false
      outlineMaterial=null
      outlineSize=0
      outlineColor=<rgba 0 0 0 1>
     }
    }
    shadowOffset=<0 0>
   }
   slot {class=VerticalBoxSlot}
  }
  CurrentMaterial {
   widgetClassPath=&DetailsView
   parent=VerticalBox_26
   widget=CurrentMaterial
   properties {categoriesToShow [MaterialSetting]}
   slot {class=VerticalBoxSlot}
  }
  CommonTextBlock_3 {
   widgetClassPath=&CommonTextBlock
   parent=VerticalBox_26
   widget=CommonTextBlock_3
   properties {
    text=Create Material Tool:
    font {
     bForceMonospaced=false
     bMaterialIsStencil=false
     fontMaterial=null
     fontObject=&Burbank
     letterSpacing=0
     monospacedWidth=1
     size=20
     skewAmount=0
     typefaceFontName=Burbank-Black
     outlineSettings {
      bApplyOutlineToDropShadows=false
      bMiteredCorners=false
      bSeparateFillAlpha=false
      outlineMaterial=null
      outlineSize=0
      outlineColor=<rgba 0 0 0 1>
     }
    }
    shadowOffset=<0 0>
   }
   slot {class=VerticalBoxSlot, properties {padding=<ltrb 0 30 0 0>}}
  }
  CreateMaterial {
   widgetClassPath=&DetailsView
   parent=VerticalBox_26
   widget=CreateMaterial
   properties {categoriesToShow [CreateMaterial]}
   slot {class=VerticalBoxSlot}
  }
  Button_132 {
   widgetClassPath=&Button
   parent=VerticalBox_26
   widget=Button_132
   properties {
    backgroundColor=<rgba 0.1 0.1 0.1 1>
    widgetStyle {
     clickedSlateSound {resourceObject=null}
     disabled {
      bIsDynamicallyLoaded=false
      drawAs=RoundedBox
      imageType=NoImage
      mirroring=NoMirror
      resourceName=None
      resourceObject=null
      tiling=NoTile
      imageSize=<0 0>
      margin=<ltrb 0 0 0 0>
      outlineSettings {
       bUseBrushTransparency=false
       roundingType=FixedRadius
       width=1
       color {colorUseRule=UseColor_ColorTable, specifiedColor=<rgba 0 0 0 0>}
       cornerRadii=<4 4 4 4>
      }
      tintColor {colorUseRule=UseColor_Specified, specifiedColor=<rgba 0.0395 0.0395 0.0395 1>}
      uVRegion {bIsValid=false, max=<0 0>, min=<0 0>}
     }
     disabledForeground {colorUseRule=UseColor_ColorTable, specifiedColor=<rgba 0 0 0 0>}
     hovered {
      bIsDynamicallyLoaded=false
      drawAs=RoundedBox
      imageType=NoImage
      mirroring=NoMirror
      resourceName=None
      resourceObject=null
      tiling=NoTile
      imageSize=<0 0>
      margin=<ltrb 0 0 0 0>
      outlineSettings {
       bUseBrushTransparency=false
       roundingType=FixedRadius
       width=1
       color {colorUseRule=UseColor_ColorTable, specifiedColor=<rgba 0 0 0 0>}
       cornerRadii=<4 4 4 4>
      }
      tintColor {colorUseRule=UseColor_Specified, specifiedColor=<rgba 0.0953 0.0953 0.0953 1>}
      uVRegion {bIsValid=false, max=<0 0>, min=<0 0>}
     }
     hoveredForeground {colorUseRule=UseColor_ColorTable, specifiedColor=<rgba 0 0 0 0>}
     hoveredSlateSound {resourceObject=null}
     normal {
      bIsDynamicallyLoaded=false
      drawAs=RoundedBox
      imageType=NoImage
      mirroring=NoMirror
      resourceName=None
      resourceObject=null
      tiling=NoTile
      imageSize=<0 0>
      margin=<ltrb 0 0 0 0>
      outlineSettings {
       bUseBrushTransparency=false
       roundingType=FixedRadius
       width=1
       color {colorUseRule=UseColor_ColorTable, specifiedColor=<rgba 0 0 0 0>}
       cornerRadii=<4 4 4 4>
      }
      tintColor {colorUseRule=UseColor_Specified, specifiedColor=<rgba 0.0395 0.0395 0.0395 1>}
      uVRegion {bIsValid=false, max=<0 0>, min=<0 0>}
     }
     normalForeground {colorUseRule=UseColor_ColorTable, specifiedColor=<rgba 0 0 0 0>}
     normalPadding=<ltrb 12 1.5 12 1.5>
     pressed {
      bIsDynamicallyLoaded=false
      drawAs=RoundedBox
      imageType=NoImage
      mirroring=NoMirror
      resourceName=None
      resourceObject=null
      tiling=NoTile
      imageSize=<0 0>
      margin=<ltrb 0 0 0 0>
      outlineSettings {
       bUseBrushTransparency=false
       roundingType=FixedRadius
       width=1
       color {colorUseRule=UseColor_ColorTable, specifiedColor=<rgba 0 0 0 0>}
       cornerRadii=<4 4 4 4>
      }
      tintColor {colorUseRule=UseColor_Specified, specifiedColor=<rgba 0.0284 0.0284 0.0284 1>}
      uVRegion {bIsValid=false, max=<0 0>, min=<0 0>}
     }
     pressedForeground {colorUseRule=UseColor_ColorTable, specifiedColor=<rgba 0 0 0 0>}
     pressedPadding=<ltrb 12 2 12 0>
     pressedSlateSound {resourceObject=null}
    }
   }
   slot {class=VerticalBoxSlot}
  }
  CommonTextBlock_1 {
   widgetClassPath=&CommonTextBlock
   parent=Button_132
   widget=CommonTextBlock_1
   properties {
    text=Create new instance from selected master material
    colorAndOpacity {colorUseRule=UseColor_Specified, specifiedColor=<rgba 0.5365 0.4209 0 1>}
    font {
     bForceMonospaced=false
     bMaterialIsStencil=false
     fontMaterial=null
     fontObject=/LauncherSocial/BrutalType
     letterSpacing=0
     monospacedWidth=1
     size=20
     skewAmount=0
     typefaceFontName=Regular
     outlineSettings {
      bApplyOutlineToDropShadows=false
      bMiteredCorners=false
      bSeparateFillAlpha=false
      outlineMaterial=null
      outlineSize=0
      outlineColor=<rgba 0 0 0 1>
     }
    }
    shadowOffset=<0 0>
   }
   slot {class=ButtonSlot, properties {horizontalAlignment=HAlign_Fill, verticalAlignment=VAlign_Fill, padding=<ltrb 0 0 0 0>}}
  }
  Button {
   widgetClassPath=&Button
   parent=VerticalBox_26
   widget=Button
   properties {
    backgroundColor=<rgba 0.1 0.1 0.1 1>
    widgetStyle {
     clickedSlateSound {resourceObject=null}
     disabled {
      bIsDynamicallyLoaded=false
      drawAs=RoundedBox
      imageType=NoImage
      mirroring=NoMirror
      resourceName=None
      resourceObject=null
      tiling=NoTile
      imageSize=<0 0>
      margin=<ltrb 0 0 0 0>
      outlineSettings {
       bUseBrushTransparency=false
       roundingType=FixedRadius
       width=1
       color {colorUseRule=UseColor_ColorTable, specifiedColor=<rgba 0 0 0 0>}
       cornerRadii=<4 4 4 4>
      }
      tintColor {colorUseRule=UseColor_Specified, specifiedColor=<rgba 0.0395 0.0395 0.0395 1>}
      uVRegion {bIsValid=false, max=<0 0>, min=<0 0>}
     }
     disabledForeground {colorUseRule=UseColor_ColorTable, specifiedColor=<rgba 0 0 0 0>}
     hovered {
      bIsDynamicallyLoaded=false
      drawAs=RoundedBox
      imageType=NoImage
      mirroring=NoMirror
      resourceName=None
      resourceObject=null
      tiling=NoTile
      imageSize=<0 0>
      margin=<ltrb 0 0 0 0>
      outlineSettings {
       bUseBrushTransparency=false
       roundingType=FixedRadius
       width=1
       color {colorUseRule=UseColor_ColorTable, specifiedColor=<rgba 0 0 0 0>}
       cornerRadii=<4 4 4 4>
      }
      tintColor {colorUseRule=UseColor_Specified, specifiedColor=<rgba 0.0953 0.0953 0.0953 1>}
      uVRegion {bIsValid=false, max=<0 0>, min=<0 0>}
     }
     hoveredForeground {colorUseRule=UseColor_ColorTable, specifiedColor=<rgba 0 0 0 0>}
     hoveredSlateSound {resourceObject=null}
     normal {
      bIsDynamicallyLoaded=false
      drawAs=RoundedBox
      imageType=NoImage
      mirroring=NoMirror
      resourceName=None
      resourceObject=null
      tiling=NoTile
      imageSize=<0 0>
      margin=<ltrb 0 0 0 0>
      outlineSettings {
       bUseBrushTransparency=false
       roundingType=FixedRadius
       width=1
       color {colorUseRule=UseColor_ColorTable, specifiedColor=<rgba 0 0 0 0>}
       cornerRadii=<4 4 4 4>
      }
      tintColor {colorUseRule=UseColor_Specified, specifiedColor=<rgba 0.0395 0.0395 0.0395 1>}
      uVRegion {bIsValid=false, max=<0 0>, min=<0 0>}
     }
     normalForeground {colorUseRule=UseColor_ColorTable, specifiedColor=<rgba 0 0 0 0>}
     normalPadding=<ltrb 12 1.5 12 1.5>
     pressed {
      bIsDynamicallyLoaded=false
      drawAs=RoundedBox
      imageType=NoImage
      mirroring=NoMirror
      resourceName=None
      resourceObject=null
      tiling=NoTile
      imageSize=<0 0>
      margin=<ltrb 0 0 0 0>
      outlineSettings {
       bUseBrushTransparency=false
       roundingType=FixedRadius
       width=1
       color {colorUseRule=UseColor_ColorTable, specifiedColor=<rgba 0 0 0 0>}
       cornerRadii=<4 4 4 4>
      }
      tintColor {colorUseRule=UseColor_Specified, specifiedColor=<rgba 0.0284 0.0284 0.0284 1>}
      uVRegion {bIsValid=false, max=<0 0>, min=<0 0>}
     }
     pressedForeground {colorUseRule=UseColor_ColorTable, specifiedColor=<rgba 0 0 0 0>}
     pressedPadding=<ltrb 12 2 12 0>
     pressedSlateSound {resourceObject=null}
    }
   }
   slot {class=VerticalBoxSlot}
  }
  CommonTextBlock_4 {
   widgetClassPath=&CommonTextBlock
   parent=Button
   widget=CommonTextBlock_4
   properties {
    text=Create new master material
    colorAndOpacity {colorUseRule=UseColor_Specified, specifiedColor=<rgba 0 0.2604 0.2388 1>}
    font {
     bForceMonospaced=false
     bMaterialIsStencil=false
     fontMaterial=null
     fontObject=&BrutalType
     letterSpacing=0
     monospacedWidth=1
     size=20
     skewAmount=0
     typefaceFontName=Regular
     outlineSettings {
      bApplyOutlineToDropShadows=false
      bMiteredCorners=false
      bSeparateFillAlpha=false
      outlineMaterial=null
      outlineSize=0
      outlineColor=<rgba 0 0 0 1>
     }
    }
    shadowOffset=<0 0>
   }
   slot {class=ButtonSlot, properties {horizontalAlignment=HAlign_Fill, verticalAlignment=VAlign_Fill, padding=<ltrb 0 0 0 0>}}
  }
  CommonTextBlock_2 {
   widgetClassPath=&CommonTextBlock
   parent=VerticalBox_26
   widget=CommonTextBlock_2
   properties {
    text="Your new master material will create in master material lib folder.\r\nIf not respon after click button, please check name,which may same to other material."
    colorAndOpacity {colorUseRule=UseColor_Specified, specifiedColor=<rgba 0 0.2604 0.2388 1>}
    font {
     bForceMonospaced=false
     bMaterialIsStencil=false
     fontMaterial=null
     fontObject=&BrutalType
     letterSpacing=0
     monospacedWidth=1
     size=20
     skewAmount=0
     typefaceFontName=Regular
     outlineSettings {
      bApplyOutlineToDropShadows=false
      bMiteredCorners=false
      bSeparateFillAlpha=false
      outlineMaterial=null
      outlineSize=0
      outlineColor=<rgba 0 0 0 1>
     }
    }
    shadowOffset=<0 0>
   }
   slot {class=VerticalBoxSlot}
  }
 ]
}
