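# Editor utility widget that previews a 3D Hologram Backpack data asset as a Niagara effect in the level.
# Start spawns or reuses an NS_3DHB actor, then feeds it the material, workflow flag and baker render targets, with diffuse from the source texture for 2D.
# UpdateState gates Start on a valid material, flags whether achieved textures exist, and refreshes any live preview.
# Stop, widget destruction or invalid data destroys the actor.
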EditorUtilityWidgetBlueprint /Game/Tools/3DHB/Previewer/UWBT_3DHB_Previewer
graphs [EventGraph {nodes [
   "On Clicked (Button_0)"() {is=title K2Node_ComponentBoundEvent}
   n1 = Get PreiviewActor() -> (PreiviewActor:Niagara Particle System Actor Object Reference)
   > n2 = Is Valid(InputObject=@n1)
   n3 = SpawnActor Niagara Particle System Actor!(
    execute:Exec=@n2.Is Not Valid
    WorldContextObject:Object Reference
    Class:Actor Class Reference=/Script/Niagara.NiagaraActor
    SpawnTransform:Transform=MakeTransform(Location=Get PreviewPosition() -> (PreviewPosition:Vector), Rotation="0, 0, 0")
    CollisionHandlingOverride:ESpawnActorCollisionHandlingMethod Enum=AlwaysSpawn
    TransformScaleMethod:ESpawnActorScaleMethod Enum=MultiplyWithRoot
    Owner:Actor Object Reference
    Instigator:Pawn Object Reference
   ) -> (then:Exec, ReturnValue:Niagara Particle System Actor Object Reference) {is=title K2Node_SpawnActorFromClass}
   > Set PreiviewActor(PreiviewActor:Niagara Particle System Actor Object Reference=@n3) -> (Output_Get:Niagara Particle System Actor Object Reference)
   n4 = Get NiagaraComponent!(self:Niagara Particle System Actor Object Reference=@n3) -> (NiagaraComponent:Niagara Particle System Component Object Reference)
   > n5 = Set Niagara System Asset!(
    execute
    self:Niagara Particle System Component Object Reference=@n4
    InAsset:Niagara System Object Reference=/Game/Tools/3DHB/FX/Core/NS_3DHB
    bResetExistingOverrideParameters:bool=true
   ) -> (then) {is=title K2Node_CallFunction}
   n6 = Get 3DHBDataAsset() -> (3DHBDataAsset:3DHologram Backpack Data Asset Object Reference)
   "Set Niagara Variable (Material)"!(
    execute=[@n5.then, @n2.Is Valid, @n25.Is Valid]
    self:Niagara Particle System Component Object Reference=@n4
    InVariableName:Name=Backpack_Mat
    Object:Material Interface Object Reference=Get MaterialInstance!(self:3DHologram Backpack Data Asset Object Reference=@n6) -> (MaterialInstance:Material Instance Object Reference)
   ) -> (then) {is=title K2Node_CallFunction}
   n7 = Get WorkflowType!(self:3DHologram Backpack Data Asset Object Reference=@n6) -> (WorkflowType:int=0)
   > "Set Niagara Variable (Float)"!(
    execute
    self:Niagara Particle System Component Object Reference=@n4
    InVariableName:Name=IsFrom3D
    InValue:float=float - float(B="To Float (Integer)"(InInt=@n7)) {is=title K2Node_CallFunction}
   ) -> (then) {is=title K2Node_CallFunction}
   > n8 = Switch on Int!(execute:Exec, Selection:int=Get WorkflowType(self=Get 3DHBDataAsset()) -> (WorkflowType=0)) -> (Default:Exec, 0:Exec, 1:Exec)
   n9 = Get NiagaraComponent(self=@n1)
   n10 = "Set Niagara Variable (Texture)"!(
    execute=@n8.0
    self:Niagara Particle System Component Object Reference=@n9
    InVariableName:Name=Backpack_DiffuseTex
    Texture:Texture Object Reference=/Game/Tools/3DHB/TextureBaker/RenderTarget/RT_3DHB_DiffuseCaptureOut
   ) -> (then) {is=title K2Node_CallFunction}
   "Set Niagara Variable (Texture)"(
    execute=[@n10.then, @n11.then_1]
    self=@n9
    InVariableName=Backpack_NormalTex
    Texture=/Game/Tools/3DHB/TextureBaker/RenderTarget/RT_3DHB_NormalCaptureOut
   )
   > "Set Niagara Variable (Texture)"(self=@n9, InVariableName=Backpack_WPOTex, Texture=/Game/Tools/3DHB/TextureBaker/RenderTarget/RT_3DHB_HeightBlurTexture)
   n11 = Sequence(execute=@n8.1)
   n12 = Initial!() -> (OutputDelegate, then, Data:3DHologram Backpack Data Asset Object Reference) {is=title K2Node_CustomEvent}
   > n13 = Set 3DHBDataAsset(3DHBDataAsset:3DHologram Backpack Data Asset Object Reference=@n12.Data) -> (Output_Get:3DHologram Backpack Data Asset Object Reference)
   "Set Niagara Variable (Texture)"(
    execute=@n11.then_0
    self=@n9
    InVariableName=Backpack_DiffuseTex
    Texture=Get From2D_SourceTexture!(self:3DHologram Backpack Data Asset Object Reference=@n13) -> (From2D_SourceTexture:Texture 2D Object Reference)
   )
   n14 = "On Clicked (Button)"() {is=title K2Node_ComponentBoundEvent}
   n15 = Get PreiviewActor()
   Is Valid(exec=[@n14.then, @n26.then, @n27.then], InputObject=@n15)
   > Destroy Actor(self=@n15)
   > Set PreiviewActor()
   UpdateState() {is=title K2Node_CustomEvent}
   > n16 = Sequence()
   > n17 = Branch(Condition=IsValid(Object=Get MaterialInstance(self=@n13)))
   n18 = Get Button_0() -> (Button_0:Button Object Reference)
   > SetIsEnabled(self=@n18, bInIsEnabled=true)
   > n19 = Switch on Int(Selection=Get WorkflowType(self=@n13) -> (WorkflowType=0))
   n20 = IsValid(Object=Get AchievedNormalTexture!(self:3DHologram Backpack Data Asset Object Reference=@n13) -> (AchievedNormalTexture:Texture 2D Object Reference))
   n21 = Get AchievedDiffuseTexture!(self:3DHologram Backpack Data Asset Object Reference=@n13) -> (AchievedDiffuseTexture:Texture 2D Object Reference)
   n22 = IsValid(Object=Get AchievedWPOTexture!(self:3DHologram Backpack Data Asset Object Reference=@n13) -> (AchievedWPOTexture:Texture 2D Object Reference))
   n23 = Branch(execute=@n19.0, Condition=AND Boolean!(A=IsValid(Object=@n21), B=@n22, C:bool=@n20) -> (ReturnValue))
   Set bIsAchieved(execute=[@n24.then, @n23.then], bIsAchieved:bool=true) -> (Output_Get:bool)
   Set bIsAchieved(execute=[@n24.else, @n23.else], bIsAchieved=false)
   n24 = Branch(execute=@n19.1, Condition=AND Boolean!(A=@n22, B=@n20) -> (ReturnValue))
   SetIsEnabled(execute=[@n17.else, @n19.Default], self=@n18, bInIsEnabled=false)
   > Turnoff()
   n25 = Is Valid(exec=@n16.then_1, InputObject=@n1)
   n26 = Turnoff() {is=title K2Node_CustomEvent}
   Event Construct()
   > SetObject(self=Get DetailsView_59() -> (DetailsView_59:Details View Object Reference), NewObject=Self-Reference())
   n27 = Event Destruct()
   UpdateState(execute=@n13.then) {is=title K2Node_CallFunction}
  ]}]
variables [
 3DHBDataAsset {type=3DHologramBackpackDataAsset_C, default=null}
 bIsAchieved {type=boolean}
 PreiviewActor {type=NiagaraActor, default=null}
 PreviewPosition {type=Vector, default=<0 0 2050>}
]
widgets {
 info {rootWidgetClass=&CanvasPanel, widgetCount=6}
 widgets [
  CanvasPanel_0 {widgetClassPath=&CanvasPanel, widget=CanvasPanel_0}
  Button_0 {
   widgetClassPath=&Button
   parent=CanvasPanel_0
   widget=Button_0
   properties {
    backgroundColor=<rgba 0.1 0.1 0.1 1>
    widgetStyle {
     clickedSlateSound {resourceObject=null}
     disabled {
      bIsDynamicallyLoaded=false
      drawAs=RoundedBox
      imageType=NoImage
      mirroring=NoMirror
      resourceName=None
      resourceObject=null
      tiling=NoTile
      imageSize=<0 0>
      margin=<ltrb 0 0 0 0>
      outlineSettings {
       bUseBrushTransparency=false
       roundingType=FixedRadius
       width=1
       color {colorUseRule=UseColor_ColorTable, specifiedColor=<rgba 0 0 0 0>}
       cornerRadii=<4 4 4 4>
      }
      tintColor {colorUseRule=UseColor_Specified, specifiedColor=<rgba 0.0395 0.0395 0.0395 1>}
      uVRegion {bIsValid=false, max=<0 0>, min=<0 0>}
     }
     disabledForeground {colorUseRule=UseColor_ColorTable, specifiedColor=<rgba 0 0 0 0>}
     hovered {
      bIsDynamicallyLoaded=false
      drawAs=RoundedBox
      imageType=NoImage
      mirroring=NoMirror
      resourceName=None
      resourceObject=null
      tiling=NoTile
      imageSize=<0 0>
      margin=<ltrb 0 0 0 0>
      outlineSettings {
       bUseBrushTransparency=false
       roundingType=FixedRadius
       width=1
       color {colorUseRule=UseColor_ColorTable, specifiedColor=<rgba 0 0 0 0>}
       cornerRadii=<4 4 4 4>
      }
      tintColor {colorUseRule=UseColor_Specified, specifiedColor=<rgba 0.0953 0.0953 0.0953 1>}
      uVRegion {bIsValid=false, max=<0 0>, min=<0 0>}
     }
     hoveredForeground {colorUseRule=UseColor_ColorTable, specifiedColor=<rgba 0 0 0 0>}
     hoveredSlateSound {resourceObject=null}
     normal {
      bIsDynamicallyLoaded=false
      drawAs=RoundedBox
      imageType=NoImage
      mirroring=NoMirror
      resourceName=None
      resourceObject=null
      tiling=NoTile
      imageSize=<0 0>
      margin=<ltrb 0 0 0 0>
      outlineSettings {
       bUseBrushTransparency=false
       roundingType=FixedRadius
       width=1
       color {colorUseRule=UseColor_ColorTable, specifiedColor=<rgba 0 0 0 0>}
       cornerRadii=<4 4 4 4>
      }
      tintColor {colorUseRule=UseColor_Specified, specifiedColor=<rgba 0.0395 0.0395 0.0395 1>}
      uVRegion {bIsValid=false, max=<0 0>, min=<0 0>}
     }
     normalForeground {colorUseRule=UseColor_ColorTable, specifiedColor=<rgba 0 0 0 0>}
     normalPadding=<ltrb 12 1.5 12 1.5>
     pressed {
      bIsDynamicallyLoaded=false
      drawAs=RoundedBox
      imageType=NoImage
      mirroring=NoMirror
      resourceName=None
      resourceObject=null
      tiling=NoTile
      imageSize=<0 0>
      margin=<ltrb 0 0 0 0>
      outlineSettings {
       bUseBrushTransparency=false
       roundingType=FixedRadius
       width=1
       color {colorUseRule=UseColor_ColorTable, specifiedColor=<rgba 0 0 0 0>}
       cornerRadii=<4 4 4 4>
      }
      tintColor {colorUseRule=UseColor_Specified, specifiedColor=<rgba 0.0284 0.0284 0.0284 1>}
      uVRegion {bIsValid=false, max=<0 0>, min=<0 0>}
     }
     pressedForeground {colorUseRule=UseColor_ColorTable, specifiedColor=<rgba 0 0 0 0>}
     pressedPadding=<ltrb 12 2 12 0>
     pressedSlateSound {resourceObject=null}
    }
   }
   slot {class=CanvasPanelSlot, properties {layoutData {alignment=<0 0>, anchors {maximum=<0 0>, minimum=<0 0>}, offsets=<ltrb 10 12 757.2286 40>}}}
  }
  CommonTextBlock_47 {
   widgetClassPath=&CommonTextBlock
   parent=Button_0
   widget=CommonTextBlock_47
   properties {
    style=/Game/UI/Foundation/Items/Parts/TextStyles/ItemCardPowerRatingTextStyle_S.ItemCardPowerRatingTextStyle_S_C
    text=StartPreview
    font {
     bForceMonospaced=false
     bMaterialIsStencil=false
     fontMaterial=null
     fontObject=&NotoSans
     letterSpacing=0
     monospacedWidth=1
     size=18
     skewAmount=0
     typefaceFontName=NotoSans-Bold
     outlineSettings {
      bApplyOutlineToDropShadows=false
      bMiteredCorners=false
      bSeparateFillAlpha=false
      outlineMaterial=null
      outlineSize=0
      outlineColor=<rgba 0 0 0 1>
     }
    }
    shadowOffset=<0 0>
   }
   slot {class=ButtonSlot}
  }
  Button {
   widgetClassPath=&Button
   parent=CanvasPanel_0
   widget=Button
   properties {
    backgroundColor=<rgba 0.1 0.1 0.1 1>
    widgetStyle {
     clickedSlateSound {resourceObject=null}
     disabled {
      bIsDynamicallyLoaded=false
      drawAs=RoundedBox
      imageType=NoImage
      mirroring=NoMirror
      resourceName=None
      resourceObject=null
      tiling=NoTile
      imageSize=<0 0>
      margin=<ltrb 0 0 0 0>
      outlineSettings {
       bUseBrushTransparency=false
       roundingType=FixedRadius
       width=1
       color {colorUseRule=UseColor_ColorTable, specifiedColor=<rgba 0 0 0 0>}
       cornerRadii=<4 4 4 4>
      }
      tintColor {colorUseRule=UseColor_Specified, specifiedColor=<rgba 0.0395 0.0395 0.0395 1>}
      uVRegion {bIsValid=false, max=<0 0>, min=<0 0>}
     }
     disabledForeground {colorUseRule=UseColor_ColorTable, specifiedColor=<rgba 0 0 0 0>}
     hovered {
      bIsDynamicallyLoaded=false
      drawAs=RoundedBox
      imageType=NoImage
      mirroring=NoMirror
      resourceName=None
      resourceObject=null
      tiling=NoTile
      imageSize=<0 0>
      margin=<ltrb 0 0 0 0>
      outlineSettings {
       bUseBrushTransparency=false
       roundingType=FixedRadius
       width=1
       color {colorUseRule=UseColor_ColorTable, specifiedColor=<rgba 0 0 0 0>}
       cornerRadii=<4 4 4 4>
      }
      tintColor {colorUseRule=UseColor_Specified, specifiedColor=<rgba 0.0953 0.0953 0.0953 1>}
      uVRegion {bIsValid=false, max=<0 0>, min=<0 0>}
     }
     hoveredForeground {colorUseRule=UseColor_ColorTable, specifiedColor=<rgba 0 0 0 0>}
     hoveredSlateSound {resourceObject=null}
     normal {
      bIsDynamicallyLoaded=false
      drawAs=RoundedBox
      imageType=NoImage
      mirroring=NoMirror
      resourceName=None
      resourceObject=null
      tiling=NoTile
      imageSize=<0 0>
      margin=<ltrb 0 0 0 0>
      outlineSettings {
       bUseBrushTransparency=false
       roundingType=FixedRadius
       width=1
       color {colorUseRule=UseColor_ColorTable, specifiedColor=<rgba 0 0 0 0>}
       cornerRadii=<4 4 4 4>
      }
      tintColor {colorUseRule=UseColor_Specified, specifiedColor=<rgba 0.0395 0.0395 0.0395 1>}
      uVRegion {bIsValid=false, max=<0 0>, min=<0 0>}
     }
     normalForeground {colorUseRule=UseColor_ColorTable, specifiedColor=<rgba 0 0 0 0>}
     normalPadding=<ltrb 12 1.5 12 1.5>
     pressed {
      bIsDynamicallyLoaded=false
      drawAs=RoundedBox
      imageType=NoImage
      mirroring=NoMirror
      resourceName=None
      resourceObject=null
      tiling=NoTile
      imageSize=<0 0>
      margin=<ltrb 0 0 0 0>
      outlineSettings {
       bUseBrushTransparency=false
       roundingType=FixedRadius
       width=1
       color {colorUseRule=UseColor_ColorTable, specifiedColor=<rgba 0 0 0 0>}
       cornerRadii=<4 4 4 4>
      }
      tintColor {colorUseRule=UseColor_Specified, specifiedColor=<rgba 0.0284 0.0284 0.0284 1>}
      uVRegion {bIsValid=false, max=<0 0>, min=<0 0>}
     }
     pressedForeground {colorUseRule=UseColor_ColorTable, specifiedColor=<rgba 0 0 0 0>}
     pressedPadding=<ltrb 12 2 12 0>
     pressedSlateSound {resourceObject=null}
    }
   }
   slot {class=CanvasPanelSlot, properties {layoutData {alignment=<0 0>, anchors {maximum=<0 0>, minimum=<0 0>}, offsets=<ltrb 10 68 757.2286 40>}}}
  }
  CommonTextBlock {
   widgetClassPath=&CommonTextBlock
   parent=Button
   widget=CommonTextBlock
   properties {
    style=&ItemCardPowerRatingTextStyle_S_C
    text=StopPreview
    font {
     bForceMonospaced=false
     bMaterialIsStencil=false
     fontMaterial=null
     fontObject=&NotoSans
     letterSpacing=0
     monospacedWidth=1
     size=18
     skewAmount=0
     typefaceFontName=NotoSans-Bold
     outlineSettings {
      bApplyOutlineToDropShadows=false
      bMiteredCorners=false
      bSeparateFillAlpha=false
      outlineMaterial=null
      outlineSize=0
      outlineColor=<rgba 0 0 0 1>
     }
    }
    shadowOffset=<0 0>
   }
   slot {class=ButtonSlot}
  }
  DetailsView_59 {
   widgetClassPath=&DetailsView
   parent=CanvasPanel_0
   widget=DetailsView_59
   properties {categoriesToShow [Preview]}
   slot {
    class=CanvasPanelSlot
    properties {layoutData {alignment=<0 0>, anchors {maximum=<0 0>, minimum=<0 0>}, offsets=<ltrb 12 124 753.32 219.1892>}}
   }
  }
 ]
}
