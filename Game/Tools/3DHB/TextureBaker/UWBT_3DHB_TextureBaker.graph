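# Editor Utility Widget that bakes diffuse, normal and WPO textures for a 3D Hologram Backpack data asset, from a static mesh or a 2D texture.
# 3D mode spawns a mesh capture actor, while 2D mode derives height, normal and mask from the source texture.
# Data asset edits update dynamic materials drawn into shared render targets after retriggerable delays.
# Achieve buttons save render targets as texture assets, duplicating a template when none exists.
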EditorUtilityWidgetBlueprint /Game/Tools/3DHB/TextureBaker/UWBT_3DHB_TextureBaker
graphs [
 AchieveTexture {nodes [
   n1 = AchieveTexture!() -> (then, From:Texture Render Target 2D Object Reference, To:Texture 2D Object Reference, TextureType:String) {is=title K2Node_FunctionEntry}
   > n2 = Is Valid(InputObject=@n1.To)
   n3 = Get 3DHologram Backpack Data Asset() -> (3DHologram Backpack Data Asset:3DHologram Backpack Data Asset Object Reference)
   n4 = GetPathNameForLoadedAsset(execute=@n2.Is Not Valid, LoadedAsset=@n3)
   n5 = Get ForCopyTexture() -> (ForCopyTexture:Texture 2D Object Reference)
   > n6 = GetPathNameForLoadedAsset(LoadedAsset=@n5)
   n7 = GetObjectName(Object=@n5)
   n8 = Append!(A=@n7, B=., C:String=@n7) -> (ReturnValue)
   n9 = GetObjectName(Object=@n3)
   n10 = Append(A=@n9, B=., C=@n9)
   n11 = Append!(
    A=Append!(A=Split(SourceString=@n4, InStr=@n10, SearchCase=CaseSensitive, SearchDir=FromEnd), B=Texture/) -> (ReturnValue)
    B=T_
    C=Get TaskName!(self:3DHologram Backpack Data Asset Object Reference=Get 3DHologram Backpack Data Asset()) -> (TaskName:String)
    D:String=_
    E:String=@n1.TextureType
   ) -> (ReturnValue)
   > DuplicateAsset!(
    execute
    SourceAssetPath:String=Append(A=Split(SourceString=@n6, InStr=@n8, SearchCase=CaseSensitive, SearchDir=FromEnd), B=@n7)
    DestinationAssetPath:String=@n11
   ) -> (then, ReturnValue:Object Reference)
   > n12 = LoadAsset(AssetPath=@n11)
   n13 = Cast To Texture2D!(Object=@n12) -> (AsTexture 2D, bSuccess:bool=~)
   > ConvertRenderTargetToTexture2DEditorOnly(RenderTarget=@n1, Texture=@n13)
   > n14 = Sequence!(execute) -> (then_0, then_1, then_2:Exec)
   n15 = Switch on String!(execute:Exec=[@n20.then, @n14.then_0], Selection:String=@n1.TextureType) -> (Default:Exec, WPO:Exec, Normal:Exec, Diffuse:Exec) {pinNames [WPO,Normal,Diffuse]}
   n16 = Cast To Texture2D(Object=SelectObject!(A:Object Reference=@n1.To, B:Object Reference=@n13, bSelectA:bool=IsValid(Object=@n1.To)) -> (ReturnValue:Object Reference)) -> (bSuccess=~)
   SetEditorProperty(
    execute=[@n18.then, @n15.Diffuse]
    Object=@n16
    PropertyName=CompressionSettings
    PropertyValue:TextureCompressionSettings Enum=Get CD() -> (CD:TextureCompressionSettings Enum)
   )
   n17 = OR Boolean(A=true, B=false)
   > SetEditorProperty(Object=@n16, PropertyName=SRGB, PropertyValue:bool=@n17)
   SetEditorProperty(
    execute=@n15.Normal
    Object=@n16
    PropertyName=CompressionSettings
    PropertyValue:TextureCompressionSettings Enum=Get CN() -> (CN:TextureCompressionSettings Enum)
   )
   > SetEditorProperty(Object=@n16, PropertyName=SRGB, PropertyValue:bool=OR Boolean(A=false, B=false))
   n18 = SetEditorProperty(execute=@n15.WPO, Object=@n16, PropertyName=CompressionNoAlpha, PropertyValue=@n17)
   n19 = Switch on String(execute=@n14.then_1, Selection=@n1.TextureType) {pinNames [WPO,Normal,Diffuse]}
   Set AchievedDiffuseTexture!(execute=@n19.Diffuse, AchievedDiffuseTexture:Texture 2D Object Reference=@n13, self:3DHologram Backpack Data Asset Object Reference=@n3) -> (then, Output_Get:Texture 2D Object Reference)
   Set AchievedNormalTexture!(execute=@n19.Normal, AchievedNormalTexture:Texture 2D Object Reference=@n13, self:3DHologram Backpack Data Asset Object Reference=@n3) -> (then, Output_Get:Texture 2D Object Reference)
   Set AchievedWPOTexture!(execute=@n19.WPO, AchievedWPOTexture:Texture 2D Object Reference=@n13, self:3DHologram Backpack Data Asset Object Reference=@n3) -> (then, Output_Get:Texture 2D Object Reference)
   n20 = ConvertRenderTargetToTexture2DEditorOnly(execute=@n2.Is Valid, RenderTarget=@n1, Texture=@n1.To)
  ]}
 EventGraph {nodes [
   n1 = "On Property Changed (3D)"!() -> (OutputDelegate, then, PropertyName:Name=None) {is=title K2Node_ComponentBoundEvent}
   > PrintString(InString="To String (Name)"(InName=@n1.PropertyName))
   > n2 = Branch(Condition="Equal (Name)"!(A:Name=@n1.PropertyName, B:Name=From3D_Resolution) -> (ReturnValue:bool) {is=title K2Node_CallFunction})
   RetriggerableDelay!(execute=[@n14.then, @n2.then], WorldContextObject:Object Reference, Duration=0.2) -> (then)
   > n3 = Sequence!(execute) -> (then_0, then_1)
   > ClearBaker()
   3DBakeReload(execute=@n3.then_1)
   3DBakerParameterRefresh(execute=[@n12.then, @n14.else, @n2.else])
   "On Clicked (2DWPOAB)"() {is=title K2Node_ComponentBoundEvent}
   > n4 = AchieveTexture!(
    execute
    self
    From:Texture Render Target 2D Object Reference=/Game/Tools/3DHB/TextureBaker/RenderTarget/RT_3DHB_HeightBlurTexture
    To:Texture 2D Object Reference=Get AchievedWPOTexture!(self:3DHologram Backpack Data Asset Object Reference=Get 3DHologram Backpack Data Asset()) -> (AchievedWPOTexture:Texture 2D Object Reference)
    TextureType:String=WPO
   ) -> (then) {is=title K2Node_CallFunction}
   DelayUntilNextTick!(
    execute=[@n27.then, @n4.then, @n18.then, @n25.then, @n26.then, @n5.then, @n6.then, @n7.then, @n28.then, @n29.then]
    WorldContextObject:Object Reference
   ) -> (then)
   > Call Property Update!(execute:Exec, self:UWBT 3DHB Texture Baker Object Reference) -> (then:Exec) {is=title K2Node_CallDelegate}
   n5 = "On Property Changed (3DDA)"!() -> (OutputDelegate, then, PropertyName:Name=None) {is=title K2Node_ComponentBoundEvent}
   n6 = "On Property Changed (3DNA)"!() -> (OutputDelegate, then, PropertyName:Name=None) {is=title K2Node_ComponentBoundEvent}
   n7 = "On Property Changed (3DWPOA)"!() -> (OutputDelegate, then, PropertyName:Name=None) {is=title K2Node_ComponentBoundEvent}
   "On Clicked (Button_187)"() {is=title K2Node_ComponentBoundEvent}
   n8 = Get 3DCaptureActor() -> (3DCaptureActor:BP 3DHB Mesh Captuer Object Reference)
   > Is Valid(InputObject=@n8)
   > ToggleCaptureFreeze(self:BP 3DHB Mesh Captuer Object Reference=@n8)
   > n9 = Flip Flop!(None:Exec) -> (A:Exec, B:Exec, IsA:bool=~) {is=title K2Node_MacroInstance}
   n10 = Get ToggleText() -> (ToggleText:Common Text Object Reference)
   > SetColorAndOpacity(self=@n10, InColorAndOpacity=Make Slate Color(SpecifiedColor=(R=1,G=0.5773,B=0,A=1)))
   SetColorAndOpacity(execute=@n9.B, self=@n10, InColorAndOpacity=Make Slate Color(SpecifiedColor=(R=1,G=1,B=1,A=1)))
   n11 = "On Property Changed (3D_ColorTweak)"!() -> (OutputDelegate, then, PropertyName:Name=None) {is=title K2Node_ComponentBoundEvent}
   > n12 = PrintString(InString="To String (Name)"(InName=@n11.PropertyName))
   n13 = "On Property Changed (3D_Mesh)"!() -> (OutputDelegate, then, PropertyName:Name=None) {is=title K2Node_ComponentBoundEvent}
   > PrintString(InString="To String (Name)"(InName=@n13.PropertyName))
   > n14 = Branch(Condition="Equal (Name)"(A=@n13.PropertyName, B=From3D_SourceStaticmesh))
   n15 = "On Property Changed (2D_TextureArtist)"!() -> (OutputDelegate, then, PropertyName:Name=None) {is=title K2Node_ComponentBoundEvent}
   > n16 = Branch(Condition="Equal (Name)"(A=@n15.PropertyName, B=From2D_SourceTexture))
   > n17 = Sequence()
   > ClearBaker()
   2DBakeReload(execute=@n17.then_1)
   2DBakerParameterRefresh(execute=[@n24.then, @n16.else])
   "On Clicked (2DNAB)"() {is=title K2Node_ComponentBoundEvent}
   > n18 = AchieveTexture(
    From=/Game/Tools/3DHB/TextureBaker/RenderTarget/RT_3DHB_NormalTexture
    To=Get AchievedNormalTexture!(self:3DHologram Backpack Data Asset Object Reference=Get 3DHologram Backpack Data Asset()) -> (AchievedNormalTexture:Texture 2D Object Reference)
    TextureType=Normal
   )
   n19 = "On Selection Changed (ComboBoxString_122)"!() -> (OutputDelegate, then, SelectedItem:String, SelectionType:ESelectInfo Enum) {is=title K2Node_ComponentBoundEvent}
   n20 = Sequence!(execute=[@n19.then, @n77.then]) -> (then_0, then_1, then_2:Exec)
   > Set WorkflowType!(
    execute
    WorkflowType:int=GetSelectedIndex!(self:"Combo Box (String) Object Reference"=Get ComboBoxString_122() -> (ComboBoxString_122:"Combo Box (String) Object Reference")) -> (ReturnValue:int=0)
    self:3DHologram Backpack Data Asset Object Reference=Get 3DHologram Backpack Data Asset()
   ) -> (then, Output_Get:int=0)
   ClearBaker(execute=@n20.then_1)
   n21 = Get ComboBoxString_122()
   n22 = Branch(execute=@n20.then_2, Condition="Equal (Integer)"(A=GetSelectedIndex(self=@n21) -> (ReturnValue=0), B=0) {is=title K2Node_CallFunction})
   n23 = Get DetailPanel() -> (DetailPanel:Common Animated Switcher Object Reference)
   > SetActiveWidgetIndex(self=@n23, Index=0)
   > 3DBakeReload()
   SetActiveWidgetIndex(execute=@n22.else, self=@n23, Index=1)
   > 2DBakeReload()
   n24 = "On Property Changed (2D)"!() -> (OutputDelegate, then, PropertyName:Name=None) {is=title K2Node_ComponentBoundEvent}
   "On Clicked (3DDAB)"() {is=title K2Node_ComponentBoundEvent}
   > n25 = AchieveTexture(
    From=/Game/Tools/3DHB/TextureBaker/RenderTarget/RT_3DHB_DiffuseCaptureOut
    To=Get AchievedDiffuseTexture!(self:3DHologram Backpack Data Asset Object Reference=Get 3DHologram Backpack Data Asset()) -> (AchievedDiffuseTexture:Texture 2D Object Reference)
    TextureType=Diffuse
   )
   "On Clicked (3DNAB)"() {is=title K2Node_ComponentBoundEvent}
   > n26 = AchieveTexture(
    From=/Game/Tools/3DHB/TextureBaker/RenderTarget/RT_3DHB_NormalCaptureOut
    To=Get AchievedNormalTexture(self=Get 3DHologram Backpack Data Asset())
    TextureType=Normal
   )
   "On Clicked (3DWPOAB)"() {is=title K2Node_ComponentBoundEvent}
   > n27 = AchieveTexture(
    From=&RT_3DHB_HeightBlurTexture
    To=Get AchievedWPOTexture(self=Get 3DHologram Backpack Data Asset())
    TextureType=WPO
   )
   n28 = "On Property Changed (2DNA)"!() -> (OutputDelegate, then, PropertyName:Name=None) {is=title K2Node_ComponentBoundEvent}
   n29 = "On Property Changed (2DWPOA)"!() -> (OutputDelegate, then, PropertyName:Name=None) {is=title K2Node_ComponentBoundEvent}
   n30 = Deactive3DBaker() {is=title K2Node_CustomEvent}
   n31 = Get 3DCaptureActor()
   Is Valid(exec=[@n90.then, @n30.then, @n113.then], InputObject=@n31)
   > Destroy Actor(self=@n31)
   n32 = "InitialRT(Delay)"!() -> (OutputDelegate, then, TextureSize:Int Point Structure) {is=title K2Node_CustomEvent}
   > n33 = SpawnActor Niagara Particle System Actor!(
    execute:Exec
    WorldContextObject:Object Reference
    Class:Actor Class Reference=/Script/Niagara.NiagaraActor
    SpawnTransform:Transform=MakeTransform(Rotation="0, 0, 0")
    CollisionHandlingOverride:ESpawnActorCollisionHandlingMethod Enum=AlwaysSpawn
    TransformScaleMethod:ESpawnActorScaleMethod Enum=MultiplyWithRoot
    Owner:Actor Object Reference
    Instigator:Pawn Object Reference
   ) -> (then:Exec, ReturnValue:Niagara Particle System Actor Object Reference) {is=title K2Node_SpawnActorFromClass}
   n34 = Get NiagaraComponent!(self:Niagara Particle System Actor Object Reference=@n33) -> (NiagaraComponent:Niagara Particle System Component Object Reference)
   > SetActive!(execute, self:Actor Component Object Reference=@n34, bNewActive:bool=false, bReset:bool=true) -> (then)
   > Set Niagara System Asset!(
    execute
    self:Niagara Particle System Component Object Reference=@n34
    InAsset:Niagara System Object Reference=/Game/Tools/3DHB/TextureBaker/Niagara/NS_3DHB_InitialRenderTarget
    bResetExistingOverrideParameters:bool=true
   ) -> (then) {is=title K2Node_CallFunction}
   n35 = Break Int Point!(IntPoint:"Int Point Structure (by ref)"=@n32.TextureSize) -> (X:int, Y:int) {is=title K2Node_BreakStruct}
   > SetIntParameter!(execute, self:FXSystem Component Object Reference=@n34, ParameterName:Name=ResU, Param:int=@n35) -> (then)
   > SetIntParameter(self=@n34, ParameterName=ResV, Param=@n35.Y)
   > SetActive(self=@n34, bNewActive=true, bReset=false)
   n36 = Get MasterPanel() -> (MasterPanel:Canvas Panel Object Reference)
   > SetVisibility(self=@n36, InVisibility=HitTestInvisible)
   > Delay!(execute, WorldContextObject:Object Reference, Duration=0.35) -> (then)
   > SetVisibility(self=@n36, InVisibility=SelfHitTestInvisible)
   > Destroy Actor(self=@n33)
   > ClearRT()
   ClearRT() {is=title K2Node_CustomEvent}
   > ClearRenderTarget2D!(
    execute
    WorldContextObject:Object Reference
    TextureRenderTarget:Texture Render Target 2D Object Reference=/Game/Tools/3DHB/TextureBaker/RenderTarget/RT_3DHB_HeightTexture
    ClearColor:Linear Color Structure=(R=0,G=0,B=0,A=1)
   ) -> (then)
   > ClearRenderTarget2D(TextureRenderTarget=/Game/Tools/3DHB/TextureBaker/RenderTarget/RT_3DHB_HTN_Blur, ClearColor=(R=0,G=0,B=0,A=1))
   > ClearRenderTarget2D(TextureRenderTarget=&RT_3DHB_NormalTexture, ClearColor=(R=0,G=0,B=0,A=1))
   > ClearRenderTarget2D(TextureRenderTarget=/Game/Tools/3DHB/TextureBaker/RenderTarget/RT_3DHB_MeshMaskTexture, ClearColor=(R=0,G=0,B=0,A=1))
   > ClearRenderTarget2D(TextureRenderTarget=/Game/Tools/3DHB/TextureBaker/RenderTarget/RT_3DHB_DepthCaptureOut, ClearColor=(R=0,G=0,B=0,A=1))
   > ClearRenderTarget2D(TextureRenderTarget=&RT_3DHB_DiffuseCaptureOut, ClearColor=(R=0,G=0,B=0,A=1))
   > ClearRenderTarget2D(TextureRenderTarget=&RT_3DHB_NormalCaptureOut, ClearColor=(R=0,G=0,B=0,A=1))
   > ClearRenderTarget2D(TextureRenderTarget=&RT_3DHB_HeightBlurTexture, ClearColor=(R=0,G=0,B=0,A=1))
   > ClearRenderTarget2D(TextureRenderTarget=/Game/Tools/3DHB/TextureBaker/RenderTarget/RT_3DHB_UnlitCaptureOut, ClearColor=(R=0,G=0,B=0,A=1))
   > ClearRenderTarget2D(TextureRenderTarget=/Game/Tools/3DHB/TextureBaker/RenderTarget/RT_3DHB_DiffuseCaptureOutOrigin, ClearColor=(R=0,G=0,B=0,A=1))
   Initial2DBaker() {is=title K2Node_CustomEvent}
   > n37 = Sequence!(execute) -> (then_0, then_1, then_2, then_3:Exec, then_4:Exec, then_5:Exec)
   > n38 = CreateDynamicMaterialInstance(Parent=/Game/Tools/3DHB/TextureBaker/BakeShader/M_3DHB_ColorToHeight, OptionalName=None, CreationFlags=None)
   > Set MID_ColorToHeight(MID_ColorToHeight:Material Instance Dynamic Object Reference=@n38) -> (Output_Get:Material Instance Dynamic Object Reference)
   > SetBrushFromMaterial(self=Get DepthTexture() -> (DepthTexture:Image Object Reference), Material=@n38)
   n39 = CreateDynamicMaterialInstance(execute=@n37.then_1, Parent=/Game/Tools/3DHB/TextureBaker/BakeShader/M_3DHB_GaussianBlur, OptionalName=None, CreationFlags=None)
   > Set MID_PreNormalBlur(MID_PreNormalBlur:Material Instance Dynamic Object Reference=@n39) -> (Output_Get:Material Instance Dynamic Object Reference)
   n40 = CreateDynamicMaterialInstance(execute=@n37.then_2, Parent=/Game/Tools/3DHB/TextureBaker/BakeShader/M_3DHB_HeightToNormal, OptionalName=None, CreationFlags=None)
   > n41 = Set MID_HeightToNormal(MID_HeightToNormal:Material Instance Dynamic Object Reference=@n40) -> (Output_Get:Material Instance Dynamic Object Reference)
   > SetBrushFromMaterial(self=Get NormalTexture() -> (NormalTexture:Image Object Reference), Material=@n41)
   n42 = CreateDynamicMaterialInstance(execute=@n37.then_3, Parent=&M_3DHB_GaussianBlur, OptionalName=None, CreationFlags=None)
   > n43 = Set MID_HeightToMeshMask(MID_HeightToMeshMask:Material Instance Dynamic Object Reference=@n42) -> (Output_Get:Material Instance Dynamic Object Reference)
   > SetBrushFromMaterial(self=Get MeshMaskTexture() -> (MeshMaskTexture:Image Object Reference), Material=@n43)
   n44 = CreateDynamicMaterialInstance(execute=@n37.then_4, Parent=&M_3DHB_GaussianBlur, OptionalName=None, CreationFlags=None)
   > n45 = Set MID_HeightBlur(MID_HeightBlur:Material Instance Dynamic Object Reference=@n44) -> (Output_Get:Material Instance Dynamic Object Reference)
   > SetBrushFromMaterial(self=Get WPOTexture() -> (WPOTexture:Image Object Reference), Material=@n45)
   n46 = Get From2D_SourceTexture!(self:3DHologram Backpack Data Asset Object Reference=Get 3DHologram Backpack Data Asset()) -> (From2D_SourceTexture:Texture 2D Object Reference)
   SetBrushFromTexture!(
    execute=@n37.then_5
    self:Image Object Reference=Get DiffuseTexture() -> (DiffuseTexture:Image Object Reference)
    Texture:Texture 2D Object Reference=@n46
    bMatchSize:bool=false
   ) -> (then)
   3DBakerParameterRefresh() {is=title K2Node_CustomEvent}
   > n47 = Sequence!(execute) -> (then_0, then_1, then_2, then_3, then_4, then_5, then_6:Exec, then_7:Exec)
   n48 = Get 3DHologram Backpack Data Asset()
   n49 = Get MID_CapturePostDepth() -> (MID_CapturePostDepth:Material Instance Dynamic Object Reference)
   > SetScalarParameterValue(
    self=@n49
    ParameterName=Remap_Far
    Value=Get From3D_Height_RemapRange_Far!(self:3DHologram Backpack Data Asset Object Reference=@n48) -> (From3D_Height_RemapRange_Far:double)
   )
   > SetScalarParameterValue(
    self=@n49
    ParameterName=Remap_Near
    Value=Get From3D_Height_RemapRange_Near!(self:3DHologram Backpack Data Asset Object Reference=@n48) -> (From3D_Height_RemapRange_Near:double)
   )
   n50 = Get MID_DiffuseMerge() -> (MID_DiffuseMerge:Material Instance Dynamic Object Reference)
   > SetScalarParameterValue(
    self=@n50
    ParameterName=Unlit
    Value=Get From3D_CaptureUnlit!(self:3DHologram Backpack Data Asset Object Reference=@n48) -> (From3D_CaptureUnlit:double)
   )
   n51 = Get 3DHologram Backpack Data Asset()
   > SetScalarParameterValue(
    self=@n50
    ParameterName=ColorTweakA_RangeOffset
    Value=Get ColorTweakA_RangeOffset!(self:3DHologram Backpack Data Asset Object Reference=@n51) -> (ColorTweakA_RangeOffset:double)
   )
   > SetVectorParameterValue(
    self=@n50
    ParameterName="ColorTweakA_TargetColor(Hue)"
    Value="Get ColorTweakA_TargetColor(Hue)"!(self:3DHologram Backpack Data Asset Object Reference=@n51) -> ("ColorTweakA_TargetColor(Hue)":Linear Color Structure)
   )
   > SetScalarParameterValue(
    self=@n50
    ParameterName=ColorTweakA_RangeSharpness
    Value="Get ColorTweakA_RangeMultiply(SharpAndStr)"!(self:3DHologram Backpack Data Asset Object Reference=@n51) -> ("ColorTweakA_RangeMultiply(SharpAndStr)":double)
   )
   > SetVectorParameterValue(
    self=@n50
    ParameterName=ColorTweakA_MulColor
    Value=Get ColorTweakA_MulColor!(self:3DHologram Backpack Data Asset Object Reference=@n51) -> (ColorTweakA_MulColor:Linear Color Structure)
   )
   > SetScalarParameterValue(
    self=@n50
    ParameterName=ColorTweakA_Desaturation
    Value=Get ColorTweakA_Desaturation!(self:3DHologram Backpack Data Asset Object Reference=@n51) -> (ColorTweakA_Desaturation:double)
   )
   > SetVectorParameterValue(
    self=@n50
    ParameterName=ColorTweakA_Add
    Value=Get ColorTweakA_AddColor!(self:3DHologram Backpack Data Asset Object Reference=@n51) -> (ColorTweakA_AddColor:Linear Color Structure)
   )
   > SetScalarParameterValue(
    self=@n50
    ParameterName=ColorTweakA_HueShift
    Value=Get ColorTweakA_HueShift!(self:3DHologram Backpack Data Asset Object Reference=@n51) -> (ColorTweakA_HueShift:double)
   )
   n52 = Get 3DHologram Backpack Data Asset()
   n53 = Get MID_DiffuseMerge()
   > SetScalarParameterValue(
    self=@n53
    ParameterName=ColorTweakB_RangeOffset
    Value=Get ColorTweakB_RangeOffset!(self:3DHologram Backpack Data Asset Object Reference=@n52) -> (ColorTweakB_RangeOffset:double)
   )
   > SetVectorParameterValue(
    self=@n53
    ParameterName="ColorTweakB_TargetColor(Hue)"
    Value="Get ColorTweakB_TargetColor(Hue)"!(self:3DHologram Backpack Data Asset Object Reference=@n52) -> ("ColorTweakB_TargetColor(Hue)":Linear Color Structure)
   )
   > SetScalarParameterValue(
    self=@n53
    ParameterName=ColorTweakB_RangeSharpness
    Value="Get ColorTweakB_RangeMultiply(SharpAndStr)"!(self:3DHologram Backpack Data Asset Object Reference=@n52) -> ("ColorTweakB_RangeMultiply(SharpAndStr)":double)
   )
   > SetVectorParameterValue(
    self=@n53
    ParameterName=ColorTweakB_MulColor
    Value=Get ColorTweakB_MulColor!(self:3DHologram Backpack Data Asset Object Reference=@n52) -> (ColorTweakB_MulColor:Linear Color Structure)
   )
   > SetScalarParameterValue(
    self=@n53
    ParameterName=ColorTweakB_Desaturation
    Value=Get ColorTweakB_Desaturation!(self:3DHologram Backpack Data Asset Object Reference=@n52) -> (ColorTweakB_Desaturation:double)
   )
   > SetVectorParameterValue(
    self=@n53
    ParameterName=ColorTweakB_Add
    Value=Get ColorTweakB_AddColor!(self:3DHologram Backpack Data Asset Object Reference=@n52) -> (ColorTweakB_AddColor:Linear Color Structure)
   )
   > SetScalarParameterValue(
    self=@n53
    ParameterName=ColorTweakB_HueShift
    Value=Get ColorTweakB_HueShift!(self:3DHologram Backpack Data Asset Object Reference=@n52) -> (ColorTweakB_HueShift:double)
   )
   n54 = Get 3DHologram Backpack Data Asset()
   > SetScalarParameterValue(
    self=@n53
    ParameterName=ColorTweakGlobal_Desaturation
    Value=Get ColorTweakGlobal_Desaturation!(self:3DHologram Backpack Data Asset Object Reference=@n54) -> (ColorTweakGlobal_Desaturation:double)
   )
   > SetVectorParameterValue(
    self=@n53
    ParameterName=ColorTweakGlobal_MulColor
    Value=Get ColorTweakGlobal_MulColor!(self:3DHologram Backpack Data Asset Object Reference=@n54) -> (ColorTweakGlobal_MulColor:Linear Color Structure)
   )
   > SetScalarParameterValue(
    self=@n53
    ParameterName=ColorTweakGlobal_HueShift
    Value=Get ColorTweakGlobal_HueShift!(self:3DHologram Backpack Data Asset Object Reference=@n54) -> (ColorTweakGlobal_HueShift:double)
   )
   > SetVectorParameterValue(
    self=@n53
    ParameterName=ColorTweakGlobal_AddColor
    Value=Get ColorTweakGlobal_AddColor!(self:3DHologram Backpack Data Asset Object Reference=@n54) -> (ColorTweakGlobal_AddColor:Linear Color Structure)
   )
   > SetScalarParameterValue(
    self=@n53
    ParameterName=ColorTweakGlobal_Contrast
    Value=Get ColorTweakGlobal_Contrast!(self:3DHologram Backpack Data Asset Object Reference=@n54) -> (ColorTweakGlobal_Contrast:double)
   )
   n55 = Get Skull!(self:BP 3DHB Mesh Captuer Object Reference=Get 3DCaptureActor()) -> (Skull:Static Mesh Component Object Reference)
   SetStaticMesh(
    execute=@n47.then_1
    self=@n55
    NewMesh=Get From3D_SourceStaticmesh!(self:3DHologram Backpack Data Asset Object Reference=@n48) -> (From3D_SourceStaticmesh:Static Mesh Object Reference)
   )
   > Set Relative Transform!(
    execute
    self:Scene Component Object Reference=@n55
    NewTransform:"Transform (by ref)"=MakeTransform(
     Location=Get From3D_MeshTransform_Offset!(self:3DHologram Backpack Data Asset Object Reference=@n48) -> (From3D_MeshTransform_Offset:Vector)
     Rotation=Get From3D_MeshTransform_Rotation!(self:3DHologram Backpack Data Asset Object Reference=@n48) -> (From3D_MeshTransform_Rotation:Rotator)
     Scale=Get From3D_MeshTransform_Scale!(self:3DHologram Backpack Data Asset Object Reference=@n48) -> (From3D_MeshTransform_Scale:Vector)
    )
    bSweep:bool=false
    bTeleport:bool=false
   ) -> (then, SweepHitResult:Hit Result Structure) {is=title K2Node_CallFunction}
   n56 = Get MID_HeightBlur() -> (MID_HeightBlur:Material Instance Dynamic Object Reference)
   SetTextureParameterValue(execute=@n47.then_2, self=@n56, ParameterName=Texture, Value=&RT_3DHB_DepthCaptureOut)
   > SetScalarParameterValue(
    self=@n56
    ParameterName=Samples
    Value=Get From3D_WPOTexture_Blur!(self:3DHologram Backpack Data Asset Object Reference=@n48) -> (From3D_WPOTexture_Blur:double)
   )
   > SetScalarParameterValue(
    self=@n56
    ParameterName=Brightness
    Value=Get From3D_WPOTexture_Brightness!(self:3DHologram Backpack Data Asset Object Reference=@n48) -> (From3D_WPOTexture_Brightness:double)
   )
   n57 = Get MID_HeightToMeshMask() -> (MID_HeightToMeshMask:Material Instance Dynamic Object Reference)
   SetTextureParameterValue(execute=@n47.then_3, self=@n57, ParameterName=Texture, Value=&RT_3DHB_HTN_Blur)
   > SetScalarParameterValue(
    self=@n57
    ParameterName=Samples
    Value=Get From3D_DiffuseToMeshMask_Blur!(self:3DHologram Backpack Data Asset Object Reference=@n48) -> (From3D_DiffuseToMeshMask_Blur:double)
   )
   > SetScalarParameterValue(self=@n57, ParameterName=UseDiffuse, Value=1)
   RetriggerableDelay(execute=@n47.then_4, Duration=0.2)
   > DrawMaterialToRenderTarget!(
    execute
    WorldContextObject:Object Reference
    TextureRenderTarget:Texture Render Target 2D Object Reference=&RT_3DHB_MeshMaskTexture
    Material:Material Interface Object Reference=Get MID_HeightToMeshMask()
   ) -> (then)
   > DelayUntilNextTick()
   > DrawMaterialToRenderTarget(
    TextureRenderTarget=&RT_3DHB_HTN_Blur
    Material=Get MID_PreNormalBlur() -> (MID_PreNormalBlur:Material Instance Dynamic Object Reference)
   )
   > DelayUntilNextTick()
   > DrawMaterialToRenderTarget(TextureRenderTarget=&RT_3DHB_HeightBlurTexture, Material=Get MID_HeightBlur())
   > DelayUntilNextTick()
   > ClearRenderTarget2D(TextureRenderTarget=&RT_3DHB_DiffuseCaptureOut, ClearColor=(R=0,G=0,B=0,A=1))
   > DrawMaterialToRenderTarget(TextureRenderTarget=&RT_3DHB_DiffuseCaptureOut, Material=Get MID_DiffuseMerge())
   > Call Property Update()
   n58 = Get 3DHologram Backpack Data Asset()
   n59 = Branch(
    execute=@n47.then_5
    Condition=IsValid(Object=Get From3D_ReplaceWPOTexture!(self:3DHologram Backpack Data Asset Object Reference=@n58) -> (From3D_ReplaceWPOTexture:Texture 2D Object Reference))
   )
   n60 = Get WPOTextureVB() -> (WPOTextureVB:Vertical Box Object Reference)
   n61 = Get 3DWPOAH() -> (3DWPOAH:Horizontal Box Object Reference)
   > SetVisibility(self=[@n60, @n61], InVisibility=Collapsed)
   SetVisibility(execute=@n59.else, self=@n60, InVisibility=SelfHitTestInvisible)
   > SetVisibility(self=@n61, InVisibility=Visible)
   n62 = Branch(
    execute=@n47.then_6
    Condition=IsValid(Object=Get From3D_ReplaceNormalTexture!(self:3DHologram Backpack Data Asset Object Reference=@n58) -> (From3D_ReplaceNormalTexture:Texture 2D Object Reference))
   )
   n63 = Get NormalTextureVB() -> (NormalTextureVB:Vertical Box Object Reference)
   n64 = Get 3DNAH() -> (3DNAH:Horizontal Box Object Reference)
   > SetVisibility(self=[@n63, @n64], InVisibility=Collapsed)
   SetVisibility(execute=@n62.else, self=@n63, InVisibility=SelfHitTestInvisible)
   > SetVisibility(self=@n64, InVisibility=Visible)
   n65 = Branch(
    execute=@n47.then_7
    Condition=IsValid(Object=Get From3D_ReplaceDiffuseTexture!(self:3DHologram Backpack Data Asset Object Reference=@n58) -> (From3D_ReplaceDiffuseTexture:Texture 2D Object Reference))
   )
   n66 = Get DiffuseTextureVB() -> (DiffuseTextureVB:Vertical Box Object Reference)
   n67 = Get 3DDAH() -> (3DDAH:Horizontal Box Object Reference)
   n68 = Get 3D_Mesh() -> (3D_Mesh:Details View Object Reference)
   > SetVisibility(self=[@n66, @n67, @n68], InVisibility=Collapsed)
   SetVisibility(execute=@n65.else, self=@n66, InVisibility=SelfHitTestInvisible)
   > SetVisibility(self=[@n67, @n68], InVisibility=Visible)
   2DBakeReload() {is=title K2Node_CustomEvent}
   > n69 = Is Valid(InputObject=@n46)
   ClearBaker(execute=@n69.Is Not Valid)
   n70 = Get Texture Size!(execute=@n69.Is Valid, Texture:Texture Object Reference=@n46, __WorldContext:Object Reference) -> (then, Size:Int Point Structure) {is=title K2Node_CallFunction}
   > "InitialRT(Delay)"!(execute, self, TextureSize:Int Point Structure=@n70) -> (then) {is=title K2Node_CallFunction}
   > Delay(Duration=0.4)
   > Initial2DBaker() {is=title K2Node_CallFunction}
   > 2DBakerParameterRefresh()
   Reactive3DBaker() {is=title K2Node_CustomEvent}
   > n71 = Is Valid(InputObject=Get 3DCaptureActor())
   Initial3DBaker(execute=@n71.Is Not Valid)
   > 3DBakerParameterRefresh() {is=title K2Node_CallFunction}
   n72 = InitialWithDataAsset!() -> (OutputDelegate, then, 3DHBDataAsset:Primary Data Asset Object Reference) {is=title K2Node_CustomEvent}
   > n73 = Cast To 3DHologramBackpackDataAsset!(execute, Object=@n72.3DHBDataAsset) -> (then, CastFailed, As3DHologram Backpack Data Asset:3DHologram Backpack Data Asset Object Reference) {is=title K2Node_DynamicCast}
   > n74 = Set 3DHologram Backpack Data Asset(3DHologram Backpack Data Asset:3DHologram Backpack Data Asset Object Reference=@n73) -> (Output_Get:3DHologram Backpack Data Asset Object Reference)
   > SetObject(
    self=[
     Get 3D_ColorTweak() -> (3D_ColorTweak:Details View Object Reference)
     Get 2D_TextureArtist() -> (2D_TextureArtist:Details View Object Reference)
     Get 3D_Mesh()
     Get 3DDA() -> (3DDA:Single Property View Object Reference)
     Get 2D() -> (2D:Details View Object Reference)
     Get 3DNA() -> (3DNA:Single Property View Object Reference)
     Get 3D() -> (3D:Details View Object Reference)
     Get 3DWPOA() -> (3DWPOA:Single Property View Object Reference)
     Get 2DNA() -> (2DNA:Single Property View Object Reference)
     Get 2DWPOA() -> (2DWPOA:Single Property View Object Reference)
    ]
    NewObject=@n74
   )
   n75 = Get ComboBoxString_122()
   n76 = Get WorkflowType!(self:3DHologram Backpack Data Asset Object Reference=@n74) -> (WorkflowType:int=0)
   > n77 = Branch(Condition="Equal (Integer)"(A=GetSelectedIndex(self=@n75) -> (ReturnValue=0), B=@n76))
   SetSelectedIndex!(execute=@n77.else, self:"Combo Box (String) Object Reference"=@n75, Index:int=@n76) -> (then)
   Initial3DBaker() {is=title K2Node_CustomEvent}
   > n78 = SpawnActor BP 3DHB Mesh Captuer!(
    execute:Exec
    WorldContextObject:Object Reference
    Class:Actor Class Reference=/Game/Tools/3DHB/TextureBaker/MeshCaputer/BP_3DHB_MeshCaptuer.BP_3DHB_MeshCaptuer_C
    SpawnTransform:Transform=MakeTransform(Location="700,-1500,3000", Rotation="0, 0, 0")
    CollisionHandlingOverride:ESpawnActorCollisionHandlingMethod Enum=AlwaysSpawn
    TransformScaleMethod:ESpawnActorScaleMethod Enum=OverrideRootScale
    Owner:Actor Object Reference
    Instigator:Pawn Object Reference
   ) -> (then:Exec, ReturnValue:BP 3DHB Mesh Captuer Object Reference) {is=title K2Node_SpawnActorFromClass}
   > n79 = CreateDynamicMaterialInstance(Parent=/Game/Tools/3DHB/TextureBaker/MeshCaputer/PostShader/M_3DHB_PostDepth, OptionalName=None, CreationFlags=None)
   > Initial!(execute, self:BP 3DHB Mesh Captuer Object Reference=@n78, PostDepthMID:Material Instance Dynamic Object Reference=@n79) -> (then)
   > Set MID_CapturePostDepth(MID_CapturePostDepth:Material Instance Dynamic Object Reference=@n79) -> (Output_Get:Material Instance Dynamic Object Reference)
   > Set 3DCaptureActor(3DCaptureActor:BP 3DHB Mesh Captuer Object Reference=@n78) -> (Output_Get:BP 3DHB Mesh Captuer Object Reference)
   > n80 = Sequence!(execute) -> (then_0, then_1, then_2, then_3, then_4, then_5, then_6)
   > n81 = CreateDynamicMaterialInstance(Parent=/Game/Tools/3DHB/TextureBaker/MeshCaputer/VisualizeShader/M_3DHB_DepthVis, OptionalName=None, CreationFlags=None)
   > SetBrushFromMaterial(self=Get DepthTexture(), Material=@n81)
   n82 = CreateDynamicMaterialInstance(
    execute=@n80.then_1
    Parent=/Game/Tools/3DHB/TextureBaker/MeshCaputer/VisualizeShader/M_3DHB_NormalVis
    OptionalName=None
    CreationFlags=None
   )
   > SetBrushFromMaterial(self=Get NormalTexture(), Material=@n82)
   n83 = CreateDynamicMaterialInstance(
    execute=@n80.then_2
    Parent=/Game/Tools/3DHB/TextureBaker/MeshCaputer/VisualizeShader/M_3DHB_DiffuseVis
    OptionalName=None
    CreationFlags=None
   )
   > SetBrushFromMaterial(self=Get DiffuseTexture(), Material=@n83)
   n84 = CreateDynamicMaterialInstance(execute=@n80.then_3, Parent=&M_3DHB_GaussianBlur, OptionalName=None, CreationFlags=None)
   > n85 = Set MID_HeightBlur(MID_HeightBlur=@n84)
   > SetBrushFromMaterial(self=Get WPOTexture(), Material=@n85)
   n86 = CreateDynamicMaterialInstance(execute=@n80.then_4, Parent=&M_3DHB_GaussianBlur, OptionalName=None, CreationFlags=None)
   > n87 = Set MID_HeightToMeshMask(MID_HeightToMeshMask=@n86)
   > SetBrushFromMaterial(self=Get MeshMaskTexture(), Material=@n87)
   n88 = CreateDynamicMaterialInstance(execute=@n80.then_5, Parent=/Game/Tools/3DHB/TextureBaker/BakeShader/M_3DHB_HeightToMask, OptionalName=None, CreationFlags=None)
   > Set MID_PreNormalBlur(MID_PreNormalBlur=@n88)
   n89 = CreateDynamicMaterialInstance(execute=@n80.then_6, Parent=/Game/Tools/3DHB/TextureBaker/BakeShader/M_3DHB_MergeDiffuseCapture, OptionalName=None, CreationFlags=None)
   > Set MID_DiffuseMerge(MID_DiffuseMerge:Material Instance Dynamic Object Reference=@n89) -> (Output_Get:Material Instance Dynamic Object Reference)
   ClearBaker() {is=title K2Node_CustomEvent}
   > n90 = SetBrushFromMaterial(
    self=[Get DepthTexture(), Get MeshMaskTexture(), Get NormalTexture(), Get WPOTexture(), Get DiffuseTexture()]
    Material=/Game/Tools/3DHB/DefaultAssets/M_3DHB_EmptyUI
   )
   3DBakeReload() {is=title K2Node_CustomEvent}
   n91 = Get 3DHologram Backpack Data Asset()
   > n92 = Is Valid(InputObject=Get From3D_SourceStaticmesh(self=@n91))
   ClearBaker(execute=@n92.Is Not Valid) {is=title K2Node_CallFunction}
   n93 = Get From3D_Resolution!(self:3DHologram Backpack Data Asset Object Reference=@n91) -> (From3D_Resolution:double)
   "InitialRT(Delay)"(
    execute=@n92.Is Valid
    TextureSize=Make Int Point!(X:int=Truncate(A=@n93), Y:int=Truncate(A=@n93)) -> (IntPoint:Int Point Structure) {is=title K2Node_MakeStruct}
   )
   > Delay(Duration=0.4)
   > Initial3DBaker() {is=title K2Node_CallFunction}
   > Delay(Duration=0.4)
   > 3DBakerParameterRefresh()
   2DBakerParameterRefresh() {is=title K2Node_CustomEvent}
   > n94 = Sequence!(execute) -> (then_0, then_1, then_2, then_3, then_4, then_5, then_6, then_7:Exec, then_8:Exec)
   n95 = Get 3DHologram Backpack Data Asset()
   n96 = Get MID_ColorToHeight() -> (MID_ColorToHeight:Material Instance Dynamic Object Reference)
   > SetScalarParameterValue(
    self=@n96
    ParameterName=Brightness
    Value=Get From2D_ColorToHeight_Brightness!(self:3DHologram Backpack Data Asset Object Reference=@n95) -> (From2D_ColorToHeight_Brightness:double)
   )
   n97 = Get From2D_SourceTexture(self=@n95)
   > SetTextureParameterValue(self=@n96, ParameterName=Texture, Value=@n97)
   > SetScalarParameterValue(
    self=@n96
    ParameterName=Add
    Value=Get From2D_WPOTexture_Add!(self:3DHologram Backpack Data Asset Object Reference=@n95) -> (From2D_WPOTexture_Add:double)
   )
   > SetScalarParameterValue(
    self=@n96
    ParameterName=Invert
    Value=Get From2D_WPOTexture_Invert!(self:3DHologram Backpack Data Asset Object Reference=@n95) -> (From2D_WPOTexture_Invert:double)
   )
   n98 = Get MID_PreNormalBlur()
   SetTextureParameterValue(execute=@n94.then_1, self=@n98, ParameterName=Texture, Value=&RT_3DHB_HeightTexture)
   > SetScalarParameterValue(
    self=@n98
    ParameterName=Samples
    Value=Get From2D_HeightToNormal_Blur!(self:3DHologram Backpack Data Asset Object Reference=@n95) -> (From2D_HeightToNormal_Blur:double)
   )
   n99 = Get Texture Size(execute=@n94.then_2, Texture=@n97)
   n100 = Break Int Point(IntPoint=@n99)
   n101 = Get MID_HeightToNormal() -> (MID_HeightToNormal:Material Instance Dynamic Object Reference)
   > SetScalarParameterValue(self=@n101, ParameterName=ResU, Value="To Float (Integer)"(InInt=@n100))
   > SetScalarParameterValue(self=@n101, ParameterName=ResV, Value="To Float (Integer)"(InInt=@n100.Y))
   > SetScalarParameterValue(
    self=@n101
    ParameterName=NormalStrength
    Value=Get From2D_HeightToNormal_Strength!(self:3DHologram Backpack Data Asset Object Reference=@n95) -> (From2D_HeightToNormal_Strength:double)
   )
   n102 = Get MID_HeightToMeshMask()
   SetTextureParameterValue(execute=@n94.then_3, self=@n102, ParameterName=Texture, Value=@n97)
   > SetScalarParameterValue(
    self=@n102
    ParameterName=Samples
    Value=Get From2D_AlphaToMeshMask_Blur!(self:3DHologram Backpack Data Asset Object Reference=@n95) -> (From2D_AlphaToMeshMask_Blur:double)
   )
   > SetScalarParameterValue(self=@n102, ParameterName=UseAChannel, Value=1)
   n103 = Get MID_HeightBlur()
   SetTextureParameterValue(execute=@n94.then_4, self=@n103, ParameterName=Texture, Value=&RT_3DHB_HeightTexture)
   > SetScalarParameterValue(
    self=@n103
    ParameterName=Samples
    Value=Get From2D_WPOTexture_Blur!(self:3DHologram Backpack Data Asset Object Reference=@n95) -> (From2D_WPOTexture_Blur:double)
   )
   > SetScalarParameterValue(
    self=@n103
    ParameterName=Brightness
    Value=Get From2D_WPOTexture_Brightness!(self:3DHologram Backpack Data Asset Object Reference=@n95) -> (From2D_WPOTexture_Brightness:double)
   )
   > SetScalarParameterValue(self=@n103, ParameterName=DeNoise, Value=0.01)
   RetriggerableDelay(execute=@n94.then_5, Duration=0.2)
   > ClearRT() {is=title K2Node_CallFunction}
   > DelayUntilNextTick()
   > DrawMaterialToRenderTarget(TextureRenderTarget=&RT_3DHB_HeightTexture, Material=Get MID_ColorToHeight())
   > DelayUntilNextTick()
   > DrawMaterialToRenderTarget(TextureRenderTarget=&RT_3DHB_HTN_Blur, Material=Get MID_PreNormalBlur())
   > DelayUntilNextTick()
   > DrawMaterialToRenderTarget(TextureRenderTarget=&RT_3DHB_NormalTexture, Material=Get MID_HeightToNormal())
   > DelayUntilNextTick()
   > DrawMaterialToRenderTarget(TextureRenderTarget=&RT_3DHB_MeshMaskTexture, Material=Get MID_HeightToMeshMask())
   > DelayUntilNextTick()
   > DrawMaterialToRenderTarget(TextureRenderTarget=&RT_3DHB_HeightBlurTexture, Material=Get MID_HeightBlur())
   > DelayUntilNextTick()
   > Call Property Update()
   n104 = Get 3DHologram Backpack Data Asset()
   n105 = Branch(
    execute=@n94.then_6
    Condition=IsValid(Object=Get From2D_ReplaceWPOTexture!(self:3DHologram Backpack Data Asset Object Reference=@n104) -> (From2D_ReplaceWPOTexture:Texture 2D Object Reference))
   )
   n106 = Get WPOTextureVB()
   n107 = Get 2DWPOAH() -> (2DWPOAH:Horizontal Box Object Reference)
   > SetVisibility(self=[@n106, @n107], InVisibility=Collapsed)
   SetVisibility(execute=@n105.else, self=@n106, InVisibility=SelfHitTestInvisible)
   > SetVisibility(self=@n107, InVisibility=Visible)
   n108 = Branch(
    execute=@n94.then_7
    Condition=IsValid(Object=Get From2D_ReplaceNormalTexture!(self:3DHologram Backpack Data Asset Object Reference=@n104) -> (From2D_ReplaceNormalTexture:Texture 2D Object Reference))
   )
   n109 = Get NormalTextureVB()
   n110 = Get 2DNAB() -> (2DNAB:Button Object Reference)
   > SetVisibility(self=[@n109, @n110], InVisibility=Collapsed)
   SetVisibility(execute=@n108.else, self=@n109, InVisibility=SelfHitTestInvisible)
   > SetVisibility(self=@n110, InVisibility=Visible)
   Event Construct()
   > n111 = GetAllActorsOfClass!(
    execute
    WorldContextObject:Object Reference
    ActorClass:Actor Class Reference=&BP_3DHB_MeshCaptuer_C
   ) -> (then, OutActors:Array of BP 3DHB Mesh Captuer Object References)
   > n112 = For Each Loop(Array:Array of BP 3DHB Mesh Captuer Object References=@n111) -> (Array Element:BP 3DHB Mesh Captuer Object Reference)
   > Destroy Actor(self=@n112)
   n113 = Event Destruct()
   Event PreConstruct()
   Event Tick()
  ]}
 PropertyUpdate {nodes [PropertyUpdate() {is=title K2Node_FunctionEntry}]}
]
variables [
 3DCaptureActor {type=BP_3DHB_MeshCaptuer_C, default=null}
 3DHologram Backpack Data Asset {type=3DHologramBackpackDataAsset_C, default=null}
 ForCopyTexture {type=Texture2D, default=/Game/Tools/3DHB/Initializer/ForCopy/T_3DHB_ForCopy}
 MID_CapturePostDepth {type=MaterialInstanceDynamic}
 MID_ColorToHeight {type=MaterialInstanceDynamic}
 MID_DiffuseMerge {type=MaterialInstanceDynamic}
 MID_HeightBlur {type=MaterialInstanceDynamic}
 MID_HeightToMeshMask {type=MaterialInstanceDynamic}
 MID_HeightToNormal {type=MaterialInstanceDynamic}
 MID_PreNormalBlur {type=MaterialInstanceDynamic}
 PropertyUpdate {default=()}
]
widgets {
 info {rootWidgetClass=&CanvasPanel, widgetCount=68}
 widgets [
  MasterPanel {widgetClassPath=&CanvasPanel, bIsVariable=true, widget=MasterPanel}
  HorizontalBox_52 {
   widgetClassPath=&HorizontalBox
   parent=MasterPanel
   widget=HorizontalBox_52
   slot {class=CanvasPanelSlot, properties {layoutData {alignment=<0 0>, anchors {maximum=<1 1>, minimum=<0 0>}, offsets=<ltrb 0 0 0 0>}}}
  }
  VerticalBox_32 {
   widgetClassPath=&VerticalBox
   parent=HorizontalBox_52
   widget=VerticalBox_32
   slot {class=HorizontalBoxSlot, properties {size {sizeRule=Fill, value=0.7}}}
  }
  HorizontalBox_154 {
   widgetClassPath=&HorizontalBox
   parent=VerticalBox_32
   widget=HorizontalBox_154
   slot {class=VerticalBoxSlot, properties {padding=<ltrb 15 15 15 15>}}
  }
  CommonTextBlock_166 {
   widgetClassPath=&CommonTextBlock
   parent=HorizontalBox_154
   widget=CommonTextBlock_166
   properties {
    style=&TextStyle-Base_C
    text="Workflow type : "
    font {
     bForceMonospaced=false
     bMaterialIsStencil=false
     fontMaterial=null
     fontObject=&NotoSans
     letterSpacing=0
     monospacedWidth=1
     size=24
     skewAmount=0
     typefaceFontName=NotoSans-Regular
     outlineSettings {
      bApplyOutlineToDropShadows=false
      bMiteredCorners=false
      bSeparateFillAlpha=false
      outlineMaterial=null
      outlineSize=0
      outlineColor=<rgba 0 0 0 1>
     }
    }
    shadowOffset=<0 0>
   }
   slot {class=HorizontalBoxSlot}
  }
  ComboBoxString_122 {
   widgetClassPath=/Script/UMG.ComboBoxString
   bIsVariable=true
   parent=HorizontalBox_154
   widget=ComboBoxString_122
   properties {
    selectedOption=From 3D asset
    defaultOptions [From 3D asset, From 2D texture]
    foregroundColor {colorUseRule=UseColor_Specified, specifiedColor=<rgba 0.5271 0.5271 0.5271 1>}
    itemStyle {
     bUseParentRowBrush=false
     activeBrush {
      bIsDynamicallyLoaded=false
      drawAs=Image
      imageType=NoImage
      mirroring=NoMirror
      resourceName=None
      resourceObject=null
      tiling=NoTile
      imageSize=<0 0>
      margin=<ltrb 0 0 0 0>
      outlineSettings {
       bUseBrushTransparency=false
       roundingType=HalfHeightRadius
       width=0
       color {colorUseRule=UseColor_Specified, specifiedColor=<rgba 0 0 0 0>}
       cornerRadii=<0 0 0 0>
      }
      tintColor {colorUseRule=UseColor_Specified, specifiedColor=<rgba 0 0.162 0.7454 1>}
      uVRegion {bIsValid=false, max=<0 0>, min=<0 0>}
     }
     activeHighlightedBrush {
      bIsDynamicallyLoaded=false
      drawAs=Image
      imageType=NoImage
      mirroring=NoMirror
      resourceName=None
      resourceObject=null
      tiling=NoTile
      imageSize=<0 0>
      margin=<ltrb 0 0 0 0>
      outlineSettings {
       bUseBrushTransparency=false
       roundingType=HalfHeightRadius
       width=0
       color {colorUseRule=UseColor_Specified, specifiedColor=<rgba 0 0 0 0>}
       cornerRadii=<0 0 0 0>
      }
      tintColor {colorUseRule=UseColor_Specified, specifiedColor=<rgba 0.0252 0.0319 0.0423 1>}
      uVRegion {bIsValid=false, max=<0 0>, min=<0 0>}
     }
     activeHoveredBrush {
      bIsDynamicallyLoaded=false
      drawAs=Image
      imageType=NoImage
      mirroring=NoMirror
      resourceName=None
      resourceObject=null
      tiling=NoTile
      imageSize=<0 0>
      margin=<ltrb 0 0 0 0>
      outlineSettings {
       bUseBrushTransparency=false
       roundingType=HalfHeightRadius
       width=0
       color {colorUseRule=UseColor_Specified, specifiedColor=<rgba 0 0 0 0>}
       cornerRadii=<0 0 0 0>
      }
      tintColor {colorUseRule=UseColor_Specified, specifiedColor=<rgba 0 0.162 0.7454 1>}
      uVRegion {bIsValid=false, max=<0 0>, min=<0 0>}
     }
     dropIndicator_Above {
      bIsDynamicallyLoaded=false
      drawAs=Box
      imageType=FullColor
      mirroring=NoMirror
      resourceName=../../../Engine/Content/Slate/Common/DropZoneIndicator_Above.png
      resourceObject=null
      tiling=NoTile
      imageSize=<0 0>
      margin=<ltrb 0.625 0.625 0 0>
      outlineSettings {
       bUseBrushTransparency=false
       roundingType=HalfHeightRadius
       width=0
       color {colorUseRule=UseColor_Specified, specifiedColor=<rgba 0 0 0 0>}
       cornerRadii=<0 0 0 0>
      }
      tintColor {colorUseRule=UseColor_Specified, specifiedColor=<rgba 0.0091 0.3515 0.9647 1>}
      uVRegion {bIsValid=false, max=<0 0>, min=<0 0>}
     }
     dropIndicator_Below {
      bIsDynamicallyLoaded=false
      drawAs=Box
      imageType=FullColor
      mirroring=NoMirror
      resourceName=../../../Engine/Content/Slate/Common/DropZoneIndicator_Below.png
      resourceObject=null
      tiling=NoTile
      imageSize=<0 0>
      margin=<ltrb 0.625 0 0 0.625>
      outlineSettings {
       bUseBrushTransparency=false
       roundingType=HalfHeightRadius
       width=0
       color {colorUseRule=UseColor_Specified, specifiedColor=<rgba 0 0 0 0>}
       cornerRadii=<0 0 0 0>
      }
      tintColor {colorUseRule=UseColor_Specified, specifiedColor=<rgba 0.0091 0.3515 0.9647 1>}
      uVRegion {bIsValid=false, max=<0 0>, min=<0 0>}
     }
     dropIndicator_Onto {
      bIsDynamicallyLoaded=false
      drawAs=Box
      imageType=FullColor
      mirroring=NoMirror
      resourceName=../../../Engine/Content/Slate/Common/DropZoneIndicator_Onto.png
      resourceObject=null
      tiling=NoTile
      imageSize=<0 0>
      margin=<ltrb 0.25 0.25 0.25 0.25>
      outlineSettings {
       bUseBrushTransparency=false
       roundingType=HalfHeightRadius
       width=0
       color {colorUseRule=UseColor_Specified, specifiedColor=<rgba 0 0 0 0>}
       cornerRadii=<0 0 0 0>
      }
      tintColor {colorUseRule=UseColor_Specified, specifiedColor=<rgba 0.0091 0.3515 0.9647 1>}
      uVRegion {bIsValid=false, max=<0 0>, min=<0 0>}
     }
     evenRowBackgroundBrush {
      bIsDynamicallyLoaded=false
      drawAs=Image
      imageType=NoImage
      mirroring=NoMirror
      resourceName=None
      resourceObject=null
      tiling=NoTile
      imageSize=<0 0>
      margin=<ltrb 0 0 0 0>
      outlineSettings {
       bUseBrushTransparency=false
       roundingType=HalfHeightRadius
       width=0
       color {colorUseRule=UseColor_Specified, specifiedColor=<rgba 0 0 0 0>}
       cornerRadii=<0 0 0 0>
      }
      tintColor {colorUseRule=UseColor_Specified, specifiedColor=<rgba 0.0103 0.0103 0.0103 1>}
      uVRegion {bIsValid=false, max=<0 0>, min=<0 0>}
     }
     evenRowBackgroundHoveredBrush {
      bIsDynamicallyLoaded=false
      drawAs=Image
      imageType=NoImage
      mirroring=NoMirror
      resourceName=None
      resourceObject=null
      tiling=NoTile
      imageSize=<0 0>
      margin=<ltrb 0 0 0 0>
      outlineSettings {
       bUseBrushTransparency=false
       roundingType=HalfHeightRadius
       width=0
       color {colorUseRule=UseColor_Specified, specifiedColor=<rgba 0 0 0 0>}
       cornerRadii=<0 0 0 0>
      }
      tintColor {colorUseRule=UseColor_Specified, specifiedColor=<rgba 0.0176 0.0176 0.0176 1>}
      uVRegion {bIsValid=false, max=<0 0>, min=<0 0>}
     }
     inactiveBrush {
      bIsDynamicallyLoaded=false
      drawAs=Image
      imageType=NoImage
      mirroring=NoMirror
      resourceName=None
      resourceObject=null
      tiling=NoTile
      imageSize=<0 0>
      margin=<ltrb 0 0 0 0>
      outlineSettings {
       bUseBrushTransparency=false
       roundingType=HalfHeightRadius
       width=0
       color {colorUseRule=UseColor_Specified, specifiedColor=<rgba 0 0 0 0>}
       cornerRadii=<0 0 0 0>
      }
      tintColor {colorUseRule=UseColor_Specified, specifiedColor=<rgba 0.0513 0.0953 0.159 1>}
      uVRegion {bIsValid=false, max=<0 0>, min=<0 0>}
     }
     inactiveHighlightedBrush {
      bIsDynamicallyLoaded=false
      drawAs=Image
      imageType=NoImage
      mirroring=NoMirror
      resourceName=None
      resourceObject=null
      tiling=NoTile
      imageSize=<0 0>
      margin=<ltrb 0 0 0 0>
      outlineSettings {
       bUseBrushTransparency=false
       roundingType=HalfHeightRadius
       width=0
       color {colorUseRule=UseColor_Specified, specifiedColor=<rgba 0 0 0 0>}
       cornerRadii=<0 0 0 0>
      }
      tintColor {colorUseRule=UseColor_Specified, specifiedColor=<rgba 0.0252 0.0319 0.0423 1>}
      uVRegion {bIsValid=false, max=<0 0>, min=<0 0>}
     }
     inactiveHoveredBrush {
      bIsDynamicallyLoaded=false
      drawAs=Image
      imageType=NoImage
      mirroring=NoMirror
      resourceName=None
      resourceObject=null
      tiling=NoTile
      imageSize=<0 0>
      margin=<ltrb 0 0 0 0>
      outlineSettings {
       bUseBrushTransparency=false
       roundingType=HalfHeightRadius
       width=0
       color {colorUseRule=UseColor_Specified, specifiedColor=<rgba 0 0 0 0>}
       cornerRadii=<0 0 0 0>
      }
      tintColor {colorUseRule=UseColor_Specified, specifiedColor=<rgba 0.0513 0.0953 0.159 1>}
      uVRegion {bIsValid=false, max=<0 0>, min=<0 0>}
     }
     oddRowBackgroundBrush {
      bIsDynamicallyLoaded=false
      drawAs=Image
      imageType=NoImage
      mirroring=NoMirror
      resourceName=None
      resourceObject=null
      tiling=NoTile
      imageSize=<0 0>
      margin=<ltrb 0 0 0 0>
      outlineSettings {
       bUseBrushTransparency=false
       roundingType=HalfHeightRadius
       width=0
       color {colorUseRule=UseColor_Specified, specifiedColor=<rgba 0 0 0 0>}
       cornerRadii=<0 0 0 0>
      }
      tintColor {colorUseRule=UseColor_Specified, specifiedColor=<rgba 0.0103 0.0103 0.0103 1>}
      uVRegion {bIsValid=false, max=<0 0>, min=<0 0>}
     }
     oddRowBackgroundHoveredBrush {
      bIsDynamicallyLoaded=false
      drawAs=Image
      imageType=NoImage
      mirroring=NoMirror
      resourceName=None
      resourceObject=null
      tiling=NoTile
      imageSize=<0 0>
      margin=<ltrb 0 0 0 0>
      outlineSettings {
       bUseBrushTransparency=false
       roundingType=HalfHeightRadius
       width=0
       color {colorUseRule=UseColor_Specified, specifiedColor=<rgba 0 0 0 0>}
       cornerRadii=<0 0 0 0>
      }
      tintColor {colorUseRule=UseColor_Specified, specifiedColor=<rgba 0.0176 0.0176 0.0176 1>}
      uVRegion {bIsValid=false, max=<0 0>, min=<0 0>}
     }
     parentRowBackgroundBrush {
      bIsDynamicallyLoaded=false
      drawAs=Image
      imageType=NoImage
      mirroring=NoMirror
      resourceName=None
      resourceObject=null
      tiling=NoTile
      imageSize=<32 32>
      margin=<ltrb 0 0 0 0>
      outlineSettings {
       bUseBrushTransparency=false
       roundingType=HalfHeightRadius
       width=0
       color {colorUseRule=UseColor_Specified, specifiedColor=<rgba 0 0 0 0>}
       cornerRadii=<0 0 0 0>
      }
      tintColor {colorUseRule=UseColor_Specified, specifiedColor=<rgba 1 1 1 1>}
      uVRegion {bIsValid=false, max=<0 0>, min=<0 0>}
     }
     parentRowBackgroundHoveredBrush {
      bIsDynamicallyLoaded=false
      drawAs=Image
      imageType=NoImage
      mirroring=NoMirror
      resourceName=None
      resourceObject=null
      tiling=NoTile
      imageSize=<32 32>
      margin=<ltrb 0 0 0 0>
      outlineSettings {
       bUseBrushTransparency=false
       roundingType=HalfHeightRadius
       width=0
       color {colorUseRule=UseColor_Specified, specifiedColor=<rgba 0 0 0 0>}
       cornerRadii=<0 0 0 0>
      }
      tintColor {colorUseRule=UseColor_Specified, specifiedColor=<rgba 1 1 1 1>}
      uVRegion {bIsValid=false, max=<0 0>, min=<0 0>}
     }
     selectedTextColor {colorUseRule=UseColor_Specified, specifiedColor=<rgba 1 1 1 1>}
     selectorFocusedBrush {
      bIsDynamicallyLoaded=false
      drawAs=Border
      imageType=FullColor
      mirroring=NoMirror
      resourceName=../../../Engine/Content/Slate/Common/Selector.png
      resourceObject=null
      tiling=Both
      imageSize=<0 0>
      margin=<ltrb 0.25 0.25 0.25 0.25>
      outlineSettings {
       bUseBrushTransparency=false
       roundingType=HalfHeightRadius
       width=0
       color {colorUseRule=UseColor_Specified, specifiedColor=<rgba 0 0 0 0>}
       cornerRadii=<0 0 0 0>
      }
      tintColor {colorUseRule=UseColor_Specified, specifiedColor=<rgba 0.701 0.225 0.003 1>}
      uVRegion {bIsValid=false, max=<0 0>, min=<0 0>}
     }
     textColor {colorUseRule=UseColor_Specified, specifiedColor=<rgba 0.5271 0.5271 0.5271 1>}
    }
    scrollBarStyle {
     thickness=8
     draggedThumbImage {
      bIsDynamicallyLoaded=false
      drawAs=RoundedBox
      imageType=NoImage
      mirroring=NoMirror
      resourceName=None
      resourceObject=null
      tiling=NoTile
      imageSize=<0 0>
      margin=<ltrb 0 0 0 0>
      outlineSettings {
       bUseBrushTransparency=false
       roundingType=FixedRadius
       width=0
       color {colorUseRule=UseColor_Specified, specifiedColor=<rgba 0 0 0 0>}
       cornerRadii=<4 4 4 4>
      }
      tintColor {colorUseRule=UseColor_Specified, specifiedColor=<rgba 0.2159 0.2159 0.2159 1>}
      uVRegion {bIsValid=false, max=<0 0>, min=<0 0>}
     }
     horizontalBackgroundImage {
      bIsDynamicallyLoaded=false
      drawAs=NoDrawType
      imageType=NoImage
      mirroring=NoMirror
      resourceName=None
      resourceObject=null
      tiling=NoTile
      imageSize=<0 0>
      margin=<ltrb 0 0 0 0>
      outlineSettings {
       bUseBrushTransparency=false
       roundingType=HalfHeightRadius
       width=0
       color {colorUseRule=UseColor_Specified, specifiedColor=<rgba 0 0 0 0>}
       cornerRadii=<0 0 0 0>
      }
      tintColor {colorUseRule=UseColor_Specified, specifiedColor=<rgba 1 1 1 1>}
      uVRegion {bIsValid=false, max=<0 0>, min=<0 0>}
     }
     horizontalBottomSlotImage {
      bIsDynamicallyLoaded=false
      drawAs=NoDrawType
      imageType=NoImage
      mirroring=NoMirror
      resourceName=None
      resourceObject=null
      tiling=NoTile
      imageSize=<0 0>
      margin=<ltrb 0 0 0 0>
      outlineSettings {
       bUseBrushTransparency=false
       roundingType=HalfHeightRadius
       width=0
       color {colorUseRule=UseColor_Specified, specifiedColor=<rgba 0 0 0 0>}
       cornerRadii=<0 0 0 0>
      }
      tintColor {colorUseRule=UseColor_Specified, specifiedColor=<rgba 1 1 1 1>}
      uVRegion {bIsValid=false, max=<0 0>, min=<0 0>}
     }
     horizontalTopSlotImage {
      bIsDynamicallyLoaded=false
      drawAs=NoDrawType
      imageType=NoImage
      mirroring=NoMirror
      resourceName=None
      resourceObject=null
      tiling=NoTile
      imageSize=<0 0>
      margin=<ltrb 0 0 0 0>
      outlineSettings {
       bUseBrushTransparency=false
       roundingType=HalfHeightRadius
       width=0
       color {colorUseRule=UseColor_Specified, specifiedColor=<rgba 0 0 0 0>}
       cornerRadii=<0 0 0 0>
      }
      tintColor {colorUseRule=UseColor_Specified, specifiedColor=<rgba 1 1 1 1>}
      uVRegion {bIsValid=false, max=<0 0>, min=<0 0>}
     }
     hoveredThumbImage {
      bIsDynamicallyLoaded=false
      drawAs=RoundedBox
      imageType=NoImage
      mirroring=NoMirror
      resourceName=None
      resourceObject=null
      tiling=NoTile
      imageSize=<0 0>
      margin=<ltrb 0 0 0 0>
      outlineSettings {
       bUseBrushTransparency=false
       roundingType=FixedRadius
       width=0
       color {colorUseRule=UseColor_Specified, specifiedColor=<rgba 0 0 0 0>}
       cornerRadii=<4 4 4 4>
      }
      tintColor {colorUseRule=UseColor_Specified, specifiedColor=<rgba 0.2159 0.2159 0.2159 1>}
      uVRegion {bIsValid=false, max=<0 0>, min=<0 0>}
     }
     normalThumbImage {
      bIsDynamicallyLoaded=false
      drawAs=RoundedBox
      imageType=NoImage
      mirroring=NoMirror
      resourceName=None
      resourceObject=null
      tiling=NoTile
      imageSize=<0 0>
      margin=<ltrb 0 0 0 0>
      outlineSettings {
       bUseBrushTransparency=false
       roundingType=FixedRadius
       width=0
       color {colorUseRule=UseColor_Specified, specifiedColor=<rgba 0 0 0 0>}
       cornerRadii=<4 4 4 4>
      }
      tintColor {colorUseRule=UseColor_Specified, specifiedColor=<rgba 0.0953 0.0953 0.0953 1>}
      uVRegion {bIsValid=false, max=<0 0>, min=<0 0>}
     }
     verticalBackgroundImage {
      bIsDynamicallyLoaded=false
      drawAs=NoDrawType
      imageType=NoImage
      mirroring=NoMirror
      resourceName=None
      resourceObject=null
      tiling=NoTile
      imageSize=<0 0>
      margin=<ltrb 0 0 0 0>
      outlineSettings {
       bUseBrushTransparency=false
       roundingType=HalfHeightRadius
       width=0
       color {colorUseRule=UseColor_Specified, specifiedColor=<rgba 0 0 0 0>}
       cornerRadii=<0 0 0 0>
      }
      tintColor {colorUseRule=UseColor_Specified, specifiedColor=<rgba 1 1 1 1>}
      uVRegion {bIsValid=false, max=<0 0>, min=<0 0>}
     }
     verticalBottomSlotImage {
      bIsDynamicallyLoaded=false
      drawAs=NoDrawType
      imageType=NoImage
      mirroring=NoMirror
      resourceName=None
      resourceObject=null
      tiling=NoTile
      imageSize=<0 0>
      margin=<ltrb 0 0 0 0>
      outlineSettings {
       bUseBrushTransparency=false
       roundingType=HalfHeightRadius
       width=0
       color {colorUseRule=UseColor_Specified, specifiedColor=<rgba 0 0 0 0>}
       cornerRadii=<0 0 0 0>
      }
      tintColor {colorUseRule=UseColor_Specified, specifiedColor=<rgba 1 1 1 1>}
      uVRegion {bIsValid=false, max=<0 0>, min=<0 0>}
     }
     verticalTopSlotImage {
      bIsDynamicallyLoaded=false
      drawAs=NoDrawType
      imageType=NoImage
      mirroring=NoMirror
      resourceName=None
      resourceObject=null
      tiling=NoTile
      imageSize=<0 0>
      margin=<ltrb 0 0 0 0>
      outlineSettings {
       bUseBrushTransparency=false
       roundingType=HalfHeightRadius
       width=0
       color {colorUseRule=UseColor_Specified, specifiedColor=<rgba 0 0 0 0>}
       cornerRadii=<0 0 0 0>
      }
      tintColor {colorUseRule=UseColor_Specified, specifiedColor=<rgba 1 1 1 1>}
      uVRegion {bIsValid=false, max=<0 0>, min=<0 0>}
     }
    }
    widgetStyle {
     comboButtonStyle {
      downArrowAlign=VAlign_Center
      buttonStyle {
       clickedSlateSound {resourceObject=null}
       disabled {
        bIsDynamicallyLoaded=false
        drawAs=RoundedBox
        imageType=NoImage
        mirroring=NoMirror
        resourceName=None
        resourceObject=null
        tiling=NoTile
        imageSize=<0 0>
        margin=<ltrb 0 0 0 0>
        outlineSettings {
         bUseBrushTransparency=false
         roundingType=FixedRadius
         width=1
         color {colorUseRule=UseColor_ColorTable, specifiedColor=<rgba 0 0 0 0>}
         cornerRadii=<4 4 4 4>
        }
        tintColor {colorUseRule=UseColor_Specified, specifiedColor=<rgba 0.0395 0.0395 0.0395 1>}
        uVRegion {bIsValid=false, max=<0 0>, min=<0 0>}
       }
       disabledForeground {colorUseRule=UseColor_ColorTable, specifiedColor=<rgba 0 0 0 0>}
       hovered {
        bIsDynamicallyLoaded=false
        drawAs=RoundedBox
        imageType=NoImage
        mirroring=NoMirror
        resourceName=None
        resourceObject=null
        tiling=NoTile
        imageSize=<0 0>
        margin=<ltrb 0 0 0 0>
        outlineSettings {
         bUseBrushTransparency=false
         roundingType=FixedRadius
         width=1
         color {colorUseRule=UseColor_ColorTable, specifiedColor=<rgba 0 0 0 0>}
         cornerRadii=<4 4 4 4>
        }
        tintColor {colorUseRule=UseColor_Specified, specifiedColor=<rgba 0.0048 0.0048 0.0048 1>}
        uVRegion {bIsValid=false, max=<0 0>, min=<0 0>}
       }
       hoveredForeground {colorUseRule=UseColor_ColorTable, specifiedColor=<rgba 0 0 0 0>}
       hoveredSlateSound {resourceObject=null}
       normal {
        bIsDynamicallyLoaded=false
        drawAs=RoundedBox
        imageType=NoImage
        mirroring=NoMirror
        resourceName=None
        resourceObject=null
        tiling=NoTile
        imageSize=<0 0>
        margin=<ltrb 0 0 0 0>
        outlineSettings {
         bUseBrushTransparency=false
         roundingType=FixedRadius
         width=1
         color {colorUseRule=UseColor_ColorTable, specifiedColor=<rgba 0 0 0 0>}
         cornerRadii=<4 4 4 4>
        }
        tintColor {colorUseRule=UseColor_Specified, specifiedColor=<rgba 0.0048 0.0048 0.0048 1>}
        uVRegion {bIsValid=false, max=<0 0>, min=<0 0>}
       }
       normalForeground {colorUseRule=UseColor_ColorTable, specifiedColor=<rgba 0 0 0 0>}
       normalPadding=<ltrb 8 1 8 1>
       pressed {
        bIsDynamicallyLoaded=false
        drawAs=RoundedBox
        imageType=NoImage
        mirroring=NoMirror
        resourceName=None
        resourceObject=null
        tiling=NoTile
        imageSize=<0 0>
        margin=<ltrb 0 0 0 0>
        outlineSettings {
         bUseBrushTransparency=false
         roundingType=FixedRadius
         width=1
         color {colorUseRule=UseColor_ColorTable, specifiedColor=<rgba 0 0 0 0>}
         cornerRadii=<4 4 4 4>
        }
        tintColor {colorUseRule=UseColor_Specified, specifiedColor=<rgba 0.0048 0.0048 0.0048 1>}
        uVRegion {bIsValid=false, max=<0 0>, min=<0 0>}
       }
       pressedForeground {colorUseRule=UseColor_ColorTable, specifiedColor=<rgba 0 0 0 0>}
       pressedPadding=<ltrb 8 2 8 0>
       pressedSlateSound {resourceObject=null}
      }
      contentPadding=<ltrb 0 0 0 0>
      downArrowImage {
       bIsDynamicallyLoaded=false
       drawAs=Image
       imageType=Vector
       mirroring=NoMirror
       resourceName=../../../Engine/Content/Slate/Starship/CoreWidgets/ComboBox/wide-chevron-down.svg
       resourceObject=null
       tiling=NoTile
       imageSize=<16 16>
       margin=<ltrb 0 0 0 0>
       outlineSettings {
        bUseBrushTransparency=false
        roundingType=HalfHeightRadius
        width=0
        color {colorUseRule=UseColor_Specified, specifiedColor=<rgba 0 0 0 0>}
        cornerRadii=<0 0 0 0>
       }
       tintColor {colorUseRule=UseColor_Specified, specifiedColor=<rgba 1 1 1 1>}
       uVRegion {bIsValid=false, max=<0 0>, min=<0 0>}
      }
      downArrowPadding=<ltrb 2 2 2 2>
      menuBorderBrush {
       bIsDynamicallyLoaded=false
       drawAs=RoundedBox
       imageType=NoImage
       mirroring=NoMirror
       resourceName=None
       resourceObject=null
       tiling=NoTile
       imageSize=<0 0>
       margin=<ltrb 0 0 0 0>
       outlineSettings {
        bUseBrushTransparency=false
        roundingType=FixedRadius
        width=1
        color {colorUseRule=UseColor_ColorTable, specifiedColor=<rgba 0 0 0 0>}
        cornerRadii=<0 0 0 0>
       }
       tintColor {colorUseRule=UseColor_Specified, specifiedColor=<rgba 0.0395 0.0395 0.0395 1>}
       uVRegion {bIsValid=false, max=<0 0>, min=<0 0>}
      }
      menuBorderPadding=<ltrb 0 0 0 0>
      shadowColorAndOpacity=<rgba 0 0 0 1>
      shadowOffset=<0 0>
     }
     contentPadding=<ltrb 0 0 0 0>
     menuRowPadding=<ltrb 8 3 8 3>
     pressedSlateSound {resourceObject=null}
     selectionChangeSlateSound {resourceObject=null}
    }
   }
   slot {class=HorizontalBoxSlot, properties {horizontalAlignment=HAlign_Left}}
  }
  ScrollBox_0 {
   widgetClassPath=&ScrollBox
   parent=VerticalBox_32
   widget=ScrollBox_0
   properties {
    widgetBarStyle {
     thickness=8
     draggedThumbImage {
      bIsDynamicallyLoaded=false
      drawAs=RoundedBox
      imageType=NoImage
      mirroring=NoMirror
      resourceName=None
      resourceObject=null
      tiling=NoTile
      imageSize=<0 0>
      margin=<ltrb 0 0 0 0>
      outlineSettings {
       bUseBrushTransparency=false
       roundingType=FixedRadius
       width=0
       color {colorUseRule=UseColor_Specified, specifiedColor=<rgba 0 0 0 0>}
       cornerRadii=<4 4 4 4>
      }
      tintColor {colorUseRule=UseColor_Specified, specifiedColor=<rgba 0.2159 0.2159 0.2159 1>}
      uVRegion {bIsValid=false, max=<0 0>, min=<0 0>}
     }
     horizontalBackgroundImage {
      bIsDynamicallyLoaded=false
      drawAs=NoDrawType
      imageType=NoImage
      mirroring=NoMirror
      resourceName=None
      resourceObject=null
      tiling=NoTile
      imageSize=<0 0>
      margin=<ltrb 0 0 0 0>
      outlineSettings {
       bUseBrushTransparency=false
       roundingType=HalfHeightRadius
       width=0
       color {colorUseRule=UseColor_Specified, specifiedColor=<rgba 0 0 0 0>}
       cornerRadii=<0 0 0 0>
      }
      tintColor {colorUseRule=UseColor_Specified, specifiedColor=<rgba 1 1 1 1>}
      uVRegion {bIsValid=false, max=<0 0>, min=<0 0>}
     }
     horizontalBottomSlotImage {
      bIsDynamicallyLoaded=false
      drawAs=NoDrawType
      imageType=NoImage
      mirroring=NoMirror
      resourceName=None
      resourceObject=null
      tiling=NoTile
      imageSize=<0 0>
      margin=<ltrb 0 0 0 0>
      outlineSettings {
       bUseBrushTransparency=false
       roundingType=HalfHeightRadius
       width=0
       color {colorUseRule=UseColor_Specified, specifiedColor=<rgba 0 0 0 0>}
       cornerRadii=<0 0 0 0>
      }
      tintColor {colorUseRule=UseColor_Specified, specifiedColor=<rgba 1 1 1 1>}
      uVRegion {bIsValid=false, max=<0 0>, min=<0 0>}
     }
     horizontalTopSlotImage {
      bIsDynamicallyLoaded=false
      drawAs=NoDrawType
      imageType=NoImage
      mirroring=NoMirror
      resourceName=None
      resourceObject=null
      tiling=NoTile
      imageSize=<0 0>
      margin=<ltrb 0 0 0 0>
      outlineSettings {
       bUseBrushTransparency=false
       roundingType=HalfHeightRadius
       width=0
       color {colorUseRule=UseColor_Specified, specifiedColor=<rgba 0 0 0 0>}
       cornerRadii=<0 0 0 0>
      }
      tintColor {colorUseRule=UseColor_Specified, specifiedColor=<rgba 1 1 1 1>}
      uVRegion {bIsValid=false, max=<0 0>, min=<0 0>}
     }
     hoveredThumbImage {
      bIsDynamicallyLoaded=false
      drawAs=RoundedBox
      imageType=NoImage
      mirroring=NoMirror
      resourceName=None
      resourceObject=null
      tiling=NoTile
      imageSize=<0 0>
      margin=<ltrb 0 0 0 0>
      outlineSettings {
       bUseBrushTransparency=false
       roundingType=FixedRadius
       width=0
       color {colorUseRule=UseColor_Specified, specifiedColor=<rgba 0 0 0 0>}
       cornerRadii=<4 4 4 4>
      }
      tintColor {colorUseRule=UseColor_Specified, specifiedColor=<rgba 0.2159 0.2159 0.2159 1>}
      uVRegion {bIsValid=false, max=<0 0>, min=<0 0>}
     }
     normalThumbImage {
      bIsDynamicallyLoaded=false
      drawAs=RoundedBox
      imageType=NoImage
      mirroring=NoMirror
      resourceName=None
      resourceObject=null
      tiling=NoTile
      imageSize=<0 0>
      margin=<ltrb 0 0 0 0>
      outlineSettings {
       bUseBrushTransparency=false
       roundingType=FixedRadius
       width=0
       color {colorUseRule=UseColor_Specified, specifiedColor=<rgba 0 0 0 0>}
       cornerRadii=<4 4 4 4>
      }
      tintColor {colorUseRule=UseColor_Specified, specifiedColor=<rgba 0.0953 0.0953 0.0953 1>}
      uVRegion {bIsValid=false, max=<0 0>, min=<0 0>}
     }
     verticalBackgroundImage {
      bIsDynamicallyLoaded=false
      drawAs=NoDrawType
      imageType=NoImage
      mirroring=NoMirror
      resourceName=None
      resourceObject=null
      tiling=NoTile
      imageSize=<0 0>
      margin=<ltrb 0 0 0 0>
      outlineSettings {
       bUseBrushTransparency=false
       roundingType=HalfHeightRadius
       width=0
       color {colorUseRule=UseColor_Specified, specifiedColor=<rgba 0 0 0 0>}
       cornerRadii=<0 0 0 0>
      }
      tintColor {colorUseRule=UseColor_Specified, specifiedColor=<rgba 1 1 1 1>}
      uVRegion {bIsValid=false, max=<0 0>, min=<0 0>}
     }
     verticalBottomSlotImage {
      bIsDynamicallyLoaded=false
      drawAs=NoDrawType
      imageType=NoImage
      mirroring=NoMirror
      resourceName=None
      resourceObject=null
      tiling=NoTile
      imageSize=<0 0>
      margin=<ltrb 0 0 0 0>
      outlineSettings {
       bUseBrushTransparency=false
       roundingType=HalfHeightRadius
       width=0
       color {colorUseRule=UseColor_Specified, specifiedColor=<rgba 0 0 0 0>}
       cornerRadii=<0 0 0 0>
      }
      tintColor {colorUseRule=UseColor_Specified, specifiedColor=<rgba 1 1 1 1>}
      uVRegion {bIsValid=false, max=<0 0>, min=<0 0>}
     }
     verticalTopSlotImage {
      bIsDynamicallyLoaded=false
      drawAs=NoDrawType
      imageType=NoImage
      mirroring=NoMirror
      resourceName=None
      resourceObject=null
      tiling=NoTile
      imageSize=<0 0>
      margin=<ltrb 0 0 0 0>
      outlineSettings {
       bUseBrushTransparency=false
       roundingType=HalfHeightRadius
       width=0
       color {colorUseRule=UseColor_Specified, specifiedColor=<rgba 0 0 0 0>}
       cornerRadii=<0 0 0 0>
      }
      tintColor {colorUseRule=UseColor_Specified, specifiedColor=<rgba 1 1 1 1>}
      uVRegion {bIsValid=false, max=<0 0>, min=<0 0>}
     }
    }
    widgetStyle {
     barThickness=8
     bottomShadowBrush {
      bIsDynamicallyLoaded=false
      drawAs=Box
      imageType=FullColor
      mirroring=NoMirror
      resourceName=../../../Engine/Content/Slate/Common/ScrollBoxShadowBottom.png
      resourceObject=null
      tiling=NoTile
      imageSize=<16 8>
      margin=<ltrb 0.5 0 0.5 1>
      outlineSettings {
       bUseBrushTransparency=false
       roundingType=HalfHeightRadius
       width=0
       color {colorUseRule=UseColor_Specified, specifiedColor=<rgba 0 0 0 0>}
       cornerRadii=<0 0 0 0>
      }
      tintColor {colorUseRule=UseColor_Specified, specifiedColor=<rgba 1 1 1 1>}
      uVRegion {bIsValid=false, max=<0 0>, min=<0 0>}
     }
     horizontalScrolledContentPadding=<ltrb 0 0 1 0>
     leftShadowBrush {
      bIsDynamicallyLoaded=false
      drawAs=Box
      imageType=FullColor
      mirroring=NoMirror
      resourceName=../../../Engine/Content/Slate/Common/ScrollBoxShadowLeft.png
      resourceObject=null
      tiling=NoTile
      imageSize=<8 16>
      margin=<ltrb 1 0.5 0 0.5>
      outlineSettings {
       bUseBrushTransparency=false
       roundingType=HalfHeightRadius
       width=0
       color {colorUseRule=UseColor_Specified, specifiedColor=<rgba 0 0 0 0>}
       cornerRadii=<0 0 0 0>
      }
      tintColor {colorUseRule=UseColor_Specified, specifiedColor=<rgba 1 1 1 1>}
      uVRegion {bIsValid=false, max=<0 0>, min=<0 0>}
     }
     rightShadowBrush {
      bIsDynamicallyLoaded=false
      drawAs=Box
      imageType=FullColor
      mirroring=NoMirror
      resourceName=../../../Engine/Content/Slate/Common/ScrollBoxShadowRight.png
      resourceObject=null
      tiling=NoTile
      imageSize=<8 16>
      margin=<ltrb 0 0.5 1 0.5>
      outlineSettings {
       bUseBrushTransparency=false
       roundingType=HalfHeightRadius
       width=0
       color {colorUseRule=UseColor_Specified, specifiedColor=<rgba 0 0 0 0>}
       cornerRadii=<0 0 0 0>
      }
      tintColor {colorUseRule=UseColor_Specified, specifiedColor=<rgba 1 1 1 1>}
      uVRegion {bIsValid=false, max=<0 0>, min=<0 0>}
     }
     topShadowBrush {
      bIsDynamicallyLoaded=false
      drawAs=Box
      imageType=FullColor
      mirroring=NoMirror
      resourceName=../../../Engine/Content/Slate/Common/ScrollBoxShadowTop.png
      resourceObject=null
      tiling=NoTile
      imageSize=<16 8>
      margin=<ltrb 0.5 1 0.5 0>
      outlineSettings {
       bUseBrushTransparency=false
       roundingType=HalfHeightRadius
       width=0
       color {colorUseRule=UseColor_Specified, specifiedColor=<rgba 0 0 0 0>}
       cornerRadii=<0 0 0 0>
      }
      tintColor {colorUseRule=UseColor_Specified, specifiedColor=<rgba 1 1 1 1>}
      uVRegion {bIsValid=false, max=<0 0>, min=<0 0>}
     }
     verticalScrolledContentPadding=<ltrb 0 0 0 1>
    }
   }
   slot {class=VerticalBoxSlot, properties {size {sizeRule=Fill, value=1}}}
  }
  DetailPanel {
   widgetClassPath=/Script/CommonUI.CommonAnimatedSwitcher
   bIsVariable=true
   parent=ScrollBox_0
   widget=DetailPanel
   slot {class=ScrollBoxSlot}
  }
  VerticalBox_1 {
   widgetClassPath=&VerticalBox
   parent=DetailPanel
   widget=VerticalBox_1
   slot {class=WidgetSwitcherSlot, properties {padding=<ltrb 0 15 0 15>}}
  }
  Border_3 {
   widgetClassPath=&Border
   parent=VerticalBox_1
   widget=Border_3
   properties {brushColor=<rgba 1 1 1 0.3>}
   slot {class=VerticalBoxSlot}
  }
  VerticalBox_3 {widgetClassPath=&VerticalBox, parent=Border_3, widget=VerticalBox_3, slot {class=BorderSlot}}
  CommonTextBlock_8 {
   widgetClassPath=&CommonTextBlock
   parent=VerticalBox_3
   widget=CommonTextBlock_8
   properties {
    style=/Game/UI/Foundation/Items/Parts/TextStyles/ItemCardPowerRatingTextStyle_XS.ItemCardPowerRatingTextStyle_XS_C
    text="↓↓↓↓↓Control by texture artist↓↓↓↓↓"
    font {
     bForceMonospaced=false
     bMaterialIsStencil=false
     fontMaterial=null
     fontObject=&NotoSans
     letterSpacing=0
     monospacedWidth=1
     size=15
     skewAmount=0
     typefaceFontName=NotoSans-Bold
     outlineSettings {
      bApplyOutlineToDropShadows=false
      bMiteredCorners=false
      bSeparateFillAlpha=false
      outlineMaterial=null
      outlineSize=0
      outlineColor=<rgba 0 0 0 1>
     }
    }
    shadowOffset=<0 0>
   }
   slot {class=VerticalBoxSlot, properties {horizontalAlignment=HAlign_Left}}
  }
  3D_Mesh {
   widgetClassPath=&DetailsView
   parent=VerticalBox_3
   widget=3D_Mesh
   properties {categoriesToShow [From 3D Mesh]}
   slot {class=VerticalBoxSlot}
  }
  3D_ColorTweak {
   widgetClassPath=&DetailsView
   parent=VerticalBox_3
   widget=3D_ColorTweak
   properties {categoriesToShow [From 3D Color Tweak]}
   slot {class=VerticalBoxSlot}
  }
  3DDAH {widgetClassPath=&HorizontalBox, bIsVariable=true, parent=VerticalBox_1, widget=3DDAH, slot {class=VerticalBoxSlot}}
  3DDA {
   widgetClassPath=&SinglePropertyView
   parent=3DDAH
   widget=3DDA
   properties {propertyName=AchievedDiffuseTexture}
   slot {class=HorizontalBoxSlot, properties {size {sizeRule=Fill, value=1}}}
  }
  3DDAB {
   widgetClassPath=&Button
   parent=3DDAH
   widget=3DDAB
   properties {
    backgroundColor=<rgba 0.1 0.1 0.1 1>
    widgetStyle {
     clickedSlateSound {resourceObject=null}
     disabled {
      bIsDynamicallyLoaded=false
      drawAs=RoundedBox
      imageType=NoImage
      mirroring=NoMirror
      resourceName=None
      resourceObject=null
      tiling=NoTile
      imageSize=<0 0>
      margin=<ltrb 0 0 0 0>
      outlineSettings {
       bUseBrushTransparency=false
       roundingType=FixedRadius
       width=1
       color {colorUseRule=UseColor_ColorTable, specifiedColor=<rgba 0 0 0 0>}
       cornerRadii=<4 4 4 4>
      }
      tintColor {colorUseRule=UseColor_Specified, specifiedColor=<rgba 0.0395 0.0395 0.0395 1>}
      uVRegion {bIsValid=false, max=<0 0>, min=<0 0>}
     }
     disabledForeground {colorUseRule=UseColor_ColorTable, specifiedColor=<rgba 0 0 0 0>}
     hovered {
      bIsDynamicallyLoaded=false
      drawAs=RoundedBox
      imageType=NoImage
      mirroring=NoMirror
      resourceName=None
      resourceObject=null
      tiling=NoTile
      imageSize=<0 0>
      margin=<ltrb 0 0 0 0>
      outlineSettings {
       bUseBrushTransparency=false
       roundingType=FixedRadius
       width=1
       color {colorUseRule=UseColor_ColorTable, specifiedColor=<rgba 0 0 0 0>}
       cornerRadii=<4 4 4 4>
      }
      tintColor {colorUseRule=UseColor_Specified, specifiedColor=<rgba 0.0953 0.0953 0.0953 1>}
      uVRegion {bIsValid=false, max=<0 0>, min=<0 0>}
     }
     hoveredForeground {colorUseRule=UseColor_ColorTable, specifiedColor=<rgba 0 0 0 0>}
     hoveredSlateSound {resourceObject=null}
     normal {
      bIsDynamicallyLoaded=false
      drawAs=RoundedBox
      imageType=NoImage
      mirroring=NoMirror
      resourceName=None
      resourceObject=null
      tiling=NoTile
      imageSize=<0 0>
      margin=<ltrb 0 0 0 0>
      outlineSettings {
       bUseBrushTransparency=false
       roundingType=FixedRadius
       width=1
       color {colorUseRule=UseColor_ColorTable, specifiedColor=<rgba 0 0 0 0>}
       cornerRadii=<4 4 4 4>
      }
      tintColor {colorUseRule=UseColor_Specified, specifiedColor=<rgba 0.0395 0.0395 0.0395 1>}
      uVRegion {bIsValid=false, max=<0 0>, min=<0 0>}
     }
     normalForeground {colorUseRule=UseColor_ColorTable, specifiedColor=<rgba 0 0 0 0>}
     normalPadding=<ltrb 12 1.5 12 1.5>
     pressed {
      bIsDynamicallyLoaded=false
      drawAs=RoundedBox
      imageType=NoImage
      mirroring=NoMirror
      resourceName=None
      resourceObject=null
      tiling=NoTile
      imageSize=<0 0>
      margin=<ltrb 0 0 0 0>
      outlineSettings {
       bUseBrushTransparency=false
       roundingType=FixedRadius
       width=1
       color {colorUseRule=UseColor_ColorTable, specifiedColor=<rgba 0 0 0 0>}
       cornerRadii=<4 4 4 4>
      }
      tintColor {colorUseRule=UseColor_Specified, specifiedColor=<rgba 0.0284 0.0284 0.0284 1>}
      uVRegion {bIsValid=false, max=<0 0>, min=<0 0>}
     }
     pressedForeground {colorUseRule=UseColor_ColorTable, specifiedColor=<rgba 0 0 0 0>}
     pressedPadding=<ltrb 12 2 12 0>
     pressedSlateSound {resourceObject=null}
    }
   }
   slot {class=HorizontalBoxSlot, properties {size {sizeRule=Fill, value=0.5}}}
  }
  CommonTextBlock_6 {
   widgetClassPath=&CommonTextBlock
   parent=3DDAB
   widget=CommonTextBlock_6
   properties {
    style=/Game/UI/Foundation/Items/Parts/TextStyles/ItemCardPowerRatingTextStyle_S.ItemCardPowerRatingTextStyle_S_C
    text=Achieve Diffuse Texture
    font {
     bForceMonospaced=false
     bMaterialIsStencil=false
     fontMaterial=null
     fontObject=&NotoSans
     letterSpacing=0
     monospacedWidth=1
     size=18
     skewAmount=0
     typefaceFontName=NotoSans-Bold
     outlineSettings {
      bApplyOutlineToDropShadows=false
      bMiteredCorners=false
      bSeparateFillAlpha=false
      outlineMaterial=null
      outlineSize=0
      outlineColor=<rgba 0 0 0 1>
     }
    }
    shadowOffset=<0 0>
   }
   slot {class=ButtonSlot, properties {verticalAlignment=VAlign_Fill, padding=<ltrb 0 0 0 0>}}
  }
  Border_4 {
   widgetClassPath=&Border
   parent=VerticalBox_1
   widget=Border_4
   properties {brushColor=<rgba 1 1 1 0.3>}
   slot {class=VerticalBoxSlot, properties {padding=<ltrb 0 50 0 0>}}
  }
  VerticalBox_5 {widgetClassPath=&VerticalBox, parent=Border_4, widget=VerticalBox_5, slot {class=BorderSlot}}
  CommonTextBlock_9 {
   widgetClassPath=&CommonTextBlock
   parent=VerticalBox_5
   widget=CommonTextBlock_9
   properties {
    text="↓↓↓↓↓Control by FX artist↓↓↓↓↓"
    font {
     bForceMonospaced=false
     bMaterialIsStencil=false
     fontMaterial=null
     fontObject=&NotoSans
     letterSpacing=0
     monospacedWidth=1
     size=15
     skewAmount=0
     typefaceFontName=NotoSans-Bold
     outlineSettings {
      bApplyOutlineToDropShadows=false
      bMiteredCorners=false
      bSeparateFillAlpha=false
      outlineMaterial=null
      outlineSize=0
      outlineColor=<rgba 0 0 0 1>
     }
    }
    shadowOffset=<0 0>
   }
   slot {class=VerticalBoxSlot}
  }
  3D {
   widgetClassPath=&DetailsView
   parent=VerticalBox_5
   widget=3D
   properties {categoriesToShow [From3D]}
   slot {class=VerticalBoxSlot}
  }
  VerticalBox_54 {
   widgetClassPath=&VerticalBox
   parent=VerticalBox_1
   widget=VerticalBox_54
   slot {class=VerticalBoxSlot, properties {padding=<ltrb 0 15 0 0>}}
  }
  3DWPOAH {widgetClassPath=&HorizontalBox, bIsVariable=true, parent=VerticalBox_54, widget=3DWPOAH, slot {class=VerticalBoxSlot}}
  3DWPOA {
   widgetClassPath=&SinglePropertyView
   parent=3DWPOAH
   widget=3DWPOA
   properties {propertyName=AchievedWPOTexture}
   slot {class=HorizontalBoxSlot, properties {size {sizeRule=Fill, value=1}}}
  }
  3DWPOAB {
   widgetClassPath=&Button
   parent=3DWPOAH
   widget=3DWPOAB
   properties {
    backgroundColor=<rgba 0.1 0.1 0.1 1>
    widgetStyle {
     clickedSlateSound {resourceObject=null}
     disabled {
      bIsDynamicallyLoaded=false
      drawAs=RoundedBox
      imageType=NoImage
      mirroring=NoMirror
      resourceName=None
      resourceObject=null
      tiling=NoTile
      imageSize=<0 0>
      margin=<ltrb 0 0 0 0>
      outlineSettings {
       bUseBrushTransparency=false
       roundingType=FixedRadius
       width=1
       color {colorUseRule=UseColor_ColorTable, specifiedColor=<rgba 0 0 0 0>}
       cornerRadii=<4 4 4 4>
      }
      tintColor {colorUseRule=UseColor_Specified, specifiedColor=<rgba 0.0395 0.0395 0.0395 1>}
      uVRegion {bIsValid=false, max=<0 0>, min=<0 0>}
     }
     disabledForeground {colorUseRule=UseColor_ColorTable, specifiedColor=<rgba 0 0 0 0>}
     hovered {
      bIsDynamicallyLoaded=false
      drawAs=RoundedBox
      imageType=NoImage
      mirroring=NoMirror
      resourceName=None
      resourceObject=null
      tiling=NoTile
      imageSize=<0 0>
      margin=<ltrb 0 0 0 0>
      outlineSettings {
       bUseBrushTransparency=false
       roundingType=FixedRadius
       width=1
       color {colorUseRule=UseColor_ColorTable, specifiedColor=<rgba 0 0 0 0>}
       cornerRadii=<4 4 4 4>
      }
      tintColor {colorUseRule=UseColor_Specified, specifiedColor=<rgba 0.0953 0.0953 0.0953 1>}
      uVRegion {bIsValid=false, max=<0 0>, min=<0 0>}
     }
     hoveredForeground {colorUseRule=UseColor_ColorTable, specifiedColor=<rgba 0 0 0 0>}
     hoveredSlateSound {resourceObject=null}
     normal {
      bIsDynamicallyLoaded=false
      drawAs=RoundedBox
      imageType=NoImage
      mirroring=NoMirror
      resourceName=None
      resourceObject=null
      tiling=NoTile
      imageSize=<0 0>
      margin=<ltrb 0 0 0 0>
      outlineSettings {
       bUseBrushTransparency=false
       roundingType=FixedRadius
       width=1
       color {colorUseRule=UseColor_ColorTable, specifiedColor=<rgba 0 0 0 0>}
       cornerRadii=<4 4 4 4>
      }
      tintColor {colorUseRule=UseColor_Specified, specifiedColor=<rgba 0.0395 0.0395 0.0395 1>}
      uVRegion {bIsValid=false, max=<0 0>, min=<0 0>}
     }
     normalForeground {colorUseRule=UseColor_ColorTable, specifiedColor=<rgba 0 0 0 0>}
     normalPadding=<ltrb 12 1.5 12 1.5>
     pressed {
      bIsDynamicallyLoaded=false
      drawAs=RoundedBox
      imageType=NoImage
      mirroring=NoMirror
      resourceName=None
      resourceObject=null
      tiling=NoTile
      imageSize=<0 0>
      margin=<ltrb 0 0 0 0>
      outlineSettings {
       bUseBrushTransparency=false
       roundingType=FixedRadius
       width=1
       color {colorUseRule=UseColor_ColorTable, specifiedColor=<rgba 0 0 0 0>}
       cornerRadii=<4 4 4 4>
      }
      tintColor {colorUseRule=UseColor_Specified, specifiedColor=<rgba 0.0284 0.0284 0.0284 1>}
      uVRegion {bIsValid=false, max=<0 0>, min=<0 0>}
     }
     pressedForeground {colorUseRule=UseColor_ColorTable, specifiedColor=<rgba 0 0 0 0>}
     pressedPadding=<ltrb 12 2 12 0>
     pressedSlateSound {resourceObject=null}
    }
   }
   slot {class=HorizontalBoxSlot, properties {size {sizeRule=Fill, value=0.5}}}
  }
  CommonTextBlock {
   widgetClassPath=&CommonTextBlock
   parent=3DWPOAB
   widget=CommonTextBlock
   properties {
    style=&ItemCardPowerRatingTextStyle_S_C
    text=Achieve WPO Texture
    font {
     bForceMonospaced=false
     bMaterialIsStencil=false
     fontMaterial=null
     fontObject=&NotoSans
     letterSpacing=0
     monospacedWidth=1
     size=18
     skewAmount=0
     typefaceFontName=NotoSans-Bold
     outlineSettings {
      bApplyOutlineToDropShadows=false
      bMiteredCorners=false
      bSeparateFillAlpha=false
      outlineMaterial=null
      outlineSize=0
      outlineColor=<rgba 0 0 0 1>
     }
    }
    shadowOffset=<0 0>
   }
   slot {class=ButtonSlot, properties {verticalAlignment=VAlign_Fill, padding=<ltrb 0 0 0 0>}}
  }
  3DNAH {widgetClassPath=&HorizontalBox, bIsVariable=true, parent=VerticalBox_54, widget=3DNAH, slot {class=VerticalBoxSlot}}
  3DNA {
   widgetClassPath=&SinglePropertyView
   parent=3DNAH
   widget=3DNA
   properties {propertyName=AchievedNormalTexture}
   slot {class=HorizontalBoxSlot, properties {size {sizeRule=Fill, value=1}}}
  }
  3DNAB {
   widgetClassPath=&Button
   parent=3DNAH
   widget=3DNAB
   properties {
    backgroundColor=<rgba 0.1 0.1 0.1 1>
    widgetStyle {
     clickedSlateSound {resourceObject=null}
     disabled {
      bIsDynamicallyLoaded=false
      drawAs=RoundedBox
      imageType=NoImage
      mirroring=NoMirror
      resourceName=None
      resourceObject=null
      tiling=NoTile
      imageSize=<0 0>
      margin=<ltrb 0 0 0 0>
      outlineSettings {
       bUseBrushTransparency=false
       roundingType=FixedRadius
       width=1
       color {colorUseRule=UseColor_ColorTable, specifiedColor=<rgba 0 0 0 0>}
       cornerRadii=<4 4 4 4>
      }
      tintColor {colorUseRule=UseColor_Specified, specifiedColor=<rgba 0.0395 0.0395 0.0395 1>}
      uVRegion {bIsValid=false, max=<0 0>, min=<0 0>}
     }
     disabledForeground {colorUseRule=UseColor_ColorTable, specifiedColor=<rgba 0 0 0 0>}
     hovered {
      bIsDynamicallyLoaded=false
      drawAs=RoundedBox
      imageType=NoImage
      mirroring=NoMirror
      resourceName=None
      resourceObject=null
      tiling=NoTile
      imageSize=<0 0>
      margin=<ltrb 0 0 0 0>
      outlineSettings {
       bUseBrushTransparency=false
       roundingType=FixedRadius
       width=1
       color {colorUseRule=UseColor_ColorTable, specifiedColor=<rgba 0 0 0 0>}
       cornerRadii=<4 4 4 4>
      }
      tintColor {colorUseRule=UseColor_Specified, specifiedColor=<rgba 0.0953 0.0953 0.0953 1>}
      uVRegion {bIsValid=false, max=<0 0>, min=<0 0>}
     }
     hoveredForeground {colorUseRule=UseColor_ColorTable, specifiedColor=<rgba 0 0 0 0>}
     hoveredSlateSound {resourceObject=null}
     normal {
      bIsDynamicallyLoaded=false
      drawAs=RoundedBox
      imageType=NoImage
      mirroring=NoMirror
      resourceName=None
      resourceObject=null
      tiling=NoTile
      imageSize=<0 0>
      margin=<ltrb 0 0 0 0>
      outlineSettings {
       bUseBrushTransparency=false
       roundingType=FixedRadius
       width=1
       color {colorUseRule=UseColor_ColorTable, specifiedColor=<rgba 0 0 0 0>}
       cornerRadii=<4 4 4 4>
      }
      tintColor {colorUseRule=UseColor_Specified, specifiedColor=<rgba 0.0395 0.0395 0.0395 1>}
      uVRegion {bIsValid=false, max=<0 0>, min=<0 0>}
     }
     normalForeground {colorUseRule=UseColor_ColorTable, specifiedColor=<rgba 0 0 0 0>}
     normalPadding=<ltrb 12 1.5 12 1.5>
     pressed {
      bIsDynamicallyLoaded=false
      drawAs=RoundedBox
      imageType=NoImage
      mirroring=NoMirror
      resourceName=None
      resourceObject=null
      tiling=NoTile
      imageSize=<0 0>
      margin=<ltrb 0 0 0 0>
      outlineSettings {
       bUseBrushTransparency=false
       roundingType=FixedRadius
       width=1
       color {colorUseRule=UseColor_ColorTable, specifiedColor=<rgba 0 0 0 0>}
       cornerRadii=<4 4 4 4>
      }
      tintColor {colorUseRule=UseColor_Specified, specifiedColor=<rgba 0.0284 0.0284 0.0284 1>}
      uVRegion {bIsValid=false, max=<0 0>, min=<0 0>}
     }
     pressedForeground {colorUseRule=UseColor_ColorTable, specifiedColor=<rgba 0 0 0 0>}
     pressedPadding=<ltrb 12 2 12 0>
     pressedSlateSound {resourceObject=null}
    }
   }
   slot {class=HorizontalBoxSlot, properties {size {sizeRule=Fill, value=0.5}}}
  }
  CommonTextBlock_4 {
   widgetClassPath=&CommonTextBlock
   parent=3DNAB
   widget=CommonTextBlock_4
   properties {
    style=&ItemCardPowerRatingTextStyle_S_C
    text=Achieve Normal Texture
    font {
     bForceMonospaced=false
     bMaterialIsStencil=false
     fontMaterial=null
     fontObject=&NotoSans
     letterSpacing=0
     monospacedWidth=1
     size=18
     skewAmount=0
     typefaceFontName=NotoSans-Bold
     outlineSettings {
      bApplyOutlineToDropShadows=false
      bMiteredCorners=false
      bSeparateFillAlpha=false
      outlineMaterial=null
      outlineSize=0
      outlineColor=<rgba 0 0 0 1>
     }
    }
    shadowOffset=<0 0>
   }
   slot {class=ButtonSlot, properties {verticalAlignment=VAlign_Fill, padding=<ltrb 0 0 0 0>}}
  }
  CommonTextBlock_7 {
   widgetClassPath=&CommonTextBlock
   parent=VerticalBox_1
   widget=CommonTextBlock_7
   properties {
    style=&ItemCardPowerRatingTextStyle_XS_C
    text="You can pause the capturer here>>\r\nThis is useful for snapping models with dynamic materials"
    font {
     bForceMonospaced=false
     bMaterialIsStencil=false
     fontMaterial=null
     fontObject=&NotoSans
     letterSpacing=0
     monospacedWidth=1
     size=15
     skewAmount=0
     typefaceFontName=NotoSans-Bold
     outlineSettings {
      bApplyOutlineToDropShadows=false
      bMiteredCorners=false
      bSeparateFillAlpha=false
      outlineMaterial=null
      outlineSize=0
      outlineColor=<rgba 0 0 0 1>
     }
    }
    shadowOffset=<0 0>
   }
   slot {class=VerticalBoxSlot, properties {horizontalAlignment=HAlign_Center}}
  }
  Button_187 {
   widgetClassPath=&Button
   parent=VerticalBox_1
   widget=Button_187
   properties {
    backgroundColor=<rgba 0.1 0.1 0.1 1>
    widgetStyle {
     clickedSlateSound {resourceObject=null}
     disabled {
      bIsDynamicallyLoaded=false
      drawAs=RoundedBox
      imageType=NoImage
      mirroring=NoMirror
      resourceName=None
      resourceObject=null
      tiling=NoTile
      imageSize=<0 0>
      margin=<ltrb 0 0 0 0>
      outlineSettings {
       bUseBrushTransparency=false
       roundingType=FixedRadius
       width=1
       color {colorUseRule=UseColor_ColorTable, specifiedColor=<rgba 0 0 0 0>}
       cornerRadii=<4 4 4 4>
      }
      tintColor {colorUseRule=UseColor_Specified, specifiedColor=<rgba 0.0395 0.0395 0.0395 1>}
      uVRegion {bIsValid=false, max=<0 0>, min=<0 0>}
     }
     disabledForeground {colorUseRule=UseColor_ColorTable, specifiedColor=<rgba 0 0 0 0>}
     hovered {
      bIsDynamicallyLoaded=false
      drawAs=RoundedBox
      imageType=NoImage
      mirroring=NoMirror
      resourceName=None
      resourceObject=null
      tiling=NoTile
      imageSize=<0 0>
      margin=<ltrb 0 0 0 0>
      outlineSettings {
       bUseBrushTransparency=false
       roundingType=FixedRadius
       width=1
       color {colorUseRule=UseColor_ColorTable, specifiedColor=<rgba 0 0 0 0>}
       cornerRadii=<4 4 4 4>
      }
      tintColor {colorUseRule=UseColor_Specified, specifiedColor=<rgba 0.0953 0.0953 0.0953 1>}
      uVRegion {bIsValid=false, max=<0 0>, min=<0 0>}
     }
     hoveredForeground {colorUseRule=UseColor_ColorTable, specifiedColor=<rgba 0 0 0 0>}
     hoveredSlateSound {resourceObject=null}
     normal {
      bIsDynamicallyLoaded=false
      drawAs=RoundedBox
      imageType=NoImage
      mirroring=NoMirror
      resourceName=None
      resourceObject=null
      tiling=NoTile
      imageSize=<0 0>
      margin=<ltrb 0 0 0 0>
      outlineSettings {
       bUseBrushTransparency=false
       roundingType=FixedRadius
       width=1
       color {colorUseRule=UseColor_ColorTable, specifiedColor=<rgba 0 0 0 0>}
       cornerRadii=<4 4 4 4>
      }
      tintColor {colorUseRule=UseColor_Specified, specifiedColor=<rgba 0.0395 0.0395 0.0395 1>}
      uVRegion {bIsValid=false, max=<0 0>, min=<0 0>}
     }
     normalForeground {colorUseRule=UseColor_ColorTable, specifiedColor=<rgba 0 0 0 0>}
     normalPadding=<ltrb 12 1.5 12 1.5>
     pressed {
      bIsDynamicallyLoaded=false
      drawAs=RoundedBox
      imageType=NoImage
      mirroring=NoMirror
      resourceName=None
      resourceObject=null
      tiling=NoTile
      imageSize=<0 0>
      margin=<ltrb 0 0 0 0>
      outlineSettings {
       bUseBrushTransparency=false
       roundingType=FixedRadius
       width=1
       color {colorUseRule=UseColor_ColorTable, specifiedColor=<rgba 0 0 0 0>}
       cornerRadii=<4 4 4 4>
      }
      tintColor {colorUseRule=UseColor_Specified, specifiedColor=<rgba 0.0284 0.0284 0.0284 1>}
      uVRegion {bIsValid=false, max=<0 0>, min=<0 0>}
     }
     pressedForeground {colorUseRule=UseColor_ColorTable, specifiedColor=<rgba 0 0 0 0>}
     pressedPadding=<ltrb 12 2 12 0>
     pressedSlateSound {resourceObject=null}
    }
   }
   slot {class=VerticalBoxSlot}
  }
  ToggleText {
   widgetClassPath=&CommonTextBlock
   bIsVariable=true
   parent=Button_187
   widget=ToggleText
   properties {
    style=&ItemCardPowerRatingTextStyle_S_C
    text=ToggleCaptureActivation
    font {
     bForceMonospaced=false
     bMaterialIsStencil=false
     fontMaterial=null
     fontObject=&NotoSans
     letterSpacing=0
     monospacedWidth=1
     size=18
     skewAmount=0
     typefaceFontName=NotoSans-Bold
     outlineSettings {
      bApplyOutlineToDropShadows=false
      bMiteredCorners=false
      bSeparateFillAlpha=false
      outlineMaterial=null
      outlineSize=0
      outlineColor=<rgba 0 0 0 1>
     }
    }
    shadowOffset=<0 0>
   }
   slot {class=ButtonSlot, properties {verticalAlignment=VAlign_Fill, padding=<ltrb 0 0 0 0>}}
  }
  VerticalBox_4 {
   widgetClassPath=&VerticalBox
   parent=DetailPanel
   widget=VerticalBox_4
   slot {class=WidgetSwitcherSlot, properties {padding=<ltrb 0 15 0 15>}}
  }
  Border_5 {
   widgetClassPath=&Border
   parent=VerticalBox_4
   widget=Border_5
   properties {brushColor=<rgba 1 1 1 0.3>}
   slot {class=VerticalBoxSlot}
  }
  VerticalBox_6 {widgetClassPath=&VerticalBox, parent=Border_5, widget=VerticalBox_6, slot {class=BorderSlot}}
  CommonTextBlock_11 {
   widgetClassPath=&CommonTextBlock
   parent=VerticalBox_6
   widget=CommonTextBlock_11
   properties {
    style=&ItemCardPowerRatingTextStyle_XS_C
    text="↓↓↓↓↓Control by texture artist↓↓↓↓↓"
    font {
     bForceMonospaced=false
     bMaterialIsStencil=false
     fontMaterial=null
     fontObject=&NotoSans
     letterSpacing=0
     monospacedWidth=1
     size=15
     skewAmount=0
     typefaceFontName=NotoSans-Bold
     outlineSettings {
      bApplyOutlineToDropShadows=false
      bMiteredCorners=false
      bSeparateFillAlpha=false
      outlineMaterial=null
      outlineSize=0
      outlineColor=<rgba 0 0 0 1>
     }
    }
    shadowOffset=<0 0>
   }
   slot {class=VerticalBoxSlot, properties {horizontalAlignment=HAlign_Left}}
  }
  2D_TextureArtist {
   widgetClassPath=&DetailsView
   parent=VerticalBox_6
   widget=2D_TextureArtist
   properties {categoriesToShow [From 2D Source Texture]}
   slot {class=VerticalBoxSlot}
  }
  Border_6 {
   widgetClassPath=&Border
   parent=VerticalBox_4
   widget=Border_6
   properties {brushColor=<rgba 1 1 1 0.3>}
   slot {class=VerticalBoxSlot, properties {padding=<ltrb 0 50 0 0>}}
  }
  VerticalBox_7 {widgetClassPath=&VerticalBox, parent=Border_6, widget=VerticalBox_7, slot {class=BorderSlot}}
  CommonTextBlock_10 {
   widgetClassPath=&CommonTextBlock
   parent=VerticalBox_7
   widget=CommonTextBlock_10
   properties {
    text="↓↓↓↓↓Control by FX artist↓↓↓↓↓"
    font {
     bForceMonospaced=false
     bMaterialIsStencil=false
     fontMaterial=null
     fontObject=&NotoSans
     letterSpacing=0
     monospacedWidth=1
     size=15
     skewAmount=0
     typefaceFontName=NotoSans-Bold
     outlineSettings {
      bApplyOutlineToDropShadows=false
      bMiteredCorners=false
      bSeparateFillAlpha=false
      outlineMaterial=null
      outlineSize=0
      outlineColor=<rgba 0 0 0 1>
     }
    }
    shadowOffset=<0 0>
   }
   slot {class=VerticalBoxSlot, properties {horizontalAlignment=HAlign_Left}}
  }
  2D {
   widgetClassPath=&DetailsView
   parent=VerticalBox_7
   widget=2D
   properties {categoriesToShow [From2D]}
   slot {class=VerticalBoxSlot}
  }
  VerticalBox {
   widgetClassPath=&VerticalBox
   parent=VerticalBox_4
   widget=VerticalBox
   slot {class=VerticalBoxSlot, properties {padding=<ltrb 0 15 0 0>}}
  }
  2DWPOAH {widgetClassPath=&HorizontalBox, bIsVariable=true, parent=VerticalBox, widget=2DWPOAH, slot {class=VerticalBoxSlot}}
  2DWPOA {
   widgetClassPath=&SinglePropertyView
   parent=2DWPOAH
   widget=2DWPOA
   properties {propertyName=AchievedWPOTexture}
   slot {class=HorizontalBoxSlot, properties {size {sizeRule=Fill, value=1}}}
  }
  2DWPOAB {
   widgetClassPath=&Button
   parent=2DWPOAH
   widget=2DWPOAB
   properties {
    backgroundColor=<rgba 0.1 0.1 0.1 1>
    widgetStyle {
     clickedSlateSound {resourceObject=null}
     disabled {
      bIsDynamicallyLoaded=false
      drawAs=RoundedBox
      imageType=NoImage
      mirroring=NoMirror
      resourceName=None
      resourceObject=null
      tiling=NoTile
      imageSize=<0 0>
      margin=<ltrb 0 0 0 0>
      outlineSettings {
       bUseBrushTransparency=false
       roundingType=FixedRadius
       width=1
       color {colorUseRule=UseColor_ColorTable, specifiedColor=<rgba 0 0 0 0>}
       cornerRadii=<4 4 4 4>
      }
      tintColor {colorUseRule=UseColor_Specified, specifiedColor=<rgba 0.0395 0.0395 0.0395 1>}
      uVRegion {bIsValid=false, max=<0 0>, min=<0 0>}
     }
     disabledForeground {colorUseRule=UseColor_ColorTable, specifiedColor=<rgba 0 0 0 0>}
     hovered {
      bIsDynamicallyLoaded=false
      drawAs=RoundedBox
      imageType=NoImage
      mirroring=NoMirror
      resourceName=None
      resourceObject=null
      tiling=NoTile
      imageSize=<0 0>
      margin=<ltrb 0 0 0 0>
      outlineSettings {
       bUseBrushTransparency=false
       roundingType=FixedRadius
       width=1
       color {colorUseRule=UseColor_ColorTable, specifiedColor=<rgba 0 0 0 0>}
       cornerRadii=<4 4 4 4>
      }
      tintColor {colorUseRule=UseColor_Specified, specifiedColor=<rgba 0.0953 0.0953 0.0953 1>}
      uVRegion {bIsValid=false, max=<0 0>, min=<0 0>}
     }
     hoveredForeground {colorUseRule=UseColor_ColorTable, specifiedColor=<rgba 0 0 0 0>}
     hoveredSlateSound {resourceObject=null}
     normal {
      bIsDynamicallyLoaded=false
      drawAs=RoundedBox
      imageType=NoImage
      mirroring=NoMirror
      resourceName=None
      resourceObject=null
      tiling=NoTile
      imageSize=<0 0>
      margin=<ltrb 0 0 0 0>
      outlineSettings {
       bUseBrushTransparency=false
       roundingType=FixedRadius
       width=1
       color {colorUseRule=UseColor_ColorTable, specifiedColor=<rgba 0 0 0 0>}
       cornerRadii=<4 4 4 4>
      }
      tintColor {colorUseRule=UseColor_Specified, specifiedColor=<rgba 0.0395 0.0395 0.0395 1>}
      uVRegion {bIsValid=false, max=<0 0>, min=<0 0>}
     }
     normalForeground {colorUseRule=UseColor_ColorTable, specifiedColor=<rgba 0 0 0 0>}
     normalPadding=<ltrb 12 1.5 12 1.5>
     pressed {
      bIsDynamicallyLoaded=false
      drawAs=RoundedBox
      imageType=NoImage
      mirroring=NoMirror
      resourceName=None
      resourceObject=null
      tiling=NoTile
      imageSize=<0 0>
      margin=<ltrb 0 0 0 0>
      outlineSettings {
       bUseBrushTransparency=false
       roundingType=FixedRadius
       width=1
       color {colorUseRule=UseColor_ColorTable, specifiedColor=<rgba 0 0 0 0>}
       cornerRadii=<4 4 4 4>
      }
      tintColor {colorUseRule=UseColor_Specified, specifiedColor=<rgba 0.0284 0.0284 0.0284 1>}
      uVRegion {bIsValid=false, max=<0 0>, min=<0 0>}
     }
     pressedForeground {colorUseRule=UseColor_ColorTable, specifiedColor=<rgba 0 0 0 0>}
     pressedPadding=<ltrb 12 2 12 0>
     pressedSlateSound {resourceObject=null}
    }
   }
   slot {class=HorizontalBoxSlot, properties {size {sizeRule=Fill, value=0.5}}}
  }
  CommonTextBlock_2 {
   widgetClassPath=&CommonTextBlock
   parent=2DWPOAB
   widget=CommonTextBlock_2
   properties {
    style=&ItemCardPowerRatingTextStyle_S_C
    text=Achieve WPO Texture
    font {
     bForceMonospaced=false
     bMaterialIsStencil=false
     fontMaterial=null
     fontObject=&NotoSans
     letterSpacing=0
     monospacedWidth=1
     size=18
     skewAmount=0
     typefaceFontName=NotoSans-Bold
     outlineSettings {
      bApplyOutlineToDropShadows=false
      bMiteredCorners=false
      bSeparateFillAlpha=false
      outlineMaterial=null
      outlineSize=0
      outlineColor=<rgba 0 0 0 1>
     }
    }
    shadowOffset=<0 0>
   }
   slot {class=ButtonSlot, properties {verticalAlignment=VAlign_Fill, padding=<ltrb 0 0 0 0>}}
  }
  2DNAH {widgetClassPath=&HorizontalBox, bIsVariable=true, parent=VerticalBox, widget=2DNAH, slot {class=VerticalBoxSlot}}
  2DNA {
   widgetClassPath=&SinglePropertyView
   parent=2DNAH
   widget=2DNA
   properties {propertyName=AchievedNormalTexture}
   slot {class=HorizontalBoxSlot, properties {size {sizeRule=Fill, value=1}}}
  }
  2DNAB {
   widgetClassPath=&Button
   parent=2DNAH
   widget=2DNAB
   properties {
    backgroundColor=<rgba 0.1 0.1 0.1 1>
    widgetStyle {
     clickedSlateSound {resourceObject=null}
     disabled {
      bIsDynamicallyLoaded=false
      drawAs=RoundedBox
      imageType=NoImage
      mirroring=NoMirror
      resourceName=None
      resourceObject=null
      tiling=NoTile
      imageSize=<0 0>
      margin=<ltrb 0 0 0 0>
      outlineSettings {
       bUseBrushTransparency=false
       roundingType=FixedRadius
       width=1
       color {colorUseRule=UseColor_ColorTable, specifiedColor=<rgba 0 0 0 0>}
       cornerRadii=<4 4 4 4>
      }
      tintColor {colorUseRule=UseColor_Specified, specifiedColor=<rgba 0.0395 0.0395 0.0395 1>}
      uVRegion {bIsValid=false, max=<0 0>, min=<0 0>}
     }
     disabledForeground {colorUseRule=UseColor_ColorTable, specifiedColor=<rgba 0 0 0 0>}
     hovered {
      bIsDynamicallyLoaded=false
      drawAs=RoundedBox
      imageType=NoImage
      mirroring=NoMirror
      resourceName=None
      resourceObject=null
      tiling=NoTile
      imageSize=<0 0>
      margin=<ltrb 0 0 0 0>
      outlineSettings {
       bUseBrushTransparency=false
       roundingType=FixedRadius
       width=1
       color {colorUseRule=UseColor_ColorTable, specifiedColor=<rgba 0 0 0 0>}
       cornerRadii=<4 4 4 4>
      }
      tintColor {colorUseRule=UseColor_Specified, specifiedColor=<rgba 0.0953 0.0953 0.0953 1>}
      uVRegion {bIsValid=false, max=<0 0>, min=<0 0>}
     }
     hoveredForeground {colorUseRule=UseColor_ColorTable, specifiedColor=<rgba 0 0 0 0>}
     hoveredSlateSound {resourceObject=null}
     normal {
      bIsDynamicallyLoaded=false
      drawAs=RoundedBox
      imageType=NoImage
      mirroring=NoMirror
      resourceName=None
      resourceObject=null
      tiling=NoTile
      imageSize=<0 0>
      margin=<ltrb 0 0 0 0>
      outlineSettings {
       bUseBrushTransparency=false
       roundingType=FixedRadius
       width=1
       color {colorUseRule=UseColor_ColorTable, specifiedColor=<rgba 0 0 0 0>}
       cornerRadii=<4 4 4 4>
      }
      tintColor {colorUseRule=UseColor_Specified, specifiedColor=<rgba 0.0395 0.0395 0.0395 1>}
      uVRegion {bIsValid=false, max=<0 0>, min=<0 0>}
     }
     normalForeground {colorUseRule=UseColor_ColorTable, specifiedColor=<rgba 0 0 0 0>}
     normalPadding=<ltrb 12 1.5 12 1.5>
     pressed {
      bIsDynamicallyLoaded=false
      drawAs=RoundedBox
      imageType=NoImage
      mirroring=NoMirror
      resourceName=None
      resourceObject=null
      tiling=NoTile
      imageSize=<0 0>
      margin=<ltrb 0 0 0 0>
      outlineSettings {
       bUseBrushTransparency=false
       roundingType=FixedRadius
       width=1
       color {colorUseRule=UseColor_ColorTable, specifiedColor=<rgba 0 0 0 0>}
       cornerRadii=<4 4 4 4>
      }
      tintColor {colorUseRule=UseColor_Specified, specifiedColor=<rgba 0.0284 0.0284 0.0284 1>}
      uVRegion {bIsValid=false, max=<0 0>, min=<0 0>}
     }
     pressedForeground {colorUseRule=UseColor_ColorTable, specifiedColor=<rgba 0 0 0 0>}
     pressedPadding=<ltrb 12 2 12 0>
     pressedSlateSound {resourceObject=null}
    }
   }
   slot {class=HorizontalBoxSlot, properties {size {sizeRule=Fill, value=0.5}}}
  }
  CommonTextBlock_5 {
   widgetClassPath=&CommonTextBlock
   parent=2DNAB
   widget=CommonTextBlock_5
   properties {
    style=&ItemCardPowerRatingTextStyle_S_C
    text=Achieve Normal Texture
    font {
     bForceMonospaced=false
     bMaterialIsStencil=false
     fontMaterial=null
     fontObject=&NotoSans
     letterSpacing=0
     monospacedWidth=1
     size=18
     skewAmount=0
     typefaceFontName=NotoSans-Bold
     outlineSettings {
      bApplyOutlineToDropShadows=false
      bMiteredCorners=false
      bSeparateFillAlpha=false
      outlineMaterial=null
      outlineSize=0
      outlineColor=<rgba 0 0 0 1>
     }
    }
    shadowOffset=<0 0>
   }
   slot {class=ButtonSlot, properties {verticalAlignment=VAlign_Fill, padding=<ltrb 0 0 0 0>}}
  }
  ScrollBox_269 {
   widgetClassPath=&ScrollBox
   parent=HorizontalBox_52
   widget=ScrollBox_269
   properties {
    widgetBarStyle {
     thickness=8
     draggedThumbImage {
      bIsDynamicallyLoaded=false
      drawAs=RoundedBox
      imageType=NoImage
      mirroring=NoMirror
      resourceName=None
      resourceObject=null
      tiling=NoTile
      imageSize=<0 0>
      margin=<ltrb 0 0 0 0>
      outlineSettings {
       bUseBrushTransparency=false
       roundingType=FixedRadius
       width=0
       color {colorUseRule=UseColor_Specified, specifiedColor=<rgba 0 0 0 0>}
       cornerRadii=<4 4 4 4>
      }
      tintColor {colorUseRule=UseColor_Specified, specifiedColor=<rgba 0.2159 0.2159 0.2159 1>}
      uVRegion {bIsValid=false, max=<0 0>, min=<0 0>}
     }
     horizontalBackgroundImage {
      bIsDynamicallyLoaded=false
      drawAs=NoDrawType
      imageType=NoImage
      mirroring=NoMirror
      resourceName=None
      resourceObject=null
      tiling=NoTile
      imageSize=<0 0>
      margin=<ltrb 0 0 0 0>
      outlineSettings {
       bUseBrushTransparency=false
       roundingType=HalfHeightRadius
       width=0
       color {colorUseRule=UseColor_Specified, specifiedColor=<rgba 0 0 0 0>}
       cornerRadii=<0 0 0 0>
      }
      tintColor {colorUseRule=UseColor_Specified, specifiedColor=<rgba 1 1 1 1>}
      uVRegion {bIsValid=false, max=<0 0>, min=<0 0>}
     }
     horizontalBottomSlotImage {
      bIsDynamicallyLoaded=false
      drawAs=NoDrawType
      imageType=NoImage
      mirroring=NoMirror
      resourceName=None
      resourceObject=null
      tiling=NoTile
      imageSize=<0 0>
      margin=<ltrb 0 0 0 0>
      outlineSettings {
       bUseBrushTransparency=false
       roundingType=HalfHeightRadius
       width=0
       color {colorUseRule=UseColor_Specified, specifiedColor=<rgba 0 0 0 0>}
       cornerRadii=<0 0 0 0>
      }
      tintColor {colorUseRule=UseColor_Specified, specifiedColor=<rgba 1 1 1 1>}
      uVRegion {bIsValid=false, max=<0 0>, min=<0 0>}
     }
     horizontalTopSlotImage {
      bIsDynamicallyLoaded=false
      drawAs=NoDrawType
      imageType=NoImage
      mirroring=NoMirror
      resourceName=None
      resourceObject=null
      tiling=NoTile
      imageSize=<0 0>
      margin=<ltrb 0 0 0 0>
      outlineSettings {
       bUseBrushTransparency=false
       roundingType=HalfHeightRadius
       width=0
       color {colorUseRule=UseColor_Specified, specifiedColor=<rgba 0 0 0 0>}
       cornerRadii=<0 0 0 0>
      }
      tintColor {colorUseRule=UseColor_Specified, specifiedColor=<rgba 1 1 1 1>}
      uVRegion {bIsValid=false, max=<0 0>, min=<0 0>}
     }
     hoveredThumbImage {
      bIsDynamicallyLoaded=false
      drawAs=RoundedBox
      imageType=NoImage
      mirroring=NoMirror
      resourceName=None
      resourceObject=null
      tiling=NoTile
      imageSize=<0 0>
      margin=<ltrb 0 0 0 0>
      outlineSettings {
       bUseBrushTransparency=false
       roundingType=FixedRadius
       width=0
       color {colorUseRule=UseColor_Specified, specifiedColor=<rgba 0 0 0 0>}
       cornerRadii=<4 4 4 4>
      }
      tintColor {colorUseRule=UseColor_Specified, specifiedColor=<rgba 0.2159 0.2159 0.2159 1>}
      uVRegion {bIsValid=false, max=<0 0>, min=<0 0>}
     }
     normalThumbImage {
      bIsDynamicallyLoaded=false
      drawAs=RoundedBox
      imageType=NoImage
      mirroring=NoMirror
      resourceName=None
      resourceObject=null
      tiling=NoTile
      imageSize=<0 0>
      margin=<ltrb 0 0 0 0>
      outlineSettings {
       bUseBrushTransparency=false
       roundingType=FixedRadius
       width=0
       color {colorUseRule=UseColor_Specified, specifiedColor=<rgba 0 0 0 0>}
       cornerRadii=<4 4 4 4>
      }
      tintColor {colorUseRule=UseColor_Specified, specifiedColor=<rgba 0.0953 0.0953 0.0953 1>}
      uVRegion {bIsValid=false, max=<0 0>, min=<0 0>}
     }
     verticalBackgroundImage {
      bIsDynamicallyLoaded=false
      drawAs=NoDrawType
      imageType=NoImage
      mirroring=NoMirror
      resourceName=None
      resourceObject=null
      tiling=NoTile
      imageSize=<0 0>
      margin=<ltrb 0 0 0 0>
      outlineSettings {
       bUseBrushTransparency=false
       roundingType=HalfHeightRadius
       width=0
       color {colorUseRule=UseColor_Specified, specifiedColor=<rgba 0 0 0 0>}
       cornerRadii=<0 0 0 0>
      }
      tintColor {colorUseRule=UseColor_Specified, specifiedColor=<rgba 1 1 1 1>}
      uVRegion {bIsValid=false, max=<0 0>, min=<0 0>}
     }
     verticalBottomSlotImage {
      bIsDynamicallyLoaded=false
      drawAs=NoDrawType
      imageType=NoImage
      mirroring=NoMirror
      resourceName=None
      resourceObject=null
      tiling=NoTile
      imageSize=<0 0>
      margin=<ltrb 0 0 0 0>
      outlineSettings {
       bUseBrushTransparency=false
       roundingType=HalfHeightRadius
       width=0
       color {colorUseRule=UseColor_Specified, specifiedColor=<rgba 0 0 0 0>}
       cornerRadii=<0 0 0 0>
      }
      tintColor {colorUseRule=UseColor_Specified, specifiedColor=<rgba 1 1 1 1>}
      uVRegion {bIsValid=false, max=<0 0>, min=<0 0>}
     }
     verticalTopSlotImage {
      bIsDynamicallyLoaded=false
      drawAs=NoDrawType
      imageType=NoImage
      mirroring=NoMirror
      resourceName=None
      resourceObject=null
      tiling=NoTile
      imageSize=<0 0>
      margin=<ltrb 0 0 0 0>
      outlineSettings {
       bUseBrushTransparency=false
       roundingType=HalfHeightRadius
       width=0
       color {colorUseRule=UseColor_Specified, specifiedColor=<rgba 0 0 0 0>}
       cornerRadii=<0 0 0 0>
      }
      tintColor {colorUseRule=UseColor_Specified, specifiedColor=<rgba 1 1 1 1>}
      uVRegion {bIsValid=false, max=<0 0>, min=<0 0>}
     }
    }
    widgetStyle {
     barThickness=8
     bottomShadowBrush {
      bIsDynamicallyLoaded=false
      drawAs=Box
      imageType=FullColor
      mirroring=NoMirror
      resourceName=../../../Engine/Content/Slate/Common/ScrollBoxShadowBottom.png
      resourceObject=null
      tiling=NoTile
      imageSize=<16 8>
      margin=<ltrb 0.5 0 0.5 1>
      outlineSettings {
       bUseBrushTransparency=false
       roundingType=HalfHeightRadius
       width=0
       color {colorUseRule=UseColor_Specified, specifiedColor=<rgba 0 0 0 0>}
       cornerRadii=<0 0 0 0>
      }
      tintColor {colorUseRule=UseColor_Specified, specifiedColor=<rgba 1 1 1 1>}
      uVRegion {bIsValid=false, max=<0 0>, min=<0 0>}
     }
     horizontalScrolledContentPadding=<ltrb 0 0 1 0>
     leftShadowBrush {
      bIsDynamicallyLoaded=false
      drawAs=Box
      imageType=FullColor
      mirroring=NoMirror
      resourceName=../../../Engine/Content/Slate/Common/ScrollBoxShadowLeft.png
      resourceObject=null
      tiling=NoTile
      imageSize=<8 16>
      margin=<ltrb 1 0.5 0 0.5>
      outlineSettings {
       bUseBrushTransparency=false
       roundingType=HalfHeightRadius
       width=0
       color {colorUseRule=UseColor_Specified, specifiedColor=<rgba 0 0 0 0>}
       cornerRadii=<0 0 0 0>
      }
      tintColor {colorUseRule=UseColor_Specified, specifiedColor=<rgba 1 1 1 1>}
      uVRegion {bIsValid=false, max=<0 0>, min=<0 0>}
     }
     rightShadowBrush {
      bIsDynamicallyLoaded=false
      drawAs=Box
      imageType=FullColor
      mirroring=NoMirror
      resourceName=../../../Engine/Content/Slate/Common/ScrollBoxShadowRight.png
      resourceObject=null
      tiling=NoTile
      imageSize=<8 16>
      margin=<ltrb 0 0.5 1 0.5>
      outlineSettings {
       bUseBrushTransparency=false
       roundingType=HalfHeightRadius
       width=0
       color {colorUseRule=UseColor_Specified, specifiedColor=<rgba 0 0 0 0>}
       cornerRadii=<0 0 0 0>
      }
      tintColor {colorUseRule=UseColor_Specified, specifiedColor=<rgba 1 1 1 1>}
      uVRegion {bIsValid=false, max=<0 0>, min=<0 0>}
     }
     topShadowBrush {
      bIsDynamicallyLoaded=false
      drawAs=Box
      imageType=FullColor
      mirroring=NoMirror
      resourceName=../../../Engine/Content/Slate/Common/ScrollBoxShadowTop.png
      resourceObject=null
      tiling=NoTile
      imageSize=<16 8>
      margin=<ltrb 0.5 1 0.5 0>
      outlineSettings {
       bUseBrushTransparency=false
       roundingType=HalfHeightRadius
       width=0
       color {colorUseRule=UseColor_Specified, specifiedColor=<rgba 0 0 0 0>}
       cornerRadii=<0 0 0 0>
      }
      tintColor {colorUseRule=UseColor_Specified, specifiedColor=<rgba 1 1 1 1>}
      uVRegion {bIsValid=false, max=<0 0>, min=<0 0>}
     }
     verticalScrolledContentPadding=<ltrb 0 0 0 1>
    }
   }
   slot {class=HorizontalBoxSlot, properties {size {sizeRule=Fill, value=0.5}}}
  }
  DiffuseTextureVB {
   widgetClassPath=&VerticalBox
   bIsVariable=true
   parent=ScrollBox_269
   widget=DiffuseTextureVB
   slot {class=ScrollBoxSlot, properties {verticalAlignment=VAlign_Center, padding=<ltrb 15 15 15 15>}}
  }
  CommonTextBlock_3 {
   widgetClassPath=&CommonTextBlock
   parent=DiffuseTextureVB
   widget=CommonTextBlock_3
   properties {
    text=DiffuseTexture :
    colorAndOpacity {colorUseRule=UseColor_Specified, specifiedColor=<rgba 1 0 0.0145 1>}
    font {
     bForceMonospaced=false
     bMaterialIsStencil=false
     fontMaterial=null
     fontObject=&Burbank
     letterSpacing=0
     monospacedWidth=1
     size=20
     skewAmount=0
     typefaceFontName=Burbank-Black
     outlineSettings {
      bApplyOutlineToDropShadows=false
      bMiteredCorners=false
      bSeparateFillAlpha=false
      outlineMaterial=null
      outlineSize=0
      outlineColor=<rgba 0 0 0 1>
     }
    }
    shadowOffset=<0 0>
   }
   slot {class=VerticalBoxSlot, properties {horizontalAlignment=HAlign_Center, verticalAlignment=VAlign_Center}}
  }
  DiffuseTexture {
   widgetClassPath=&Image
   bIsVariable=true
   parent=DiffuseTextureVB
   widget=DiffuseTexture
   properties {brush {
     bIsDynamicallyLoaded=false
     drawAs=Image
     imageType=NoImage
     mirroring=NoMirror
     resourceName=None
     resourceObject=null
     tiling=NoTile
     imageSize=<512 512>
     margin=<ltrb 0 0 0 0>
     outlineSettings {
      bUseBrushTransparency=false
      roundingType=HalfHeightRadius
      width=0
      color {colorUseRule=UseColor_Specified, specifiedColor=<rgba 0 0 0 0>}
      cornerRadii=<0 0 0 0>
     }
     tintColor {colorUseRule=UseColor_Specified, specifiedColor=<rgba 1 1 1 1>}
     uVRegion {bIsValid=false, max=<0 0>, min=<0 0>}
    }}
   slot {class=VerticalBoxSlot, properties {horizontalAlignment=HAlign_Center, verticalAlignment=VAlign_Center}}
  }
  VerticalBox_2 {
   widgetClassPath=&VerticalBox
   parent=ScrollBox_269
   widget=VerticalBox_2
   slot {class=ScrollBoxSlot, properties {verticalAlignment=VAlign_Center, padding=<ltrb 15 15 15 15>}}
  }
  CommonTextBlock_71 {
   widgetClassPath=&CommonTextBlock
   parent=VerticalBox_2
   widget=CommonTextBlock_71
   properties {
    text=HeightTexture :
    colorAndOpacity {colorUseRule=UseColor_Specified, specifiedColor=<rgba 0 0.8712 1 1>}
    font {
     bForceMonospaced=false
     bMaterialIsStencil=false
     fontMaterial=null
     fontObject=&Burbank
     letterSpacing=0
     monospacedWidth=1
     size=20
     skewAmount=0
     typefaceFontName=Burbank-Black
     outlineSettings {
      bApplyOutlineToDropShadows=false
      bMiteredCorners=false
      bSeparateFillAlpha=false
      outlineMaterial=null
      outlineSize=0
      outlineColor=<rgba 0 0 0 1>
     }
    }
    shadowOffset=<0 0>
   }
   slot {class=VerticalBoxSlot, properties {horizontalAlignment=HAlign_Center, verticalAlignment=VAlign_Center}}
  }
  DepthTexture {
   widgetClassPath=&Image
   bIsVariable=true
   parent=VerticalBox_2
   widget=DepthTexture
   properties {brush {
     bIsDynamicallyLoaded=false
     drawAs=Image
     imageType=NoImage
     mirroring=NoMirror
     resourceName=None
     resourceObject=null
     tiling=NoTile
     imageSize=<512 512>
     margin=<ltrb 0 0 0 0>
     outlineSettings {
      bUseBrushTransparency=false
      roundingType=HalfHeightRadius
      width=0
      color {colorUseRule=UseColor_Specified, specifiedColor=<rgba 0 0 0 0>}
      cornerRadii=<0 0 0 0>
     }
     tintColor {colorUseRule=UseColor_Specified, specifiedColor=<rgba 1 1 1 1>}
     uVRegion {bIsValid=false, max=<0 0>, min=<0 0>}
    }}
   slot {class=VerticalBoxSlot, properties {horizontalAlignment=HAlign_Center, verticalAlignment=VAlign_Center}}
  }
  WPOTextureVB {
   widgetClassPath=&VerticalBox
   bIsVariable=true
   parent=ScrollBox_269
   widget=WPOTextureVB
   slot {class=ScrollBoxSlot, properties {verticalAlignment=VAlign_Center, padding=<ltrb 15 15 15 15>}}
  }
  WPOTextureText {
   widgetClassPath=&CommonTextBlock
   parent=WPOTextureVB
   widget=WPOTextureText
   properties {
    text=WPOTexture :
    colorAndOpacity {colorUseRule=UseColor_Specified, specifiedColor=<rgba 1 0.5066 0 1>}
    font {
     bForceMonospaced=false
     bMaterialIsStencil=false
     fontMaterial=null
     fontObject=&Burbank
     letterSpacing=0
     monospacedWidth=1
     size=20
     skewAmount=0
     typefaceFontName=Burbank-Black
     outlineSettings {
      bApplyOutlineToDropShadows=false
      bMiteredCorners=false
      bSeparateFillAlpha=false
      outlineMaterial=null
      outlineSize=0
      outlineColor=<rgba 0 0 0 1>
     }
    }
    shadowOffset=<0 0>
   }
   slot {class=VerticalBoxSlot, properties {horizontalAlignment=HAlign_Center, verticalAlignment=VAlign_Center}}
  }
  WPOTexture {
   widgetClassPath=&Image
   bIsVariable=true
   parent=WPOTextureVB
   widget=WPOTexture
   properties {brush {
     bIsDynamicallyLoaded=false
     drawAs=Image
     imageType=NoImage
     mirroring=NoMirror
     resourceName=None
     resourceObject=null
     tiling=NoTile
     imageSize=<512 512>
     margin=<ltrb 0 0 0 0>
     outlineSettings {
      bUseBrushTransparency=false
      roundingType=HalfHeightRadius
      width=0
      color {colorUseRule=UseColor_Specified, specifiedColor=<rgba 0 0 0 0>}
      cornerRadii=<0 0 0 0>
     }
     tintColor {colorUseRule=UseColor_Specified, specifiedColor=<rgba 1 1 1 1>}
     uVRegion {bIsValid=false, max=<0 0>, min=<0 0>}
    }}
   slot {class=VerticalBoxSlot, properties {horizontalAlignment=HAlign_Center, verticalAlignment=VAlign_Center}}
  }
  NormalTextureVB {
   widgetClassPath=&VerticalBox
   bIsVariable=true
   parent=ScrollBox_269
   widget=NormalTextureVB
   slot {class=ScrollBoxSlot, properties {verticalAlignment=VAlign_Center, padding=<ltrb 15 15 15 15>}}
  }
  NormalTextureText {
   widgetClassPath=&CommonTextBlock
   parent=NormalTextureVB
   widget=NormalTextureText
   properties {
    text=NormalTexture :
    colorAndOpacity {colorUseRule=UseColor_Specified, specifiedColor=<rgba 0 1 0.4121 1>}
    font {
     bForceMonospaced=false
     bMaterialIsStencil=false
     fontMaterial=null
     fontObject=&Burbank
     letterSpacing=0
     monospacedWidth=1
     size=20
     skewAmount=0
     typefaceFontName=Burbank-Black
     outlineSettings {
      bApplyOutlineToDropShadows=false
      bMiteredCorners=false
      bSeparateFillAlpha=false
      outlineMaterial=null
      outlineSize=0
      outlineColor=<rgba 0 0 0 1>
     }
    }
    shadowOffset=<0 0>
   }
   slot {class=VerticalBoxSlot, properties {horizontalAlignment=HAlign_Center, verticalAlignment=VAlign_Center}}
  }
  NormalTexture {
   widgetClassPath=&Image
   bIsVariable=true
   parent=NormalTextureVB
   widget=NormalTexture
   properties {brush {
     bIsDynamicallyLoaded=false
     drawAs=Image
     imageType=NoImage
     mirroring=NoMirror
     resourceName=None
     resourceObject=null
     tiling=NoTile
     imageSize=<512 512>
     margin=<ltrb 0 0 0 0>
     outlineSettings {
      bUseBrushTransparency=false
      roundingType=HalfHeightRadius
      width=0
      color {colorUseRule=UseColor_Specified, specifiedColor=<rgba 0 0 0 0>}
      cornerRadii=<0 0 0 0>
     }
     tintColor {colorUseRule=UseColor_Specified, specifiedColor=<rgba 1 1 1 1>}
     uVRegion {bIsValid=false, max=<0 0>, min=<0 0>}
    }}
   slot {class=VerticalBoxSlot, properties {horizontalAlignment=HAlign_Center, verticalAlignment=VAlign_Center}}
  }
  VerticalBox_0 {
   widgetClassPath=&VerticalBox
   parent=ScrollBox_269
   widget=VerticalBox_0
   properties {visibility=Collapsed}
   slot {class=ScrollBoxSlot, properties {verticalAlignment=VAlign_Center, padding=<ltrb 15 15 15 15>}}
  }
  CommonTextBlock_1 {
   widgetClassPath=&CommonTextBlock
   parent=VerticalBox_0
   widget=CommonTextBlock_1
   properties {
    text=MeshMaskTexture :
    colorAndOpacity {colorUseRule=UseColor_Specified, specifiedColor=<rgba 0.4941 1 0 1>}
    font {
     bForceMonospaced=false
     bMaterialIsStencil=false
     fontMaterial=null
     fontObject=&Burbank
     letterSpacing=0
     monospacedWidth=1
     size=20
     skewAmount=0
     typefaceFontName=Burbank-Black
     outlineSettings {
      bApplyOutlineToDropShadows=false
      bMiteredCorners=false
      bSeparateFillAlpha=false
      outlineMaterial=null
      outlineSize=0
      outlineColor=<rgba 0 0 0 1>
     }
    }
    shadowOffset=<0 0>
   }
   slot {class=VerticalBoxSlot, properties {horizontalAlignment=HAlign_Center, verticalAlignment=VAlign_Center}}
  }
  MeshMaskTexture {
   widgetClassPath=&Image
   bIsVariable=true
   parent=VerticalBox_0
   widget=MeshMaskTexture
   properties {brush {
     bIsDynamicallyLoaded=false
     drawAs=Image
     imageType=NoImage
     mirroring=NoMirror
     resourceName=None
     resourceObject=null
     tiling=NoTile
     imageSize=<512 512>
     margin=<ltrb 0 0 0 0>
     outlineSettings {
      bUseBrushTransparency=false
      roundingType=HalfHeightRadius
      width=0
      color {colorUseRule=UseColor_Specified, specifiedColor=<rgba 0 0 0 0>}
      cornerRadii=<0 0 0 0>
     }
     tintColor {colorUseRule=UseColor_Specified, specifiedColor=<rgba 1 1 1 1>}
     uVRegion {bIsValid=false, max=<0 0>, min=<0 0>}
    }}
   slot {class=VerticalBoxSlot, properties {horizontalAlignment=HAlign_Center, verticalAlignment=VAlign_Center}}
  }
 ]
}
