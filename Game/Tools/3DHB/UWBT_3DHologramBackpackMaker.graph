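# Editor utility widget for making 3D hologram backpacks, with four step buttons that switch between Initializer, Texture Baker, Material Maker and Previewer panels.
# The Initializer's state-change event enables steps 2-4 when the workspace is ready and passes its target data asset to the other panels.
# The same event also deactivates the baker and turns off the previewer.
# Texture edits reapply to the material and refresh the preview; material changes refresh the preview.
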EditorUtilityWidgetBlueprint /Game/Tools/3DHB/UWBT_3DHologramBackpackMaker
graphs [EventGraph {nodes [
   "On Clicked (Step1)"() {is=title K2Node_ComponentBoundEvent}
   > SetActiveWidgetIndex(self=Get MainTableSwitcher() -> (MainTableSwitcher:Common Activatable Widget Switcher Object Reference), Index=0)
   "On Clicked (Step3)"() {is=title K2Node_ComponentBoundEvent}
   n1 = Get MainTableSwitcher()
   > SetActiveWidgetIndex(self=@n1, Index=2)
   "On Clicked (Step4)"() {is=title K2Node_ComponentBoundEvent}
   > SetActiveWidgetIndex(self=@n1, Index=3)
   "On Clicked (Step2)"() {is=title K2Node_ComponentBoundEvent}
   > SetActiveWidgetIndex(self=Get MainTableSwitcher(), Index=1)
   n2 = Event_WhenInitializerStatChange!() -> (OutputDelegate, then, bWorkspaceReady:bool=~) {is=title K2Node_CustomEvent}
   > n3 = Sequence!(execute) -> (then_0, then_1, then_2:Exec)
   > n4 = Set bWorkspaceReady(bWorkspaceReady:bool=@n2.bWorkspaceReady) -> (Output_Get:bool)
   > SetIsEnabled(
    self=[
     Get Step4() -> (Step4:Button Object Reference)
     Get Step3() -> (Step3:Button Object Reference)
     Get Step2() -> (Step2:Button Object Reference)
    ]
    bInIsEnabled=@n4
   )
   > Branch(Condition=@n4)
   n5 = Get Target3DHBDataAsset!(self:UWBT 3DHB Initializer Object Reference=Get UWBT_3DHB_Initializer() -> (UWBT_3DHB_Initializer:UWBT 3DHB Initializer Object Reference)) -> (Target3DHBDataAsset:Primary Data Asset Object Reference)
   > InitialWithDataAsset!(
    execute
    self:UWBT 3DHB Texture Baker Object Reference=Get UWBT_3DHB_TextureBaker() -> (UWBT_3DHB_TextureBaker:UWBT 3DHB Texture Baker Object Reference)
    3DHBDataAsset:Primary Data Asset Object Reference=@n5
   ) -> (then)
   n6 = Cast To 3DHologramBackpackDataAsset!(Object=@n5) -> (As3DHologram Backpack Data Asset:3DHologram Backpack Data Asset Object Reference, bSuccess:bool=~) {is=title K2Node_DynamicCast}
   n7 = Get UWBT_3DHB_MaterialMaker() -> (UWBT_3DHB_MaterialMaker:UWBT 3DHB Material Maker Object Reference)
   > Initial!(execute, self:UWBT 3DHB Material Maker Object Reference=@n7, DataAsset:3DHologram Backpack Data Asset Object Reference=@n6) -> (then)
   n8 = Get UWBT_3DHB_Previewer() -> (UWBT_3DHB_Previewer:UWBT 3DHB Previewer Object Reference)
   > Initial!(execute, self:UWBT 3DHB Previewer Object Reference=@n8, Data:3DHologram Backpack Data Asset Object Reference=@n6) -> (then)
   Deactive3DBaker(execute=@n3.then_1, self:UWBT 3DHB Texture Baker Object Reference=Get UWBT_3DHB_TextureBaker())
   > Turnoff(self:UWBT 3DHB Previewer Object Reference=Get UWBT_3DHB_Previewer())
   n9 = Event_WhenTextureBakerPropertyUpdate() {is=title K2Node_CustomEvent}
   > ApplyTexture(self:UWBT 3DHB Material Maker Object Reference=@n7)
   > UpdateState(self:UWBT 3DHB Previewer Object Reference=@n8)
   n10 = Event_WhenChangeMaterial() {is=title K2Node_CustomEvent}
   > UpdateState(self=@n8)
   Event Construct()
   > Bind Event to When Stat Change(self:UWBT 3DHB Initializer Object Reference=Get UWBT_3DHB_Initializer(), Delegate=@n2) {is=title K2Node_AddDelegate}
   > Bind Event to Property Update(self:UWBT 3DHB Texture Baker Object Reference=Get UWBT_3DHB_TextureBaker(), Delegate=@n9) {is=title K2Node_AddDelegate}
   > Bind Event to Property Change(self:UWBT 3DHB Material Maker Object Reference=Get UWBT_3DHB_MaterialMaker(), Delegate=@n10) {is=title K2Node_AddDelegate}
  ]}]
variables [bWorkspaceReady {type=boolean}]
widgets {
 info {rootWidgetClass=&CanvasPanel, widgetCount=24}
 widgets [
  CanvasPanel_85 {widgetClassPath=&CanvasPanel, widget=CanvasPanel_85}
  HorizontalBox_0 {
   widgetClassPath=&HorizontalBox
   parent=CanvasPanel_85
   widget=HorizontalBox_0
   slot {class=CanvasPanelSlot, properties {layoutData {alignment=<0 0>, anchors {maximum=<1 1>, minimum=<0 0>}, offsets=<ltrb 0 0 0 0>}}}
  }
  CanvasPanel_134 {
   widgetClassPath=&CanvasPanel
   parent=HorizontalBox_0
   widget=CanvasPanel_134
   slot {class=HorizontalBoxSlot, properties {size {sizeRule=Fill, value=0.3}}}
  }
  Overlay_57 {
   widgetClassPath=&Overlay
   parent=CanvasPanel_134
   widget=Overlay_57
   slot {class=CanvasPanelSlot, properties {layoutData {alignment=<0 0>, anchors {maximum=<1 1>, minimum=<0 0>}, offsets=<ltrb 0 0 0 0>}}}
  }
  Border_67 {
   widgetClassPath=&Border
   parent=Overlay_57
   widget=Border_67
   properties {brushColor=<rgba 0.01 0.01 0.01 1>}
   slot {class=OverlaySlot, properties {horizontalAlignment=HAlign_Fill, verticalAlignment=VAlign_Fill}}
  }
  VerticalBox_121 {
   widgetClassPath=&VerticalBox
   parent=Overlay_57
   widget=VerticalBox_121
   slot {class=OverlaySlot, properties {horizontalAlignment=HAlign_Fill, verticalAlignment=VAlign_Fill}}
  }
  CommonTextBlock_95 {
   widgetClassPath=&CommonTextBlock
   parent=VerticalBox_121
   widget=CommonTextBlock_95
   properties {
    justification=Center
    text=3D Hologram Backpack Maker
    font {
     bForceMonospaced=false
     bMaterialIsStencil=false
     fontMaterial=null
     fontObject=&Burbank
     letterSpacing=0
     monospacedWidth=1
     size=26
     skewAmount=0
     typefaceFontName=Burbank-Black
     outlineSettings {
      bApplyOutlineToDropShadows=false
      bMiteredCorners=false
      bSeparateFillAlpha=false
      outlineMaterial=null
      outlineSize=0
      outlineColor=<rgba 0 0 0 1>
     }
    }
    shadowOffset=<0 0>
   }
   slot {class=VerticalBoxSlot, properties {padding=<ltrb 0 20 0 0>}}
  }
  Border_171 {
   widgetClassPath=&Border
   parent=VerticalBox_121
   widget=Border_171
   properties {brushColor=<rgba 0.026 0.026 0.026 1>}
   slot {class=VerticalBoxSlot, properties {padding=<ltrb 0 20 0 0>}}
  }
  Step1 {
   widgetClassPath=&Button
   parent=VerticalBox_121
   widget=Step1
   properties {
    backgroundColor=<rgba 0.0521 0.0521 0.0521 1>
    widgetStyle {
     clickedSlateSound {resourceObject=null}
     disabled {
      bIsDynamicallyLoaded=false
      drawAs=RoundedBox
      imageType=NoImage
      mirroring=NoMirror
      resourceName=None
      resourceObject=null
      tiling=NoTile
      imageSize=<0 0>
      margin=<ltrb 0 0 0 0>
      outlineSettings {
       bUseBrushTransparency=false
       roundingType=FixedRadius
       width=1
       color {colorUseRule=UseColor_ColorTable, specifiedColor=<rgba 0 0 0 0>}
       cornerRadii=<4 4 4 4>
      }
      tintColor {colorUseRule=UseColor_Specified, specifiedColor=<rgba 0.0395 0.0395 0.0395 1>}
      uVRegion {bIsValid=false, max=<0 0>, min=<0 0>}
     }
     disabledForeground {colorUseRule=UseColor_ColorTable, specifiedColor=<rgba 0 0 0 0>}
     hovered {
      bIsDynamicallyLoaded=false
      drawAs=RoundedBox
      imageType=NoImage
      mirroring=NoMirror
      resourceName=None
      resourceObject=null
      tiling=NoTile
      imageSize=<0 0>
      margin=<ltrb 0 0 0 0>
      outlineSettings {
       bUseBrushTransparency=false
       roundingType=FixedRadius
       width=1
       color {colorUseRule=UseColor_ColorTable, specifiedColor=<rgba 0 0 0 0>}
       cornerRadii=<4 4 4 4>
      }
      tintColor {colorUseRule=UseColor_Specified, specifiedColor=<rgba 0.0953 0.0953 0.0953 1>}
      uVRegion {bIsValid=false, max=<0 0>, min=<0 0>}
     }
     hoveredForeground {colorUseRule=UseColor_ColorTable, specifiedColor=<rgba 0 0 0 0>}
     hoveredSlateSound {resourceObject=null}
     normal {
      bIsDynamicallyLoaded=false
      drawAs=RoundedBox
      imageType=NoImage
      mirroring=NoMirror
      resourceName=None
      resourceObject=null
      tiling=NoTile
      imageSize=<0 0>
      margin=<ltrb 0 0 0 0>
      outlineSettings {
       bUseBrushTransparency=false
       roundingType=FixedRadius
       width=1
       color {colorUseRule=UseColor_ColorTable, specifiedColor=<rgba 0 0 0 0>}
       cornerRadii=<4 4 4 4>
      }
      tintColor {colorUseRule=UseColor_Specified, specifiedColor=<rgba 0.0395 0.0395 0.0395 1>}
      uVRegion {bIsValid=false, max=<0 0>, min=<0 0>}
     }
     normalForeground {colorUseRule=UseColor_ColorTable, specifiedColor=<rgba 0 0 0 0>}
     normalPadding=<ltrb 12 1.5 12 1.5>
     pressed {
      bIsDynamicallyLoaded=false
      drawAs=RoundedBox
      imageType=NoImage
      mirroring=NoMirror
      resourceName=None
      resourceObject=null
      tiling=NoTile
      imageSize=<0 0>
      margin=<ltrb 0 0 0 0>
      outlineSettings {
       bUseBrushTransparency=false
       roundingType=FixedRadius
       width=1
       color {colorUseRule=UseColor_ColorTable, specifiedColor=<rgba 0 0 0 0>}
       cornerRadii=<4 4 4 4>
      }
      tintColor {colorUseRule=UseColor_Specified, specifiedColor=<rgba 0.0284 0.0284 0.0284 1>}
      uVRegion {bIsValid=false, max=<0 0>, min=<0 0>}
     }
     pressedForeground {colorUseRule=UseColor_ColorTable, specifiedColor=<rgba 0 0 0 0>}
     pressedPadding=<ltrb 12 2 12 0>
     pressedSlateSound {resourceObject=null}
    }
   }
   slot {class=VerticalBoxSlot, properties {padding=<ltrb 1 25 1 0>}}
  }
  CommonTextBlock {
   widgetClassPath=&CommonTextBlock
   parent=Step1
   widget=CommonTextBlock
   properties {
    justification=Center
    text=Initial workspace
    colorAndOpacity {colorUseRule=UseColor_Specified, specifiedColor=<rgba 0.6042 0.6042 0.6042 1>}
    font {
     bForceMonospaced=false
     bMaterialIsStencil=false
     fontMaterial=null
     fontObject=&Burbank
     letterSpacing=0
     monospacedWidth=1
     size=26
     skewAmount=0
     typefaceFontName=Burbank-Black
     outlineSettings {
      bApplyOutlineToDropShadows=false
      bMiteredCorners=false
      bSeparateFillAlpha=false
      outlineMaterial=null
      outlineSize=0
      outlineColor=<rgba 0 0 0 1>
     }
    }
    shadowOffset=<0 0>
   }
   slot {class=ButtonSlot, properties {horizontalAlignment=HAlign_Fill, verticalAlignment=VAlign_Fill, padding=<ltrb 0 0 0 0>}}
  }
  Step2 {
   widgetClassPath=&Button
   parent=VerticalBox_121
   widget=Step2
   properties {
    bIsEnabled=false
    backgroundColor=<rgba 0.0521 0.0521 0.0521 1>
    widgetStyle {
     clickedSlateSound {resourceObject=null}
     disabled {
      bIsDynamicallyLoaded=false
      drawAs=RoundedBox
      imageType=NoImage
      mirroring=NoMirror
      resourceName=None
      resourceObject=null
      tiling=NoTile
      imageSize=<0 0>
      margin=<ltrb 0 0 0 0>
      outlineSettings {
       bUseBrushTransparency=false
       roundingType=FixedRadius
       width=1
       color {colorUseRule=UseColor_ColorTable, specifiedColor=<rgba 0 0 0 0>}
       cornerRadii=<4 4 4 4>
      }
      tintColor {colorUseRule=UseColor_Specified, specifiedColor=<rgba 0.0395 0.0395 0.0395 1>}
      uVRegion {bIsValid=false, max=<0 0>, min=<0 0>}
     }
     disabledForeground {colorUseRule=UseColor_ColorTable, specifiedColor=<rgba 0 0 0 0>}
     hovered {
      bIsDynamicallyLoaded=false
      drawAs=RoundedBox
      imageType=NoImage
      mirroring=NoMirror
      resourceName=None
      resourceObject=null
      tiling=NoTile
      imageSize=<0 0>
      margin=<ltrb 0 0 0 0>
      outlineSettings {
       bUseBrushTransparency=false
       roundingType=FixedRadius
       width=1
       color {colorUseRule=UseColor_ColorTable, specifiedColor=<rgba 0 0 0 0>}
       cornerRadii=<4 4 4 4>
      }
      tintColor {colorUseRule=UseColor_Specified, specifiedColor=<rgba 0.0953 0.0953 0.0953 1>}
      uVRegion {bIsValid=false, max=<0 0>, min=<0 0>}
     }
     hoveredForeground {colorUseRule=UseColor_ColorTable, specifiedColor=<rgba 0 0 0 0>}
     hoveredSlateSound {resourceObject=null}
     normal {
      bIsDynamicallyLoaded=false
      drawAs=RoundedBox
      imageType=NoImage
      mirroring=NoMirror
      resourceName=None
      resourceObject=null
      tiling=NoTile
      imageSize=<0 0>
      margin=<ltrb 0 0 0 0>
      outlineSettings {
       bUseBrushTransparency=false
       roundingType=FixedRadius
       width=1
       color {colorUseRule=UseColor_ColorTable, specifiedColor=<rgba 0 0 0 0>}
       cornerRadii=<4 4 4 4>
      }
      tintColor {colorUseRule=UseColor_Specified, specifiedColor=<rgba 0.0395 0.0395 0.0395 1>}
      uVRegion {bIsValid=false, max=<0 0>, min=<0 0>}
     }
     normalForeground {colorUseRule=UseColor_ColorTable, specifiedColor=<rgba 0 0 0 0>}
     normalPadding=<ltrb 12 1.5 12 1.5>
     pressed {
      bIsDynamicallyLoaded=false
      drawAs=RoundedBox
      imageType=NoImage
      mirroring=NoMirror
      resourceName=None
      resourceObject=null
      tiling=NoTile
      imageSize=<0 0>
      margin=<ltrb 0 0 0 0>
      outlineSettings {
       bUseBrushTransparency=false
       roundingType=FixedRadius
       width=1
       color {colorUseRule=UseColor_ColorTable, specifiedColor=<rgba 0 0 0 0>}
       cornerRadii=<4 4 4 4>
      }
      tintColor {colorUseRule=UseColor_Specified, specifiedColor=<rgba 0.0284 0.0284 0.0284 1>}
      uVRegion {bIsValid=false, max=<0 0>, min=<0 0>}
     }
     pressedForeground {colorUseRule=UseColor_ColorTable, specifiedColor=<rgba 0 0 0 0>}
     pressedPadding=<ltrb 12 2 12 0>
     pressedSlateSound {resourceObject=null}
    }
   }
   slot {class=VerticalBoxSlot, properties {padding=<ltrb 1 25 1 0>}}
  }
  CommonTextBlock_2 {
   widgetClassPath=&CommonTextBlock
   parent=Step2
   widget=CommonTextBlock_2
   properties {
    justification=Center
    text=Texture setting
    colorAndOpacity {colorUseRule=UseColor_Specified, specifiedColor=<rgba 0.6042 0.6042 0.6042 1>}
    font {
     bForceMonospaced=false
     bMaterialIsStencil=false
     fontMaterial=null
     fontObject=&Burbank
     letterSpacing=0
     monospacedWidth=1
     size=26
     skewAmount=0
     typefaceFontName=Burbank-Black
     outlineSettings {
      bApplyOutlineToDropShadows=false
      bMiteredCorners=false
      bSeparateFillAlpha=false
      outlineMaterial=null
      outlineSize=0
      outlineColor=<rgba 0 0 0 1>
     }
    }
    shadowOffset=<0 0>
   }
   slot {class=ButtonSlot, properties {horizontalAlignment=HAlign_Fill, verticalAlignment=VAlign_Fill, padding=<ltrb 0 0 0 0>}}
  }
  Step3 {
   widgetClassPath=&Button
   parent=VerticalBox_121
   widget=Step3
   properties {
    bIsEnabled=false
    backgroundColor=<rgba 0.0521 0.0521 0.0521 1>
    widgetStyle {
     clickedSlateSound {resourceObject=null}
     disabled {
      bIsDynamicallyLoaded=false
      drawAs=RoundedBox
      imageType=NoImage
      mirroring=NoMirror
      resourceName=None
      resourceObject=null
      tiling=NoTile
      imageSize=<0 0>
      margin=<ltrb 0 0 0 0>
      outlineSettings {
       bUseBrushTransparency=false
       roundingType=FixedRadius
       width=1
       color {colorUseRule=UseColor_ColorTable, specifiedColor=<rgba 0 0 0 0>}
       cornerRadii=<4 4 4 4>
      }
      tintColor {colorUseRule=UseColor_Specified, specifiedColor=<rgba 0.0395 0.0395 0.0395 1>}
      uVRegion {bIsValid=false, max=<0 0>, min=<0 0>}
     }
     disabledForeground {colorUseRule=UseColor_ColorTable, specifiedColor=<rgba 0 0 0 0>}
     hovered {
      bIsDynamicallyLoaded=false
      drawAs=RoundedBox
      imageType=NoImage
      mirroring=NoMirror
      resourceName=None
      resourceObject=null
      tiling=NoTile
      imageSize=<0 0>
      margin=<ltrb 0 0 0 0>
      outlineSettings {
       bUseBrushTransparency=false
       roundingType=FixedRadius
       width=1
       color {colorUseRule=UseColor_ColorTable, specifiedColor=<rgba 0 0 0 0>}
       cornerRadii=<4 4 4 4>
      }
      tintColor {colorUseRule=UseColor_Specified, specifiedColor=<rgba 0.0953 0.0953 0.0953 1>}
      uVRegion {bIsValid=false, max=<0 0>, min=<0 0>}
     }
     hoveredForeground {colorUseRule=UseColor_ColorTable, specifiedColor=<rgba 0 0 0 0>}
     hoveredSlateSound {resourceObject=null}
     normal {
      bIsDynamicallyLoaded=false
      drawAs=RoundedBox
      imageType=NoImage
      mirroring=NoMirror
      resourceName=None
      resourceObject=null
      tiling=NoTile
      imageSize=<0 0>
      margin=<ltrb 0 0 0 0>
      outlineSettings {
       bUseBrushTransparency=false
       roundingType=FixedRadius
       width=1
       color {colorUseRule=UseColor_ColorTable, specifiedColor=<rgba 0 0 0 0>}
       cornerRadii=<4 4 4 4>
      }
      tintColor {colorUseRule=UseColor_Specified, specifiedColor=<rgba 0.0395 0.0395 0.0395 1>}
      uVRegion {bIsValid=false, max=<0 0>, min=<0 0>}
     }
     normalForeground {colorUseRule=UseColor_ColorTable, specifiedColor=<rgba 0 0 0 0>}
     normalPadding=<ltrb 12 1.5 12 1.5>
     pressed {
      bIsDynamicallyLoaded=false
      drawAs=RoundedBox
      imageType=NoImage
      mirroring=NoMirror
      resourceName=None
      resourceObject=null
      tiling=NoTile
      imageSize=<0 0>
      margin=<ltrb 0 0 0 0>
      outlineSettings {
       bUseBrushTransparency=false
       roundingType=FixedRadius
       width=1
       color {colorUseRule=UseColor_ColorTable, specifiedColor=<rgba 0 0 0 0>}
       cornerRadii=<4 4 4 4>
      }
      tintColor {colorUseRule=UseColor_Specified, specifiedColor=<rgba 0.0284 0.0284 0.0284 1>}
      uVRegion {bIsValid=false, max=<0 0>, min=<0 0>}
     }
     pressedForeground {colorUseRule=UseColor_ColorTable, specifiedColor=<rgba 0 0 0 0>}
     pressedPadding=<ltrb 12 2 12 0>
     pressedSlateSound {resourceObject=null}
    }
   }
   slot {class=VerticalBoxSlot, properties {padding=<ltrb 1 25 1 0>}}
  }
  CommonTextBlock_1 {
   widgetClassPath=&CommonTextBlock
   parent=Step3
   widget=CommonTextBlock_1
   properties {
    justification=Center
    text=Material setting
    colorAndOpacity {colorUseRule=UseColor_Specified, specifiedColor=<rgba 0.6042 0.6042 0.6042 1>}
    font {
     bForceMonospaced=false
     bMaterialIsStencil=false
     fontMaterial=null
     fontObject=&Burbank
     letterSpacing=0
     monospacedWidth=1
     size=26
     skewAmount=0
     typefaceFontName=Burbank-Black
     outlineSettings {
      bApplyOutlineToDropShadows=false
      bMiteredCorners=false
      bSeparateFillAlpha=false
      outlineMaterial=null
      outlineSize=0
      outlineColor=<rgba 0 0 0 1>
     }
    }
    shadowOffset=<0 0>
   }
   slot {class=ButtonSlot, properties {horizontalAlignment=HAlign_Fill, verticalAlignment=VAlign_Fill, padding=<ltrb 0 0 0 0>}}
  }
  Step4 {
   widgetClassPath=&Button
   parent=VerticalBox_121
   widget=Step4
   properties {
    bIsEnabled=false
    backgroundColor=<rgba 0.0521 0.0521 0.0521 1>
    widgetStyle {
     clickedSlateSound {resourceObject=null}
     disabled {
      bIsDynamicallyLoaded=false
      drawAs=RoundedBox
      imageType=NoImage
      mirroring=NoMirror
      resourceName=None
      resourceObject=null
      tiling=NoTile
      imageSize=<0 0>
      margin=<ltrb 0 0 0 0>
      outlineSettings {
       bUseBrushTransparency=false
       roundingType=FixedRadius
       width=1
       color {colorUseRule=UseColor_ColorTable, specifiedColor=<rgba 0 0 0 0>}
       cornerRadii=<4 4 4 4>
      }
      tintColor {colorUseRule=UseColor_Specified, specifiedColor=<rgba 0.0395 0.0395 0.0395 1>}
      uVRegion {bIsValid=false, max=<0 0>, min=<0 0>}
     }
     disabledForeground {colorUseRule=UseColor_ColorTable, specifiedColor=<rgba 0 0 0 0>}
     hovered {
      bIsDynamicallyLoaded=false
      drawAs=RoundedBox
      imageType=NoImage
      mirroring=NoMirror
      resourceName=None
      resourceObject=null
      tiling=NoTile
      imageSize=<0 0>
      margin=<ltrb 0 0 0 0>
      outlineSettings {
       bUseBrushTransparency=false
       roundingType=FixedRadius
       width=1
       color {colorUseRule=UseColor_ColorTable, specifiedColor=<rgba 0 0 0 0>}
       cornerRadii=<4 4 4 4>
      }
      tintColor {colorUseRule=UseColor_Specified, specifiedColor=<rgba 0.0953 0.0953 0.0953 1>}
      uVRegion {bIsValid=false, max=<0 0>, min=<0 0>}
     }
     hoveredForeground {colorUseRule=UseColor_ColorTable, specifiedColor=<rgba 0 0 0 0>}
     hoveredSlateSound {resourceObject=null}
     normal {
      bIsDynamicallyLoaded=false
      drawAs=RoundedBox
      imageType=NoImage
      mirroring=NoMirror
      resourceName=None
      resourceObject=null
      tiling=NoTile
      imageSize=<0 0>
      margin=<ltrb 0 0 0 0>
      outlineSettings {
       bUseBrushTransparency=false
       roundingType=FixedRadius
       width=1
       color {colorUseRule=UseColor_ColorTable, specifiedColor=<rgba 0 0 0 0>}
       cornerRadii=<4 4 4 4>
      }
      tintColor {colorUseRule=UseColor_Specified, specifiedColor=<rgba 0.0395 0.0395 0.0395 1>}
      uVRegion {bIsValid=false, max=<0 0>, min=<0 0>}
     }
     normalForeground {colorUseRule=UseColor_ColorTable, specifiedColor=<rgba 0 0 0 0>}
     normalPadding=<ltrb 12 1.5 12 1.5>
     pressed {
      bIsDynamicallyLoaded=false
      drawAs=RoundedBox
      imageType=NoImage
      mirroring=NoMirror
      resourceName=None
      resourceObject=null
      tiling=NoTile
      imageSize=<0 0>
      margin=<ltrb 0 0 0 0>
      outlineSettings {
       bUseBrushTransparency=false
       roundingType=FixedRadius
       width=1
       color {colorUseRule=UseColor_ColorTable, specifiedColor=<rgba 0 0 0 0>}
       cornerRadii=<4 4 4 4>
      }
      tintColor {colorUseRule=UseColor_Specified, specifiedColor=<rgba 0.0284 0.0284 0.0284 1>}
      uVRegion {bIsValid=false, max=<0 0>, min=<0 0>}
     }
     pressedForeground {colorUseRule=UseColor_ColorTable, specifiedColor=<rgba 0 0 0 0>}
     pressedPadding=<ltrb 12 2 12 0>
     pressedSlateSound {resourceObject=null}
    }
   }
   slot {class=VerticalBoxSlot, properties {padding=<ltrb 1 25 1 0>}}
  }
  CommonTextBlock_3 {
   widgetClassPath=&CommonTextBlock
   parent=Step4
   widget=CommonTextBlock_3
   properties {
    justification=Center
    text=Preview
    colorAndOpacity {colorUseRule=UseColor_Specified, specifiedColor=<rgba 0.6042 0.6042 0.6042 1>}
    font {
     bForceMonospaced=false
     bMaterialIsStencil=false
     fontMaterial=null
     fontObject=&Burbank
     letterSpacing=0
     monospacedWidth=1
     size=26
     skewAmount=0
     typefaceFontName=Burbank-Black
     outlineSettings {
      bApplyOutlineToDropShadows=false
      bMiteredCorners=false
      bSeparateFillAlpha=false
      outlineMaterial=null
      outlineSize=0
      outlineColor=<rgba 0 0 0 1>
     }
    }
    shadowOffset=<0 0>
   }
   slot {class=ButtonSlot, properties {horizontalAlignment=HAlign_Fill, verticalAlignment=VAlign_Fill, padding=<ltrb 0 0 0 0>}}
  }
  CanvasPanel {
   widgetClassPath=&CanvasPanel
   parent=HorizontalBox_0
   widget=CanvasPanel
   slot {class=HorizontalBoxSlot, properties {size {sizeRule=Fill, value=1}}}
  }
  Overlay {
   widgetClassPath=&Overlay
   parent=CanvasPanel
   widget=Overlay
   slot {class=CanvasPanelSlot, properties {layoutData {alignment=<0 0>, anchors {maximum=<1 1>, minimum=<0 0>}, offsets=<ltrb 0 0 0 0>}}}
  }
  Border {
   widgetClassPath=&Border
   parent=Overlay
   widget=Border
   properties {brushColor=<rgba 0.026 0.026 0.026 1>}
   slot {class=OverlaySlot, properties {horizontalAlignment=HAlign_Fill, verticalAlignment=VAlign_Fill}}
  }
  MainTableSwitcher {
   widgetClassPath=&CommonActivatableWidgetSwitcher
   bIsVariable=true
   parent=Overlay
   widget=MainTableSwitcher
   slot {class=OverlaySlot, properties {horizontalAlignment=HAlign_Fill, verticalAlignment=VAlign_Fill}}
  }
  UWBT_3DHB_Initializer {
   widgetClassPath=/Game/Tools/3DHB/Initializer/UWBT_3DHB_Initializer
   bIsVariable=true
   parent=MainTableSwitcher
   widget=UWBT_3DHB_Initializer
   slot {class=WidgetSwitcherSlot}
  }
  UWBT_3DHB_TextureBaker {
   widgetClassPath=/Game/Tools/3DHB/TextureBaker/UWBT_3DHB_TextureBaker
   bIsVariable=true
   parent=MainTableSwitcher
   widget=UWBT_3DHB_TextureBaker
   slot {class=WidgetSwitcherSlot}
  }
  UWBT_3DHB_MaterialMaker {
   widgetClassPath=/Game/Tools/3DHB/MaterialMaker/UWBT_3DHB_MaterialMaker
   bIsVariable=true
   parent=MainTableSwitcher
   widget=UWBT_3DHB_MaterialMaker
   slot {class=WidgetSwitcherSlot}
  }
  UWBT_3DHB_Previewer {
   widgetClassPath=/Game/Tools/3DHB/Previewer/UWBT_3DHB_Previewer
   bIsVariable=true
   parent=MainTableSwitcher
   widget=UWBT_3DHB_Previewer
   slot {class=WidgetSwitcherSlot}
  }
 ]
}
